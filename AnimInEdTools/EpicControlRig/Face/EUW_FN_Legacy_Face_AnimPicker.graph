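# Editor Utility Widget that serves as the legacy Fortnite face animation picker, a child of EUW_FN_Face_Base with sliders and buttons over a face image.
# On load it categorizes canvas children into 1D and 2D sliders, maps buttons to Control Rig control names, and hard-codes upper and lower face slider groups.
# Buttons select all controls or either region, and zero all or selected sliders through parent-class functions.
# Debug prints remain on mouse enter and sync.
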EditorUtilityWidgetBlueprint /AnimInEdTools/EpicControlRig/Face/EUW_FN_Legacy_Face_AnimPicker
graphs [
 EventGraph {nodes [
   "Custom on Clicked Event (SelectAll)"() {is=title K2Node_ComponentBoundEvent}
   > Select All Controls(self:EUW FN Face Base Object Reference)
   "Custom on Clicked Event (UpperFace)"() {is=title K2Node_ComponentBoundEvent}
   > Select Multiple Sliders!(
    execute
    self:EUW FN Face Base Object Reference
    Button:EUW Picker Element Object Reference=Get UpperFace() -> (UpperFace:EUW Picker Button Object Reference)
    2D Sliders to Select:Array of EUW Picker 2DSlider Object References=Get UpperFace2DSliders() -> (UpperFace2DSliders:Array of EUW Picker 2DSlider Object References)
    1D Sliders to Select:Array of EUW Picker 1DSlider Object References=Get UpperFace1DSliders() -> (UpperFace1DSliders:Array of EUW Picker 1DSlider Object References)
   ) -> (then)
   "Custom on Clicked Event (LowerFace)"() {is=title K2Node_ComponentBoundEvent}
   > Select Multiple Sliders(
    Button=Get LowerFace() -> (LowerFace:EUW Picker Button Object Reference)
    2D Sliders to Select=Get LowerFace2DSliders() -> (LowerFace2DSliders:Array of EUW Picker 2DSlider Object References)
    1D Sliders to Select=Get LowerFace1DSliders() -> (LowerFace1DSliders:Array of EUW Picker 1DSlider Object References)
   )
   "On Clicked (ZeroAll)"() {is=title K2Node_ComponentBoundEvent}
   > Set Zero Sliders!(
    execute
    self:EUW FN Face Base Object Reference
    Button Ctrl Pairs:Map of Widget Object References to Names=Get Button Control Pairs() -> (Button Control Pairs:Map of Widget Object References to Names)
    Selected Ctrl Only:bool=false
   ) -> (then) {is=title K2Node_CallFunction}
   "On Clicked (ZeroSelected)"() {is=title K2Node_ComponentBoundEvent}
   > Set Zero Sliders(Button Ctrl Pairs=Get Button Control Pairs(), Selected Ctrl Only=true)
   n1 = Event Widget has loaded!() -> (
    OutputDelegate
    then
    Control Rig Proxy Binding:Control Rig Sequencer Binding Proxy Structure
    Animatable Creature:Animatable Creature Object Reference
    Anim Picker EUW:Animation Picker Object Reference
   ) {is=title K2Node_Event}
   > Parent: Widget has loaded!(
    execute
    Control Rig Proxy Binding:Control Rig Sequencer Binding Proxy Structure=@n1.Control Rig Proxy Binding
    Animatable Creature:Animatable Creature Object Reference=@n1.Animatable Creature
    Anim Picker EUW:Animation Picker Object Reference=@n1.Anim Picker EUW
   ) -> (then) {is=title K2Node_CallParentFunction}
   > n2 = Sequence()
   > Set Face Core Variables(self:EUW FN Face Base Object Reference) {is=title K2Node_CallFunction}
   > n3 = Categorize Face Ctrls!(
    execute
    self:EUW FN Face Base Object Reference
    Controls:Array of Widget Object References=GetAllChildren(self=Get Legacy_Face_CanvasPanel() -> (Legacy_Face_CanvasPanel:Canvas Panel Object Reference))
   ) -> (then, 1D Sliders:Array of EUW Picker 1DSlider Object References, 2D Sliders:Array of EUW Picker 2DSlider Object References)
   > Set All 1DSliders(All 1DSliders:Array of EUW Picker 1DSlider Object References=@n3) -> (Output_Get:Array of EUW Picker 1DSlider Object References)
   > Set All 2DSliders(All 2DSliders:Array of EUW Picker 2DSlider Object References=@n3.2D Sliders) -> (Output_Get:Array of EUW Picker 2DSlider Object References)
   > n4 = Setup Face MR Naming For Buttons!(
    execute
    self:EUW FN Face Base Object Reference
    Main Canvas:Panel Widget Object Reference=Get Legacy_Face_CanvasPanel()
    Control Rig:Control Rig Object Reference=Get Control Rig() -> (Control Rig:Control Rig Object Reference)
    Excluded Buttons:Array of Widget Object References
   ) -> (then, NewParam:Map of Widget Object References to Names)
   > Set Button Control Pairs(Button Control Pairs:Map of Widget Object References to Names=@n4) -> (Output_Get:Map of Widget Object References to Names)
   Set LowerFace2DSliders(
    execute=@n2.then_1
    LowerFace2DSliders:Array of EUW Picker 2DSlider Object References=Make Array!(
     "[0]":EUW Picker 2DSlider Object Reference=Get tongue() -> (tongue:EUW Picker 2DSlider Object Reference)
     "[1]":EUW Picker 2DSlider Object Reference=Get L_Lipcorner() -> (L_Lipcorner:EUW Picker 2DSlider Object Reference)
     "[2]":EUW Picker 2DSlider Object Reference=Get Jaw() -> (Jaw:EUW Picker 2DSlider Object Reference)
     "[3]":EUW Picker 2DSlider Object Reference=Get R_Lipcorner() -> (R_Lipcorner:EUW Picker 2DSlider Object Reference)
    ) -> (Array:Array of EUW Picker 2DSlider Object References)
   ) -> (Output_Get:Array of EUW Picker 2DSlider Object References)
   > Set LowerFace1DSliders(LowerFace1DSliders:Array of EUW Picker 1DSlider Object References=Make Array!(
     "[0]":EUW Picker 1DSlider Object Reference=Get Mouth() -> (Mouth:EUW Picker 1DSlider Object Reference)
     "[1]":EUW Picker 1DSlider Object Reference=Get R_Lip_Upper() -> (R_Lip_Upper:EUW Picker 1DSlider Object Reference)
     "[2]":EUW Picker 1DSlider Object Reference=Get R_Lip_Lower() -> (R_Lip_Lower:EUW Picker 1DSlider Object Reference)
     "[3]":EUW Picker 1DSlider Object Reference=Get L_Lip_Upper() -> (L_Lip_Upper:EUW Picker 1DSlider Object Reference)
     "[4]":EUW Picker 1DSlider Object Reference=Get L_Lip_Lower() -> (L_Lip_Lower:EUW Picker 1DSlider Object Reference)
     "[5]":EUW Picker 1DSlider Object Reference=Get Jaw_FwdBk() -> (Jaw_FwdBk:EUW Picker 1DSlider Object Reference)
     "[6]":EUW Picker 1DSlider Object Reference=Get Lip_Upper() -> (Lip_Upper:EUW Picker 1DSlider Object Reference)
     "[7]":EUW Picker 1DSlider Object Reference=Get Lip_Lower() -> (Lip_Lower:EUW Picker 1DSlider Object Reference)
     "[8]":EUW Picker 1DSlider Object Reference=Get Jaw_FwdBk()
    ) -> (Array:Array of EUW Picker 1DSlider Object References)) -> (Output_Get:Array of EUW Picker 1DSlider Object References)
   > Set UpperFace2DSliders(UpperFace2DSliders:Array of EUW Picker 2DSlider Object References=Make Array!(
     "[0]":EUW Picker 2DSlider Object Reference=Get L_Brow() -> (L_Brow:EUW Picker 2DSlider Object Reference)
     "[1]":EUW Picker 2DSlider Object Reference=Get R_Eyelid() -> (R_Eyelid:EUW Picker 2DSlider Object Reference)
     "[2]":EUW Picker 2DSlider Object Reference=Get R_Cheek() -> (R_Cheek:EUW Picker 2DSlider Object Reference)
     "[3]":EUW Picker 2DSlider Object Reference=Get R_Brow() -> (R_Brow:EUW Picker 2DSlider Object Reference)
     "[4]":EUW Picker 2DSlider Object Reference=Get L_Eyelid() -> (L_Eyelid:EUW Picker 2DSlider Object Reference)
     "[5]":EUW Picker 2DSlider Object Reference=Get L_Cheek() -> (L_Cheek:EUW Picker 2DSlider Object Reference)
     "[6]":EUW Picker 2DSlider Object Reference=Get Look() -> (Look:EUW Picker 2DSlider Object Reference)
    ) -> (Array:Array of EUW Picker 2DSlider Object References)) -> (Output_Get:Array of EUW Picker 2DSlider Object References)
   > Set UpperFace1DSliders(UpperFace1DSliders:Array of EUW Picker 1DSlider Object References=Make Array!(
     "[0]":EUW Picker 1DSlider Object Reference=Get Nose() -> (Nose:EUW Picker 1DSlider Object Reference)
     "[1]":EUW Picker 1DSlider Object Reference=Get midbrow() -> (midbrow:EUW Picker 1DSlider Object Reference)
    ) -> (Array:Array of EUW Picker 1DSlider Object References)) -> (Output_Get:Array of EUW Picker 1DSlider Object References)
   n5 = Event Mouse Enter!() -> (OutputDelegate, then, Mouse Event:Pointer Event Structure) {is=title K2Node_Event}
   > Parent: Mouse Enter!(execute, Mouse Event:Pointer Event Structure=@n5.Mouse Event) -> (then) {is=title K2Node_CallParentFunction}
   > PrintString(InString=mouse enter event)
   Event Update From Scene() {is=title K2Node_Event}
   > Parent: Update From Scene() {is=title K2Node_CallParentFunction}
   Make Array(
    "[0]":EUW Picker Button Object Reference=Get R_Eye_Aim() -> (R_Eye_Aim:EUW Picker Button Object Reference)
    "[1]":EUW Picker Button Object Reference=Get L_Eye_Aim() -> (L_Eye_Aim:EUW Picker Button Object Reference)
   ) -> (Array:Array of EUW Picker Button Object References)
  ]}
 Get Primary Panel {nodes [
   Get Primary Panel() {is=title K2Node_FunctionEntry}
   > Return Node!(execute, Panel:Panel Widget Object Reference=Get Legacy_Face_CanvasPanel())
  ]}
 Get all Selectable Elements {nodes [
   Get all Selectable Elements() {is=title K2Node_FunctionEntry}
   > Return Node!(execute, All Elements:Array of EUW Picker Element Object References)
  ]}
 Legacy Sync Controls to Buttons {nodes [
   Legacy Sync Controls to Buttons() {is=title K2Node_FunctionEntry}
   > Sync Controls to Buttons!(execute, self:EUW FN Face Base Object Reference, Button Ctrl Pairs:Map of Widget Object References to Names=Get Button Control Pairs()) -> (then)
   > PrintString(InString=3333)
  ]}
]
variables [
 All 1DSliders {type="array<EUW_Picker_1DSlider_C>", default []}
 All 2DSliders {type="array<EUW_Picker_2DSlider_C>", default []}
 Button Control Pairs {default {}}
 CurrentSelection {type="array<string>", default []}
 LowerFace1DSliders {type="array<EUW_Picker_1DSlider_C>", default []}
 LowerFace2DSliders {type="array<EUW_Picker_2DSlider_C>", default []}
 Prefix {type=string, default=""}
 Upper Face Buttons {type="array<EUW_Picker_Button_C>", default []}
 UpperFace1DSliders {type="array<EUW_Picker_1DSlider_C>", default []}
 UpperFace2DSliders {type="array<EUW_Picker_2DSlider_C>", default []}
]
widgets {
 info {parentClass=&/EUW_FN_Face_Base.EUW_FN_Face_Base_C, rootWidgetClass=&CanvasPanel, widgetCount=39}
 widgets [
  Legacy_Face_CanvasPanel {widgetClassPath=&CanvasPanel, bIsVariable=true, widget=Legacy_Face_CanvasPanel, properties {renderTransformPivot=<0 0>}}
  R_Eyelid {
   widgetClassPath=/AnimInEdTools/AnimationPicker/PickerElements/EUW_Picker_2DSlider
   bIsVariable=true
   parent=Legacy_Face_CanvasPanel
   widget=R_Eyelid
   properties {bOverride_Cursor=true, max Value Y=-1, min Value Y=1, colorAndOpacity=<rgba 1 0.0521 0.0739 1>}
   slot {
    class=CanvasPanelSlot
    properties {layoutData {alignment=<0.5 0.5>, anchors {maximum=<0.5 0.5>, minimum=<0.5 0.5>}, offsets=<ltrb -313.4423 -156.422 55.0373 56.237>}}
   }
  }
  L_Eyelid {
   widgetClassPath=&EUW_Picker_2DSlider
   bIsVariable=true
   parent=Legacy_Face_CanvasPanel
   widget=L_Eyelid
   properties {
    bOverride_Cursor=true
    max Value Y=-1
    min Value Y=1
    colorAndOpacity=<rgba 0.2613 1 0.1894 1>
    renderTransform {angle=0, scale=<-1 1>, shear=<0 0>, translation=<0 0>}
   }
   slot {
    class=CanvasPanelSlot
    properties {layoutData {alignment=<0.5 0.5>, anchors {maximum=<0.5 0.5>, minimum=<0.5 0.5>}, offsets=<ltrb -53.4423 -152.422 55.0373 56.237>}}
   }
  }
  Look {
   widgetClassPath=&EUW_Picker_2DSlider
   bIsVariable=true
   parent=Legacy_Face_CanvasPanel
   widget=Look
   properties {
    bOverride_Cursor=true
    max Value Y=-1
    min Value Y=1
    colorAndOpacity=<rgba 1 0.8354 0 1>
    renderTransform {angle=0, scale=<-1 1>, shear=<0 0>, translation=<0 0>}
   }
   slot {
    class=CanvasPanelSlot
    properties {layoutData {alignment=<0.5 0.5>, anchors {maximum=<0.5 0.5>, minimum=<0.5 0.5>}, offsets=<ltrb -185.4423 -156.422 70 70>}}
   }
  }
  Jaw {
   widgetClassPath=&EUW_Picker_2DSlider
   bIsVariable=true
   parent=Legacy_Face_CanvasPanel
   widget=Jaw
   properties {
    bOverride_Cursor=true
    max Value Y=-1
    min Value Y=1
    colorAndOpacity=<rgba 1 0.8564 0 1>
    renderTransform {angle=0, scale=<-1 1>, shear=<0 0>, translation=<0 0>}
   }
   slot {
    class=CanvasPanelSlot
    properties {layoutData {alignment=<0.5 0.5>, anchors {maximum=<0.5 0.5>, minimum=<0.5 0.5>}, offsets=<ltrb -184.9609 243.4595 120 120>}}
   }
  }
  tongue {
   widgetClassPath=&EUW_Picker_2DSlider
   bIsVariable=true
   parent=Legacy_Face_CanvasPanel
   widget=tongue
   properties {bOverride_Cursor=true, max Value Y=-1, min Value Y=1, colorAndOpacity=<rgba 1 0.8564 0 1>}
   slot {
    class=CanvasPanelSlot
    properties {layoutData {alignment=<0.5 0.5>, anchors {maximum=<0.5 0.5>, minimum=<0.5 0.5>}, offsets=<ltrb -37.2743 261.1135 79.3733 75.308>}}
   }
  }
  L_Lipcorner {
   widgetClassPath=&EUW_Picker_2DSlider
   bIsVariable=true
   parent=Legacy_Face_CanvasPanel
   widget=L_Lipcorner
   properties {
    bOverride_Cursor=true
    max Value Y=-1
    min Value Y=1
    colorAndOpacity=<rgba 0.2613 1 0.1894 1>
    renderTransform {angle=0, scale=<-1 1>, shear=<0 0>, translation=<0 0>}
   }
   slot {
    class=CanvasPanelSlot
    properties {layoutData {alignment=<0.5 0.5>, anchors {maximum=<0.5 0.5>, minimum=<0.5 0.5>}, offsets=<ltrb -35.6844 68 79.3733 75.308>}}
   }
  }
  L_Cheek {
   widgetClassPath=&EUW_Picker_2DSlider
   bIsVariable=true
   parent=Legacy_Face_CanvasPanel
   widget=L_Cheek
   properties {
    bOverride_Cursor=true
    max Value Y=-1
    min Value Y=1
    colorAndOpacity=<rgba 0.2613 1 0.1894 1>
    renderTransform {angle=0, scale=<-1 1>, shear=<0 0>, translation=<0 0>}
   }
   slot {
    class=CanvasPanelSlot
    properties {layoutData {alignment=<0.5 0.5>, anchors {maximum=<0.5 0.5>, minimum=<0.5 0.5>}, offsets=<ltrb -25.0321 -66 79.3733 75.308>}}
   }
  }
  R_Lipcorner {
   widgetClassPath=&EUW_Picker_2DSlider
   bIsVariable=true
   parent=Legacy_Face_CanvasPanel
   widget=R_Lipcorner
   properties {bOverride_Cursor=true, max Value Y=-1, min Value Y=1, colorAndOpacity=<rgba 1 0.0521 0.0739 1>}
   slot {
    class=CanvasPanelSlot
    properties {layoutData {alignment=<0.5 0.5>, anchors {maximum=<0.5 0.5>, minimum=<0.5 0.5>}, offsets=<ltrb -336.2587 68 79.3733 75.308>}}
   }
  }
  R_Cheek {
   widgetClassPath=&EUW_Picker_2DSlider
   bIsVariable=true
   parent=Legacy_Face_CanvasPanel
   widget=R_Cheek
   properties {bOverride_Cursor=true, max Value Y=-1, min Value Y=1, colorAndOpacity=<rgba 1 0.0521 0.0739 1>}
   slot {
    class=CanvasPanelSlot
    properties {layoutData {alignment=<0.5 0.5>, anchors {maximum=<0.5 0.5>, minimum=<0.5 0.5>}, offsets=<ltrb -342.4164 -66 79.3733 75.308>}}
   }
  }
  L_Brow {
   widgetClassPath=&EUW_Picker_2DSlider
   bIsVariable=true
   parent=Legacy_Face_CanvasPanel
   widget=L_Brow
   properties {
    bOverride_Cursor=true
    max Value Y=-1
    min Value Y=1
    colorAndOpacity=<rgba 0.2613 1 0.1894 1>
    renderTransform {angle=0, scale=<-1 1>, shear=<0 0>, translation=<0 0>}
   }
   slot {
    class=CanvasPanelSlot
    properties {layoutData {alignment=<0.5 0.5>, anchors {maximum=<0.5 0.5>, minimum=<0.5 0.5>}, offsets=<ltrb -61.005 -310 75 75.308>}}
   }
  }
  R_Brow {
   widgetClassPath=&EUW_Picker_2DSlider
   bIsVariable=true
   parent=Legacy_Face_CanvasPanel
   widget=R_Brow
   properties {bOverride_Cursor=true, max Value Y=-1, min Value Y=1, colorAndOpacity=<rgba 1 0.0521 0.0739 1>}
   slot {
    class=CanvasPanelSlot
    properties {layoutData {alignment=<0.5 0.5>, anchors {maximum=<0.5 0.5>, minimum=<0.5 0.5>}, offsets=<ltrb -307.8622 -310 75 75.308>}}
   }
  }
  Graphics {
   widgetClassPath=&CanvasPanel
   parent=Legacy_Face_CanvasPanel
   widget=Graphics
   properties {renderTransform {angle=0, scale=<1 1>, shear=<0 0>, translation=<7.0547 -1.4777>}}
   slot {
    class=CanvasPanelSlot
    properties {zOrder=-1, layoutData {alignment=<0.5 0.5>, anchors {maximum=<1 1>, minimum=<0 0>}, offsets=<ltrb 0 0 0 0>}}
   }
  }
  Face_Image {
   widgetClassPath=&Image
   bIsVariable=true
   parent=Graphics
   widget=Face_Image
   properties {
    brush {
     bIsDynamicallyLoaded=false
     drawAs=Image
     imageType=FullColor
     mirroring=NoMirror
     resourceName=None
     resourceObject=/Game/Characters/Player/Common/SharedRigs/Graphics/Picker_Biped_Jonesy_Face
     tiling=NoTile
     imageSize=<412 611>
     margin=<ltrb 0 0 0 0>
     outlineSettings {
      bUseBrushTransparency=false
      roundingType=HalfHeightRadius
      width=0
      color {colorUseRule=UseColor_Specified, specifiedColor=<rgba 0 0 0 0>}
      cornerRadii=<0 0 0 0>
     }
     tintColor {colorUseRule=UseColor_Specified, specifiedColor=<rgba 1 1 1 1>}
     uVRegion {bIsValid=false, max=<0 0>, min=<0 0>}
    }
    colorAndOpacity=<rgba 1 1 1 0.4>
    renderTransform {angle=0, scale=<1 1>, shear=<0 0>, translation=<-3.575 -3.0507>}
   }
   slot {
    class=CanvasPanelSlot
    properties {layoutData {alignment=<0.5 0.5>, anchors {maximum=<0.5 0.5>, minimum=<0.5 0.5>}, offsets=<ltrb -196.2385 -60.9804 681.4449 839.1203>}}
   }
  }
  TextBlock_1 {
   widgetClassPath=&TextBlock
   parent=Graphics
   widget=TextBlock_1
   properties {
    text=Tongue
    colorAndOpacity {colorUseRule=UseColor_Specified, specifiedColor=<rgba 1 1 1 0.5>}
    font {
     bForceMonospaced=false
     bMaterialIsStencil=false
     fontMaterial=null
     fontObject=&Roboto
     letterSpacing=0
     monospacedWidth=1
     size=18.75
     skewAmount=0
     typefaceFontName=Bold
     outlineSettings {
      bApplyOutlineToDropShadows=false
      bMiteredCorners=false
      bSeparateFillAlpha=false
      outlineMaterial=null
      outlineSize=0
      outlineColor=<rgba 0 0 0 1>
     }
    }
   }
   slot {
    class=CanvasPanelSlot
    properties {bAutoSize=true, layoutData {alignment=<0 0>, anchors {maximum=<0.5 0.5>, minimum=<0.5 0.5>}, offsets=<ltrb -84.9609 311.4595 100 30>}}
   }
  }
  TextBlock {
   widgetClassPath=&TextBlock
   parent=Graphics
   widget=TextBlock
   properties {
    text=Jaw
    colorAndOpacity {colorUseRule=UseColor_Specified, specifiedColor=<rgba 1 1 1 0.5>}
    font {
     bForceMonospaced=false
     bMaterialIsStencil=false
     fontMaterial=null
     fontObject=&Roboto
     letterSpacing=0
     monospacedWidth=1
     size=18.75
     skewAmount=0
     typefaceFontName=Bold
     outlineSettings {
      bApplyOutlineToDropShadows=false
      bMiteredCorners=false
      bSeparateFillAlpha=false
      outlineMaterial=null
      outlineSize=0
      outlineColor=<rgba 0 0 0 1>
     }
    }
   }
   slot {
    class=CanvasPanelSlot
    properties {bAutoSize=true, layoutData {alignment=<0 0>, anchors {maximum=<0.5 0.5>, minimum=<0.5 0.5>}, offsets=<ltrb -212.9609 311.4595 100 30>}}
   }
  }
  TextBlock_2 {
   widgetClassPath=&TextBlock
   parent=Graphics
   widget=TextBlock_2
   properties {
    text=Face Mstr Ctrl
    colorAndOpacity {colorUseRule=UseColor_Specified, specifiedColor=<rgba 1 1 1 0.5>}
    font {
     bForceMonospaced=false
     bMaterialIsStencil=false
     fontMaterial=null
     fontObject=&Roboto
     letterSpacing=0
     monospacedWidth=1
     size=18.75
     skewAmount=0
     typefaceFontName=Bold
     outlineSettings {
      bApplyOutlineToDropShadows=false
      bMiteredCorners=false
      bSeparateFillAlpha=false
      outlineMaterial=null
      outlineSize=0
      outlineColor=<rgba 0 0 0 1>
     }
    }
   }
   slot {
    class=CanvasPanelSlot
    properties {bAutoSize=true, layoutData {alignment=<0 0>, anchors {maximum=<0.5 0.5>, minimum=<0.5 0.5>}, offsets=<ltrb -272.9609 415.4595 100 30>}}
   }
  }
  TextBlock_5 {
   widgetClassPath=&TextBlock
   parent=Graphics
   widget=TextBlock_5
   properties {
    text=Jaw FwdBk
    colorAndOpacity {colorUseRule=UseColor_Specified, specifiedColor=<rgba 1 1 1 0.5>}
    font {
     bForceMonospaced=false
     bMaterialIsStencil=false
     fontMaterial=null
     fontObject=&Roboto
     letterSpacing=0
     monospacedWidth=1
     size=18.75
     skewAmount=0
     typefaceFontName=Bold
     outlineSettings {
      bApplyOutlineToDropShadows=false
      bMiteredCorners=false
      bSeparateFillAlpha=false
      outlineMaterial=null
      outlineSize=0
      outlineColor=<rgba 0 0 0 1>
     }
    }
   }
   slot {
    class=CanvasPanelSlot
    properties {bAutoSize=true, layoutData {alignment=<0 0>, anchors {maximum=<0.5 0.5>, minimum=<0.5 0.5>}, offsets=<ltrb -424.9609 315.4595 100 30>}}
   }
  }
  Face_Ctrl_Master {
   widgetClassPath=/AnimInEdTools/AnimationPicker/EUW_Picker_Button
   bIsVariable=true
   parent=Legacy_Face_CanvasPanel
   widget=Face_Ctrl_Master
   properties {draw Style=Square, label Font Size=8, normal Tint {colorUseRule=UseColor_Specified, specifiedColor=<rgba 0 0.0108 0.4156 1>}}
   slot {
    class=CanvasPanelSlot
    properties {layoutData {alignment=<0.5 0.5>, anchors {maximum=<0.5 0.5>, minimum=<0.5 0.5>}, offsets=<ltrb -186.9609 393.4595 100 20>}}
   }
  }
  C_Eye_Aim {
   widgetClassPath=&EUW_Picker_Button
   bIsVariable=true
   parent=Legacy_Face_CanvasPanel
   widget=C_Eye_Aim
   properties {
    label Text=Aim
    colorAndOpacity=<rgba 1 0.9695 0 1>
    label Color {colorUseRule=UseColor_Specified, specifiedColor=<rgba 0 0 0 1>}
    normal Tint {colorUseRule=UseColor_Specified, specifiedColor=<rgba 0.6771 0.6771 0.6771 1>}
   }
   slot {
    class=CanvasPanelSlot
    properties {layoutData {alignment=<0.5 0.5>, anchors {maximum=<0.5 0.5>, minimum=<0.5 0.5>}, offsets=<ltrb -182.7069 -222.1633 76.5081 36.7544>}}
   }
  }
  L_Eye_Aim {
   widgetClassPath=&EUW_Picker_Button
   bIsVariable=true
   parent=Legacy_Face_CanvasPanel
   widget=L_Eye_Aim
   properties {
    draw Style=Circle
    label Font Size=6
    colorAndOpacity=<rgba 0.2613 1 0.1894 1>
    normal Tint {colorUseRule=UseColor_Specified, specifiedColor=<rgba 0.6094 0.6094 0.6094 1>}
   }
   slot {
    class=CanvasPanelSlot
    properties {zOrder=2, layoutData {alignment=<0.5 0.5>, anchors {maximum=<0.5 0.5>, minimum=<0.5 0.5>}, offsets=<ltrb -157.4609 -221.0405 15 15>}}
   }
  }
  R_Eye_Aim {
   widgetClassPath=&EUW_Picker_Button
   bIsVariable=true
   parent=Legacy_Face_CanvasPanel
   widget=R_Eye_Aim
   properties {
    draw Style=Circle
    label Font Size=6
    colorAndOpacity=<rgba 1 0.0521 0.0739 1>
    normal Tint {colorUseRule=UseColor_Specified, specifiedColor=<rgba 0.5625 0.5625 0.5625 1>}
   }
   slot {
    class=CanvasPanelSlot
    properties {zOrder=2, layoutData {alignment=<0.5 0.5>, anchors {maximum=<0.5 0.5>, minimum=<0.5 0.5>}, offsets=<ltrb -205.4609 -221.0405 15 15>}}
   }
  }
  midbrow {
   widgetClassPath=/AnimInEdTools/AnimationPicker/PickerElements/EUW_Picker_1DSlider
   bIsVariable=true
   parent=Legacy_Face_CanvasPanel
   widget=midbrow
   properties {
    min Value=-1
    colorAndOpacity=<rgba 1 0.8354 0 1>
    forward vector=<1 0>
    renderTransform {angle=-90, scale=<1 1>, shear=<0 0>, translation=<0 0>}
   }
   slot {class=CanvasPanelSlot, properties {layoutData {alignment=<0 0>, anchors {maximum=<0 0>, minimum=<0 0>}, offsets=<ltrb 724 220 75 20>}}}
  }
  Jaw_FwdBk {
   widgetClassPath=&EUW_Picker_1DSlider
   bIsVariable=true
   parent=Legacy_Face_CanvasPanel
   widget=Jaw_FwdBk
   properties {
    min Value=-1
    colorAndOpacity=<rgba 1 0.8354 0 1>
    forward vector=<1 0>
    renderTransform {angle=-90, scale=<1 1>, shear=<0 0>, translation=<0 0>}
   }
   slot {class=CanvasPanelSlot, properties {layoutData {alignment=<0 0>, anchors {maximum=<0 0>, minimum=<0 0>}, offsets=<ltrb 536 768 100 20>}}}
  }
  Nose {
   widgetClassPath=&EUW_Picker_1DSlider
   bIsVariable=true
   parent=Legacy_Face_CanvasPanel
   widget=Nose
   properties {colorAndOpacity=<rgba 1 0.8354 0 1>, forward vector=<1 0>, renderTransform {angle=-90, scale=<1 1>, shear=<0 0>, translation=<0 0>}}
   slot {class=CanvasPanelSlot, properties {layoutData {alignment=<0 0>, anchors {maximum=<0 0>, minimum=<0 0>}, offsets=<ltrb 736 460 60 20>}}}
  }
  Lip_Upper {
   widgetClassPath=&EUW_Picker_1DSlider
   bIsVariable=true
   parent=Legacy_Face_CanvasPanel
   widget=Lip_Upper
   properties {
    control name override=Lip_Upper
    min Value=-1
    colorAndOpacity=<rgba 0.4198 0.3435 0 1>
    renderTransform {angle=-90, scale=<1 1>, shear=<0 0>, translation=<0 0>}
   }
   slot {class=CanvasPanelSlot, properties {layoutData {alignment=<0 0>, anchors {maximum=<0 0>, minimum=<0 0>}, offsets=<ltrb 736 528 60 20>}}}
  }
  Lip_Lower {
   widgetClassPath=&EUW_Picker_1DSlider
   bIsVariable=true
   parent=Legacy_Face_CanvasPanel
   widget=Lip_Lower
   properties {
    control name override=Lip_Lower
    min Value=-1
    colorAndOpacity=<rgba 0.4198 0.3435 0 1>
    renderTransform {angle=90, scale=<1 1>, shear=<0 0>, translation=<0 0>}
   }
   slot {class=CanvasPanelSlot, properties {layoutData {alignment=<0 0>, anchors {maximum=<0 0>, minimum=<0 0>}, offsets=<ltrb 736 672 60 20>}}}
  }
  L_Lip_Upper {
   widgetClassPath=&EUW_Picker_1DSlider
   bIsVariable=true
   parent=Legacy_Face_CanvasPanel
   widget=L_Lip_Upper
   properties {
    min Value=-1
    colorAndOpacity=<rgba 0.2613 1 0.1894 1>
    forward vector=<1 0>
    renderTransform {angle=-45, scale=<1 1>, shear=<0 0>, translation=<0 0>}
   }
   slot {class=CanvasPanelSlot, properties {layoutData {alignment=<0 0>, anchors {maximum=<0 0>, minimum=<0 0>}, offsets=<ltrb 816 552 60 20>}}}
  }
  L_Lip_Lower {
   widgetClassPath=&EUW_Picker_1DSlider
   bIsVariable=true
   parent=Legacy_Face_CanvasPanel
   widget=L_Lip_Lower
   properties {
    min Value=-1
    colorAndOpacity=<rgba 0.2613 1 0.1894 1>
    forward vector=<1 0>
    renderTransform {angle=45, scale=<1 1>, shear=<0 0>, translation=<0 0>}
   }
   slot {class=CanvasPanelSlot, properties {layoutData {alignment=<0 0>, anchors {maximum=<0 0>, minimum=<0 0>}, offsets=<ltrb 816 632 60 20>}}}
  }
  R_Lip_Upper {
   widgetClassPath=&EUW_Picker_1DSlider
   bIsVariable=true
   parent=Legacy_Face_CanvasPanel
   widget=R_Lip_Upper
   properties {
    min Value=-1
    colorAndOpacity=<rgba 1 0.0521 0.0739 1>
    forward vector=<1 0>
    renderTransform {angle=-135, scale=<1 1>, shear=<0 0>, translation=<0 0>}
   }
   slot {class=CanvasPanelSlot, properties {layoutData {alignment=<0 0>, anchors {maximum=<0 0>, minimum=<0 0>}, offsets=<ltrb 660 552 60 20>}}}
  }
  R_Lip_Lower {
   widgetClassPath=&EUW_Picker_1DSlider
   bIsVariable=true
   parent=Legacy_Face_CanvasPanel
   widget=R_Lip_Lower
   properties {
    min Value=-1
    colorAndOpacity=<rgba 1 0.0521 0.0739 1>
    forward vector=<1 0>
    renderTransform {angle=135, scale=<1 1>, shear=<0 0>, translation=<0 0>}
   }
   slot {class=CanvasPanelSlot, properties {layoutData {alignment=<0 0>, anchors {maximum=<0 0>, minimum=<0 0>}, offsets=<ltrb 660 632 60 20>}}}
  }
  Mouth {
   widgetClassPath=&EUW_Picker_1DSlider
   bIsVariable=true
   parent=Legacy_Face_CanvasPanel
   widget=Mouth
   properties {
    min Value=-1
    colorAndOpacity=<rgba 1 0.8354 0 1>
    forward vector=<1 0>
    renderTransform {angle=-90, scale=<1 1>, shear=<0 0>, translation=<0 0>}
   }
   slot {class=CanvasPanelSlot, properties {layoutData {alignment=<0 0>, anchors {maximum=<0 0>, minimum=<0 0>}, offsets=<ltrb 736 600 60 20>}}}
  }
  SelectAll {
   widgetClassPath=&EUW_Picker_Button
   bIsVariable=true
   parent=Legacy_Face_CanvasPanel
   widget=SelectAll
   properties {
    label Font Size=15
    label Text=Select All
    toolTipText=Selects all Face Rig Controls
    normal Tint {colorUseRule=UseColor_Specified, specifiedColor=<rgba 0.625 0.2072 0.0359 1>}
   }
   slot {class=CanvasPanelSlot, properties {layoutData {alignment=<0 0>, anchors {maximum=<0 0>, minimum=<0 0>}, offsets=<ltrb 1156 172 150 35>}}}
  }
  UpperFace {
   widgetClassPath=&EUW_Picker_Button
   bIsVariable=true
   parent=Legacy_Face_CanvasPanel
   widget=UpperFace
   properties {
    label Font Size=15
    label Text=Upper Face
    toolTipText=Selects Upper Face Region Controls
    normal Tint {colorUseRule=UseColor_Specified, specifiedColor=<rgba 0.625 0.4915 0 1>}
   }
   slot {class=CanvasPanelSlot, properties {layoutData {alignment=<0 0>, anchors {maximum=<0 0>, minimum=<0 0>}, offsets=<ltrb 1156 340 150 35>}}}
  }
  LowerFace {
   widgetClassPath=&EUW_Picker_Button
   bIsVariable=true
   parent=Legacy_Face_CanvasPanel
   widget=LowerFace
   properties {
    label Font Size=15
    label Text=Lower Face
    toolTipText=Selects Lower Face Region Controls
    normal Tint {colorUseRule=UseColor_Specified, specifiedColor=<rgba 0.625 0.4915 0 1>}
   }
   slot {class=CanvasPanelSlot, properties {layoutData {alignment=<0 0>, anchors {maximum=<0 0>, minimum=<0 0>}, offsets=<ltrb 1156 396 150 35>}}}
  }
  ZeroAll {
   widgetClassPath=&EditorUtilityButton
   parent=Legacy_Face_CanvasPanel
   widget=ZeroAll
   properties {widgetStyle {
     clickedSlateSound {resourceObject=null}
     disabled {
      bIsDynamicallyLoaded=false
      drawAs=RoundedBox
      imageType=NoImage
      mirroring=NoMirror
      resourceName=None
      resourceObject=null
      tiling=NoTile
      imageSize=<0 0>
      margin=<ltrb 0 0 0 0>
      outlineSettings {
       bUseBrushTransparency=false
       roundingType=FixedRadius
       width=1
       color {colorUseRule=UseColor_ColorTable, specifiedColor=<rgba 0 0 0 0>}
       cornerRadii=<4 4 4 4>
      }
      tintColor {colorUseRule=UseColor_Specified, specifiedColor=<rgba 0.0395 0.0395 0.0395 1>}
      uVRegion {bIsValid=false, max=<0 0>, min=<0 0>}
     }
     disabledForeground {colorUseRule=UseColor_ColorTable, specifiedColor=<rgba 0 0 0 0>}
     hovered {
      bIsDynamicallyLoaded=false
      drawAs=RoundedBox
      imageType=NoImage
      mirroring=NoMirror
      resourceName=None
      resourceObject=null
      tiling=NoTile
      imageSize=<32 32>
      margin=<ltrb 0 0 0 0>
      outlineSettings {
       bUseBrushTransparency=false
       roundingType=FixedRadius
       width=1
       color {colorUseRule=UseColor_ColorTable, specifiedColor=<rgba 0 0 0 0>}
       cornerRadii=<4 4 4 4>
      }
      tintColor {colorUseRule=UseColor_Specified, specifiedColor=<rgba 0.0953 0.0953 0.0953 1>}
      uVRegion {bIsValid=false, max=<0 0>, min=<0 0>}
     }
     hoveredForeground {colorUseRule=UseColor_ColorTable, specifiedColor=<rgba 0 0 0 0>}
     hoveredSlateSound {resourceObject=null}
     normal {
      bIsDynamicallyLoaded=false
      drawAs=RoundedBox
      imageType=NoImage
      mirroring=NoMirror
      resourceName=None
      resourceObject=null
      tiling=NoTile
      imageSize=<32 32>
      margin=<ltrb 0 0 0 0>
      outlineSettings {
       bUseBrushTransparency=false
       roundingType=FixedRadius
       width=1
       color {colorUseRule=UseColor_ColorTable, specifiedColor=<rgba 0 0 0 0>}
       cornerRadii=<4 4 4 4>
      }
      tintColor {colorUseRule=UseColor_Specified, specifiedColor=<rgba 0.624 0.2086 0.0356 1>}
      uVRegion {bIsValid=false, max=<0 0>, min=<0 0>}
     }
     normalForeground {colorUseRule=UseColor_ColorTable, specifiedColor=<rgba 0 0 0 0>}
     normalPadding=<ltrb 12 1.5 12 1.5>
     pressed {
      bIsDynamicallyLoaded=false
      drawAs=RoundedBox
      imageType=NoImage
      mirroring=NoMirror
      resourceName=None
      resourceObject=null
      tiling=NoTile
      imageSize=<32 32>
      margin=<ltrb 0 0 0 0>
      outlineSettings {
       bUseBrushTransparency=false
       roundingType=FixedRadius
       width=1
       color {colorUseRule=UseColor_ColorTable, specifiedColor=<rgba 0 0 0 0>}
       cornerRadii=<4 4 4 4>
      }
      tintColor {colorUseRule=UseColor_Specified, specifiedColor=<rgba 0.0284 0.0284 0.0284 1>}
      uVRegion {bIsValid=false, max=<0 0>, min=<0 0>}
     }
     pressedForeground {colorUseRule=UseColor_ColorTable, specifiedColor=<rgba 0 0 0 0>}
     pressedPadding=<ltrb 12 2.5 12 0.5>
     pressedSlateSound {resourceObject=null}
    }}
   slot {
    class=CanvasPanelSlot
    properties {layoutData {alignment=<0 0>, anchors {maximum=<0 0>, minimum=<0 0>}, offsets=<ltrb 1156 228 150.0119 34.1656>}}
   }
  }
  TextBlock_0 {
   widgetClassPath=&TextBlock
   parent=ZeroAll
   widget=TextBlock_0
   properties {
    text=Zero All
    font {
     bForceMonospaced=false
     bMaterialIsStencil=false
     fontMaterial=null
     fontObject=&Roboto
     letterSpacing=0
     monospacedWidth=1
     size=15
     skewAmount=0
     typefaceFontName=Bold
     outlineSettings {
      bApplyOutlineToDropShadows=false
      bMiteredCorners=false
      bSeparateFillAlpha=false
      outlineMaterial=null
      outlineSize=0
      outlineColor=<rgba 0 0 0 1>
     }
    }
   }
   slot {class=ButtonSlot}
  }
  ZeroSelected {
   widgetClassPath=&EditorUtilityButton
   parent=Legacy_Face_CanvasPanel
   widget=ZeroSelected
   properties {widgetStyle {
     clickedSlateSound {resourceObject=null}
     disabled {
      bIsDynamicallyLoaded=false
      drawAs=RoundedBox
      imageType=NoImage
      mirroring=NoMirror
      resourceName=None
      resourceObject=null
      tiling=NoTile
      imageSize=<0 0>
      margin=<ltrb 0 0 0 0>
      outlineSettings {
       bUseBrushTransparency=false
       roundingType=FixedRadius
       width=1
       color {colorUseRule=UseColor_ColorTable, specifiedColor=<rgba 0 0 0 0>}
       cornerRadii=<4 4 4 4>
      }
      tintColor {colorUseRule=UseColor_Specified, specifiedColor=<rgba 0.0395 0.0395 0.0395 1>}
      uVRegion {bIsValid=false, max=<0 0>, min=<0 0>}
     }
     disabledForeground {colorUseRule=UseColor_ColorTable, specifiedColor=<rgba 0 0 0 0>}
     hovered {
      bIsDynamicallyLoaded=false
      drawAs=RoundedBox
      imageType=NoImage
      mirroring=NoMirror
      resourceName=None
      resourceObject=null
      tiling=NoTile
      imageSize=<32 32>
      margin=<ltrb 0 0 0 0>
      outlineSettings {
       bUseBrushTransparency=false
       roundingType=FixedRadius
       width=1
       color {colorUseRule=UseColor_ColorTable, specifiedColor=<rgba 0 0 0 0>}
       cornerRadii=<4 4 4 4>
      }
      tintColor {colorUseRule=UseColor_Specified, specifiedColor=<rgba 0.0953 0.0953 0.0953 1>}
      uVRegion {bIsValid=false, max=<0 0>, min=<0 0>}
     }
     hoveredForeground {colorUseRule=UseColor_ColorTable, specifiedColor=<rgba 0 0 0 0>}
     hoveredSlateSound {resourceObject=null}
     normal {
      bIsDynamicallyLoaded=false
      drawAs=RoundedBox
      imageType=NoImage
      mirroring=NoMirror
      resourceName=None
      resourceObject=null
      tiling=NoTile
      imageSize=<32 32>
      margin=<ltrb 0 0 0 0>
      outlineSettings {
       bUseBrushTransparency=false
       roundingType=FixedRadius
       width=1
       color {colorUseRule=UseColor_ColorTable, specifiedColor=<rgba 0 0 0 0>}
       cornerRadii=<4 4 4 4>
      }
      tintColor {colorUseRule=UseColor_Specified, specifiedColor=<rgba 0.624 0.2086 0.0356 1>}
      uVRegion {bIsValid=false, max=<0 0>, min=<0 0>}
     }
     normalForeground {colorUseRule=UseColor_ColorTable, specifiedColor=<rgba 0 0 0 0>}
     normalPadding=<ltrb 12 1.5 12 1.5>
     pressed {
      bIsDynamicallyLoaded=false
      drawAs=RoundedBox
      imageType=NoImage
      mirroring=NoMirror
      resourceName=None
      resourceObject=null
      tiling=NoTile
      imageSize=<32 32>
      margin=<ltrb 0 0 0 0>
      outlineSettings {
       bUseBrushTransparency=false
       roundingType=FixedRadius
       width=1
       color {colorUseRule=UseColor_ColorTable, specifiedColor=<rgba 0 0 0 0>}
       cornerRadii=<4 4 4 4>
      }
      tintColor {colorUseRule=UseColor_Specified, specifiedColor=<rgba 0.0284 0.0284 0.0284 1>}
      uVRegion {bIsValid=false, max=<0 0>, min=<0 0>}
     }
     pressedForeground {colorUseRule=UseColor_ColorTable, specifiedColor=<rgba 0 0 0 0>}
     pressedPadding=<ltrb 12 2.5 12 0.5>
     pressedSlateSound {resourceObject=null}
    }}
   slot {
    class=CanvasPanelSlot
    properties {layoutData {alignment=<0 0>, anchors {maximum=<0 0>, minimum=<0 0>}, offsets=<ltrb 1156 284 150.2979 34.5088>}}
   }
  }
  TextBlock_3 {
   widgetClassPath=&TextBlock
   parent=ZeroSelected
   widget=TextBlock_3
   properties {
    text=Zero selected
    font {
     bForceMonospaced=false
     bMaterialIsStencil=false
     fontMaterial=null
     fontObject=&Roboto
     letterSpacing=0
     monospacedWidth=1
     size=15
     skewAmount=0
     typefaceFontName=Bold
     outlineSettings {
      bApplyOutlineToDropShadows=false
      bMiteredCorners=false
      bSeparateFillAlpha=false
      outlineMaterial=null
      outlineSize=0
      outlineColor=<rgba 0 0 0 1>
     }
    }
   }
   slot {class=ButtonSlot}
  }
 ]
}
