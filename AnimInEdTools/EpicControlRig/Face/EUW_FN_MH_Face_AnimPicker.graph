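# Editor utility widget that serves as the FN MH face animation picker, a canvas of 1D and 2D sliders named after Control Rig controls.
# On load it categorizes sliders, maps buttons to controls and hides V2-only sliders for V1 faces.
# Buttons select all, upper or lower face, or zero sliders.
# An update function syncs slider values and selection from the Control Rig at the focused Level Sequence's current frame.
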EditorUtilityWidgetBlueprint /AnimInEdTools/EpicControlRig/Face/EUW_FN_MH_Face_AnimPicker
graphs [
 EventGraph {nodes [
   "Custom on Clicked Event (SelectAll)"() {is=title K2Node_ComponentBoundEvent}
   > Select All Controls(self:EUW FN Face Base Object Reference)
   "On Clicked (zeroAll)"() {is=title K2Node_ComponentBoundEvent}
   > Set Zero Sliders!(
    execute
    self:EUW FN Face Base Object Reference
    Button Ctrl Pairs:Map of Widget Object References to Names=Get Button Control Pairs() -> (Button Control Pairs:Map of Widget Object References to Names)
    Selected Ctrl Only:bool=false
   ) -> (then) {is=title K2Node_CallFunction}
   "On Clicked (zeroSelected)"() {is=title K2Node_ComponentBoundEvent}
   > Set Zero Sliders(Button Ctrl Pairs=Get Button Control Pairs(), Selected Ctrl Only=true)
   "Custom on Clicked Event (UpperFace)"() {is=title K2Node_ComponentBoundEvent}
   > Select Multiple Sliders!(
    execute
    self:EUW FN Face Base Object Reference
    Button:EUW Picker Element Object Reference=Get UpperFace() -> (UpperFace:EUW Picker Button Object Reference)
    2D Sliders to Select:Array of EUW Picker 2DSlider Object References=Get UpperFace2DSliders() -> (UpperFace2DSliders:Array of EUW Picker 2DSlider Object References)
    1D Sliders to Select:Array of EUW Picker 1DSlider Object References=Get UpperFace1DSliders() -> (UpperFace1DSliders:Array of EUW Picker 1DSlider Object References)
   ) -> (then)
   "Custom on Clicked Event (LowerFace)"() {is=title K2Node_ComponentBoundEvent}
   > Select Multiple Sliders(
    Button=Get LowerFace() -> (LowerFace:EUW Picker Button Object Reference)
    1D Sliders to Select=Get LowerFace1DSliders() -> (LowerFace1DSliders:Array of EUW Picker 1DSlider Object References)
   )
   n1 = Event Widget has loaded!() -> (
    OutputDelegate
    then
    Control Rig Proxy Binding:Control Rig Sequencer Binding Proxy Structure
    Animatable Creature:Animatable Creature Object Reference
    Anim Picker EUW:Animation Picker Object Reference
   ) {is=title K2Node_Event}
   > Parent: Widget has loaded!(
    execute
    Control Rig Proxy Binding:Control Rig Sequencer Binding Proxy Structure=@n1.Control Rig Proxy Binding
    Animatable Creature:Animatable Creature Object Reference=@n1.Animatable Creature
    Anim Picker EUW:Animation Picker Object Reference=@n1.Anim Picker EUW
   ) -> (then) {is=title K2Node_CallParentFunction}
   > Set Face Core Variables(self:EUW FN Face Base Object Reference) {is=title K2Node_CallFunction}
   > n2 = Sequence()
   n3 = Get FN_MH_Face_CanvasPanel() -> (FN_MH_Face_CanvasPanel:Canvas Panel Object Reference)
   > n4 = Categorize Face Ctrls!(execute, self:EUW FN Face Base Object Reference, Controls:Array of Widget Object References=GetAllChildren(self=@n3)) -> (then, 1D Sliders:Array of EUW Picker 1DSlider Object References, 2D Sliders:Array of EUW Picker 2DSlider Object References)
   > Set All 1DSliders(All 1DSliders:Array of EUW Picker 1DSlider Object References=@n4) -> (Output_Get:Array of EUW Picker 1DSlider Object References)
   > Set All 2DSliders(All 2DSliders:Array of EUW Picker 2DSlider Object References=@n4.2D Sliders) -> (Output_Get:Array of EUW Picker 2DSlider Object References)
   > n5 = Setup Face MR Naming For Buttons!(
    execute
    self:EUW FN Face Base Object Reference
    Main Canvas:Panel Widget Object Reference=@n3
    Control Rig:Control Rig Object Reference=Get Control Rig() -> (Control Rig:Control Rig Object Reference)
    Excluded Buttons:Array of Widget Object References
   ) -> (then, NewParam:Map of Widget Object References to Names)
   > Set Button Control Pairs(Button Control Pairs:Map of Widget Object References to Names=@n5) -> (Output_Get:Map of Widget Object References to Names)
   > Set V2 Button Visibility!(execute, self, visible:bool="Not Equal Exactly (String)"(A=Get Face Type() -> (Face Type:String), B=V1)) -> (then) {is=title K2Node_CallFunction}
   Set LowerFace1DSliders(
    execute=@n2.then_1
    LowerFace1DSliders:Array of EUW Picker 1DSlider Object References=Make Array!(
     "[0]":EUW Picker 1DSlider Object Reference=Get CTRL_R_mouth_upperLipRaise() -> (CTRL_R_mouth_upperLipRaise:EUW Picker 1DSlider Object Reference)
     "[1]":EUW Picker 1DSlider Object Reference=Get CTRL_R_mouth_towardsU() -> (CTRL_R_mouth_towardsU:EUW Picker 1DSlider Object Reference)
     "[2]":EUW Picker 1DSlider Object Reference=Get CTRL_R_mouth_towardsD() -> (CTRL_R_mouth_towardsD:EUW Picker 1DSlider Object Reference)
     "[3]":EUW Picker 1DSlider Object Reference=Get CTRL_R_mouth_suckBlow() -> (CTRL_R_mouth_suckBlow:EUW Picker 1DSlider Object Reference)
     "[4]":EUW Picker 1DSlider Object Reference=Get CTRL_R_mouth_stretch() -> (CTRL_R_mouth_stretch:EUW Picker 1DSlider Object Reference)
     "[5]":EUW Picker 1DSlider Object Reference=Get CTRL_R_mouth_sharpCornerPull() -> (CTRL_R_mouth_sharpCornerPull:EUW Picker 1DSlider Object Reference)
     "[6]":EUW Picker 1DSlider Object Reference=Get CTRL_R_mouth_purseU() -> (CTRL_R_mouth_purseU:EUW Picker 1DSlider Object Reference)
     "[7]":EUW Picker 1DSlider Object Reference=Get CTRL_R_mouth_purseD() -> (CTRL_R_mouth_purseD:EUW Picker 1DSlider Object Reference)
     "[8]":EUW Picker 1DSlider Object Reference=Get CTRL_R_mouth_lowerLipDepress() -> (CTRL_R_mouth_lowerLipDepress:EUW Picker 1DSlider Object Reference)
     "[9]":EUW Picker 1DSlider Object Reference=Get CTRL_R_mouth_lipsTowardsTeethU() -> (CTRL_R_mouth_lipsTowardsTeethU:EUW Picker 1DSlider Object Reference)
     "[10]":EUW Picker 1DSlider Object Reference=Get CTRL_R_mouth_lipsTowardsTeethD() -> (CTRL_R_mouth_lipsTowardsTeethD:EUW Picker 1DSlider Object Reference)
     "[11]":EUW Picker 1DSlider Object Reference=Get CTRL_R_mouth_lipsTogetherU() -> (CTRL_R_mouth_lipsTogetherU:EUW Picker 1DSlider Object Reference)
     "[12]":EUW Picker 1DSlider Object Reference=Get CTRL_R_mouth_lipsTogetherD() -> (CTRL_R_mouth_lipsTogetherD:EUW Picker 1DSlider Object Reference)
     "[13]":EUW Picker 1DSlider Object Reference=Get CTRL_R_mouth_lipsBlow() -> (CTRL_R_mouth_lipsBlow:EUW Picker 1DSlider Object Reference)
     "[14]":EUW Picker 1DSlider Object Reference=Get CTRL_R_mouth_lipBiteU() -> (CTRL_R_mouth_lipBiteU:EUW Picker 1DSlider Object Reference)
     "[15]":EUW Picker 1DSlider Object Reference=Get CTRL_R_mouth_lipBiteD() -> (CTRL_R_mouth_lipBiteD:EUW Picker 1DSlider Object Reference)
     "[16]":EUW Picker 1DSlider Object Reference=Get CTRL_R_mouth_funnelU() -> (CTRL_R_mouth_funnelU:EUW Picker 1DSlider Object Reference)
     "[17]":EUW Picker 1DSlider Object Reference=Get CTRL_R_mouth_dimple() -> (CTRL_R_mouth_dimple:EUW Picker 1DSlider Object Reference)
     "[18]":EUW Picker 1DSlider Object Reference=Get CTRL_R_mouth_cornerSharpnessU() -> (CTRL_R_mouth_cornerSharpnessU:EUW Picker 1DSlider Object Reference)
     "[19]":EUW Picker 1DSlider Object Reference=Get CTRL_R_mouth_cornerSharpnessD() -> (CTRL_R_mouth_cornerSharpnessD:EUW Picker 1DSlider Object Reference)
     "[20]":EUW Picker 1DSlider Object Reference=Get CTRL_R_mouth_cornerDepress() -> (CTRL_R_mouth_cornerDepress:EUW Picker 1DSlider Object Reference)
     "[21]":EUW Picker 1DSlider Object Reference=Get CTRL_R_jaw_ChinRaiseU() -> (CTRL_R_jaw_ChinRaiseU:EUW Picker 1DSlider Object Reference)
     "[22]":EUW Picker 1DSlider Object Reference=Get CTRL_R_jaw_ChinRaiseD() -> (CTRL_R_jaw_ChinRaiseD:EUW Picker 1DSlider Object Reference)
     "[23]":EUW Picker 1DSlider Object Reference=Get CTRL_L_mouth_upperLipRaise() -> (CTRL_L_mouth_upperLipRaise:EUW Picker 1DSlider Object Reference)
     "[24]":EUW Picker 1DSlider Object Reference=Get CTRL_L_mouth_towardsU() -> (CTRL_L_mouth_towardsU:EUW Picker 1DSlider Object Reference)
     "[25]":EUW Picker 1DSlider Object Reference=Get CTRL_L_mouth_towardsD() -> (CTRL_L_mouth_towardsD:EUW Picker 1DSlider Object Reference)
     "[26]":EUW Picker 1DSlider Object Reference=Get CTRL_L_mouth_suckBlow() -> (CTRL_L_mouth_suckBlow:EUW Picker 1DSlider Object Reference)
     "[27]":EUW Picker 1DSlider Object Reference=Get CTRL_L_mouth_stretch() -> (CTRL_L_mouth_stretch:EUW Picker 1DSlider Object Reference)
     "[28]":EUW Picker 1DSlider Object Reference=Get CTRL_L_mouth_purseU() -> (CTRL_L_mouth_purseU:EUW Picker 1DSlider Object Reference)
     "[29]":EUW Picker 1DSlider Object Reference=Get CTRL_L_mouth_sharpCornerPull() -> (CTRL_L_mouth_sharpCornerPull:EUW Picker 1DSlider Object Reference)
     "[30]":EUW Picker 1DSlider Object Reference=Get CTRL_L_mouth_purseD() -> (CTRL_L_mouth_purseD:EUW Picker 1DSlider Object Reference)
     "[31]":EUW Picker 1DSlider Object Reference=Get CTRL_L_mouth_pressU() -> (CTRL_L_mouth_pressU:EUW Picker 1DSlider Object Reference)
     "[32]":EUW Picker 1DSlider Object Reference=Get CTRL_L_mouth_pressD() -> (CTRL_L_mouth_pressD:EUW Picker 1DSlider Object Reference)
     "[33]":EUW Picker 1DSlider Object Reference=Get CTRL_L_mouth_lowerLipDepress() -> (CTRL_L_mouth_lowerLipDepress:EUW Picker 1DSlider Object Reference)
     "[34]":EUW Picker 1DSlider Object Reference=Get CTRL_L_mouth_lipsTowardsTeethU() -> (CTRL_L_mouth_lipsTowardsTeethU:EUW Picker 1DSlider Object Reference)
     "[35]":EUW Picker 1DSlider Object Reference=Get CTRL_L_mouth_lipsTowardsTeethD() -> (CTRL_L_mouth_lipsTowardsTeethD:EUW Picker 1DSlider Object Reference)
     "[36]":EUW Picker 1DSlider Object Reference=Get CTRL_L_mouth_lipsTogetherU() -> (CTRL_L_mouth_lipsTogetherU:EUW Picker 1DSlider Object Reference)
     "[37]":EUW Picker 1DSlider Object Reference=Get CTRL_L_mouth_lipsTogetherD() -> (CTRL_L_mouth_lipsTogetherD:EUW Picker 1DSlider Object Reference)
     "[38]":EUW Picker 1DSlider Object Reference=Get CTRL_L_mouth_lipsBlow() -> (CTRL_L_mouth_lipsBlow:EUW Picker 1DSlider Object Reference)
     "[39]":EUW Picker 1DSlider Object Reference=Get CTRL_L_mouth_lipBiteU() -> (CTRL_L_mouth_lipBiteU:EUW Picker 1DSlider Object Reference)
     "[40]":EUW Picker 1DSlider Object Reference=Get CTRL_L_mouth_lipBiteD() -> (CTRL_L_mouth_lipBiteD:EUW Picker 1DSlider Object Reference)
     "[41]":EUW Picker 1DSlider Object Reference=Get CTRL_L_mouth_funnelU() -> (CTRL_L_mouth_funnelU:EUW Picker 1DSlider Object Reference)
     "[42]":EUW Picker 1DSlider Object Reference=Get CTRL_L_mouth_funnelD() -> (CTRL_L_mouth_funnelD:EUW Picker 1DSlider Object Reference)
     "[43]":EUW Picker 1DSlider Object Reference=Get CTRL_L_mouth_dimple() -> (CTRL_L_mouth_dimple:EUW Picker 1DSlider Object Reference)
     "[44]":EUW Picker 1DSlider Object Reference=Get CTRL_L_mouth_cornerSharpnessU() -> (CTRL_L_mouth_cornerSharpnessU:EUW Picker 1DSlider Object Reference)
     "[45]":EUW Picker 1DSlider Object Reference=Get CTRL_L_mouth_cornerSharpnessD() -> (CTRL_L_mouth_cornerSharpnessD:EUW Picker 1DSlider Object Reference)
     "[46]":EUW Picker 1DSlider Object Reference=Get CTRL_L_mouth_cornerDepress() -> (CTRL_L_mouth_cornerDepress:EUW Picker 1DSlider Object Reference)
     "[47]":EUW Picker 1DSlider Object Reference=Get CTRL_L_jaw_ChinRaiseU() -> (CTRL_L_jaw_ChinRaiseU:EUW Picker 1DSlider Object Reference)
     "[48]":EUW Picker 1DSlider Object Reference=Get CTRL_L_jaw_ChinRaiseD() -> (CTRL_L_jaw_ChinRaiseD:EUW Picker 1DSlider Object Reference)
    ) -> (Array:Array of EUW Picker 1DSlider Object References)
   ) -> (Output_Get:Array of EUW Picker 1DSlider Object References)
   > Set UpperFace2DSliders(UpperFace2DSliders:Array of EUW Picker 2DSlider Object References=Make Array!(
     "[0]":EUW Picker 2DSlider Object Reference=Get CTRL_R_eye_eyelidU() -> (CTRL_R_eye_eyelidU:EUW Picker 2DSlider Object Reference)
     "[1]":EUW Picker 2DSlider Object Reference=Get CTRL_R_eye_eyelidD() -> (CTRL_R_eye_eyelidD:EUW Picker 2DSlider Object Reference)
     "[2]":EUW Picker 2DSlider Object Reference=Get CTRL_R_eye() -> (CTRL_R_eye:EUW Picker 2DSlider Object Reference)
     "[3]":EUW Picker 2DSlider Object Reference
     "[4]":EUW Picker 2DSlider Object Reference=Get CTRL_R_nose() -> (CTRL_R_nose:EUW Picker 2DSlider Object Reference)
     "[5]":EUW Picker 2DSlider Object Reference=Get CTRL_L_nose() -> (CTRL_L_nose:EUW Picker 2DSlider Object Reference)
     "[6]":EUW Picker 2DSlider Object Reference=Get CTRL_L_eye_eyelidU() -> (CTRL_L_eye_eyelidU:EUW Picker 2DSlider Object Reference)
     "[7]":EUW Picker 2DSlider Object Reference=Get CTRL_L_eye_eyelidD() -> (CTRL_L_eye_eyelidD:EUW Picker 2DSlider Object Reference)
     "[8]":EUW Picker 2DSlider Object Reference=Get CTRL_L_eye() -> (CTRL_L_eye:EUW Picker 2DSlider Object Reference)
     "[9]":EUW Picker 2DSlider Object Reference
     "[10]":EUW Picker 2DSlider Object Reference
     "[11]":EUW Picker 2DSlider Object Reference
     "[12]":EUW Picker 2DSlider Object Reference
    ) -> (Array:Array of EUW Picker 2DSlider Object References)) -> (Output_Get:Array of EUW Picker 2DSlider Object References)
   > Set UpperFace1DSliders(UpperFace1DSliders:Array of EUW Picker 1DSlider Object References=Make Array!(
     "[0]":EUW Picker 1DSlider Object Reference=Get CTRL_R_eye_squintInner() -> (CTRL_R_eye_squintInner:EUW Picker 1DSlider Object Reference)
     "[1]":EUW Picker 1DSlider Object Reference=Get CTRL_R_eye_cheekRaise() -> (CTRL_R_eye_cheekRaise:EUW Picker 1DSlider Object Reference)
     "[2]":EUW Picker 1DSlider Object Reference=Get CTRL_R_eye_blink() -> (CTRL_R_eye_blink:EUW Picker 1DSlider Object Reference)
     "[3]":EUW Picker 1DSlider Object Reference=Get CTRL_R_brow_raiseOut() -> (CTRL_R_brow_raiseOut:EUW Picker 1DSlider Object Reference)
     "[4]":EUW Picker 1DSlider Object Reference=Get CTRL_R_brow_raiseIn() -> (CTRL_R_brow_raiseIn:EUW Picker 1DSlider Object Reference)
     "[5]":EUW Picker 1DSlider Object Reference=Get CTRL_R_brow_outer() -> (CTRL_R_brow_outer:EUW Picker 1DSlider Object Reference)
     "[6]":EUW Picker 1DSlider Object Reference=Get CTRL_R_brow_lateral() -> (CTRL_R_brow_lateral:EUW Picker 1DSlider Object Reference)
     "[7]":EUW Picker 1DSlider Object Reference=Get CTRL_R_brow_down() -> (CTRL_R_brow_down:EUW Picker 1DSlider Object Reference)
     "[8]":EUW Picker 1DSlider Object Reference=Get CTRL_lookAtSwitch() -> (CTRL_lookAtSwitch:EUW Picker 1DSlider Object Reference)
     "[9]":EUW Picker 1DSlider Object Reference=Get CTRL_R_nose_wrinkleUpper() -> (CTRL_R_nose_wrinkleUpper:EUW Picker 1DSlider Object Reference)
     "[10]":EUW Picker 1DSlider Object Reference=Get CTRL_L_nose_wrinkleUpper() -> (CTRL_L_nose_wrinkleUpper:EUW Picker 1DSlider Object Reference)
     "[11]":EUW Picker 1DSlider Object Reference=Get CTRL_L_eye_squintInner() -> (CTRL_L_eye_squintInner:EUW Picker 1DSlider Object Reference)
     "[12]":EUW Picker 1DSlider Object Reference=Get CTRL_L_eye_cheekRaise() -> (CTRL_L_eye_cheekRaise:EUW Picker 1DSlider Object Reference)
     "[13]":EUW Picker 1DSlider Object Reference=Get CTRL_L_eye_blink() -> (CTRL_L_eye_blink:EUW Picker 1DSlider Object Reference)
     "[14]":EUW Picker 1DSlider Object Reference=Get CTRL_L_brow_raiseOut() -> (CTRL_L_brow_raiseOut:EUW Picker 1DSlider Object Reference)
     "[15]":EUW Picker 1DSlider Object Reference=Get CTRL_L_brow_raiseIn() -> (CTRL_L_brow_raiseIn:EUW Picker 1DSlider Object Reference)
     "[16]":EUW Picker 1DSlider Object Reference=Get CTRL_L_brow_outer() -> (CTRL_L_brow_outer:EUW Picker 1DSlider Object Reference)
     "[17]":EUW Picker 1DSlider Object Reference=Get CTRL_L_brow_lateral() -> (CTRL_L_brow_lateral:EUW Picker 1DSlider Object Reference)
     "[18]":EUW Picker 1DSlider Object Reference=Get CTRL_L_brow_down() -> (CTRL_L_brow_down:EUW Picker 1DSlider Object Reference)
     "[19]":EUW Picker 1DSlider Object Reference=Get CTRL_L_brow_inner() -> (CTRL_L_brow_inner:EUW Picker 1DSlider Object Reference)
     "[20]":EUW Picker 1DSlider Object Reference=Get CTRL_R_brow_inner() -> (CTRL_R_brow_inner:EUW Picker 1DSlider Object Reference)
    ) -> (Array:Array of EUW Picker 1DSlider Object References)) -> (Output_Get:Array of EUW Picker 1DSlider Object References)
   n6 = Event Mouse Enter!() -> (OutputDelegate, then, Mouse Event:Pointer Event Structure) {is=title K2Node_Event}
   > Parent: Mouse Enter!(execute, Mouse Event:Pointer Event Structure=@n6.Mouse Event) -> (then) {is=title K2Node_CallParentFunction}
   Update Sliders based on CR()
   Event Update From Scene() {is=title K2Node_Event}
  ]}
 Get Primary Panel {nodes [
   Get Primary Panel() {is=title K2Node_FunctionEntry}
   > Return Node!(execute, Panel:Panel Widget Object Reference=Get FN_MH_Face_CanvasPanel())
  ]}
 Get all Selectable Elements {nodes [
   Get all Selectable Elements() {is=title K2Node_FunctionEntry}
   > Return Node!(execute, All Elements:Array of EUW Picker Element Object References)
  ]}
 Set V2 Button Visibility {nodes [
   Set V2 Button Visibility!() -> (then, visible:bool=~) {is=title K2Node_FunctionEntry}
   > n1 = For Each Loop(Array:Array of EUW Picker 1DSlider Object References=Make Array!(
     "[0]"=Get CTRL_R_nose_nasolabialDeepen() -> (CTRL_R_nose_nasolabialDeepen:EUW Picker 1DSlider Object Reference)
     "[1]"=Get CTRL_R_mouth_tightenU() -> (CTRL_R_mouth_tightenU:EUW Picker 1DSlider Object Reference)
     "[2]"=Get CTRL_R_mouth_tightenD() -> (CTRL_R_mouth_tightenD:EUW Picker 1DSlider Object Reference)
     "[3]"=Get CTRL_R_mouth_thicknessU() -> (CTRL_R_mouth_thicknessU:EUW Picker 1DSlider Object Reference)
     "[4]"=Get CTRL_R_mouth_thicknessInwardU() -> (CTRL_R_mouth_thicknessInwardU:EUW Picker 1DSlider Object Reference)
     "[5]"=Get CTRL_R_mouth_thicknessInwardD() -> (CTRL_R_mouth_thicknessInwardD:EUW Picker 1DSlider Object Reference)
     "[6]"=Get CTRL_R_mouth_thicknessD() -> (CTRL_R_mouth_thicknessD:EUW Picker 1DSlider Object Reference)
     "[7]"=Get CTRL_R_mouth_pushPullU() -> (CTRL_R_mouth_pushPullU:EUW Picker 1DSlider Object Reference)
     "[8]"=Get CTRL_R_mouth_pushPullD() -> (CTRL_R_mouth_pushPullD:EUW Picker 1DSlider Object Reference)
     "[9]"=Get CTRL_R_mouth_lipsRollU() -> (CTRL_R_mouth_lipsRollU:EUW Picker 1DSlider Object Reference)
     "[10]"=Get CTRL_R_mouth_lipsRollD() -> (CTRL_R_mouth_lipsRollD:EUW Picker 1DSlider Object Reference)
     "[11]"=Get CTRL_R_mouth_lipsPressU() -> (CTRL_R_mouth_lipsPressU:EUW Picker 1DSlider Object Reference)
     "[12]"=Get CTRL_R_mouth_lipsPressD() -> (CTRL_R_mouth_lipsPressD:EUW Picker 1DSlider Object Reference)
     "[13]"=Get CTRL_R_eye_pupil() -> (CTRL_R_eye_pupil:EUW Picker 1DSlider Object Reference)
     "[14]"=Get CTRL_R_eye_faceScrunch() -> (CTRL_R_eye_faceScrunch:EUW Picker 1DSlider Object Reference)
     "[15]"=Get CTRL_L_nose_nasolabialDeepen() -> (CTRL_L_nose_nasolabialDeepen:EUW Picker 1DSlider Object Reference)
     "[16]"=Get CTRL_L_mouth_tightenU() -> (CTRL_L_mouth_tightenU:EUW Picker 1DSlider Object Reference)
     "[17]"=Get CTRL_L_mouth_tightenD() -> (CTRL_L_mouth_tightenD:EUW Picker 1DSlider Object Reference)
     "[18]"=Get CTRL_L_mouth_thicknessU() -> (CTRL_L_mouth_thicknessU:EUW Picker 1DSlider Object Reference)
     "[19]"=Get CTRL_L_mouth_thicknessInwardU() -> (CTRL_L_mouth_thicknessInwardU:EUW Picker 1DSlider Object Reference)
     "[20]"=Get CTRL_L_mouth_thicknessInwardD() -> (CTRL_L_mouth_thicknessInwardD:EUW Picker 1DSlider Object Reference)
     "[21]":EUW Picker 1DSlider Object Reference=Get CTRL_L_mouth_thicknessD() -> (CTRL_L_mouth_thicknessD:EUW Picker 1DSlider Object Reference)
     "[22]":EUW Picker 1DSlider Object Reference=Get CTRL_L_mouth_pushPullU() -> (CTRL_L_mouth_pushPullU:EUW Picker 1DSlider Object Reference)
     "[23]":EUW Picker 1DSlider Object Reference=Get CTRL_L_mouth_pushPullD() -> (CTRL_L_mouth_pushPullD:EUW Picker 1DSlider Object Reference)
     "[24]":EUW Picker 1DSlider Object Reference=Get CTRL_L_mouth_lipsRollU() -> (CTRL_L_mouth_lipsRollU:EUW Picker 1DSlider Object Reference)
     "[25]":EUW Picker 1DSlider Object Reference=Get CTRL_L_mouth_lipsRollD() -> (CTRL_L_mouth_lipsRollD:EUW Picker 1DSlider Object Reference)
     "[26]":EUW Picker 1DSlider Object Reference=Get CTRL_L_mouth_lipsRollU()
     "[27]":EUW Picker 1DSlider Object Reference=Get CTRL_L_mouth_lipsRollD()
     "[28]":EUW Picker 1DSlider Object Reference=Get CTRL_L_mouth_lipsPressU() -> (CTRL_L_mouth_lipsPressU:EUW Picker 1DSlider Object Reference)
     "[29]":EUW Picker 1DSlider Object Reference=Get CTRL_L_mouth_lipsPressD() -> (CTRL_L_mouth_lipsPressD:EUW Picker 1DSlider Object Reference)
     "[30]":EUW Picker 1DSlider Object Reference=Get CTRL_L_eye_pupil() -> (CTRL_L_eye_pupil:EUW Picker 1DSlider Object Reference)
     "[31]":EUW Picker 1DSlider Object Reference=Get CTRL_L_eye_faceScrunch() -> (CTRL_L_eye_faceScrunch:EUW Picker 1DSlider Object Reference)
     "[32]":EUW Picker 1DSlider Object Reference=Get CTRL_R_mouth_sharpCornerPull()
     "[33]":EUW Picker 1DSlider Object Reference=Get CTRL_L_mouth_sharpCornerPull()
    ) -> (Array)) -> (Array Element:EUW Picker 1DSlider Object Reference)
   Reroute Node(InputPin=@n1.Completed)
   n2 = Branch(execute=@n1.LoopBody, Condition=Get visible() -> (visible:bool))
   > SetVisibility(self=@n1, InVisibility=Visible)
   SetVisibility(execute=@n2.else, self=@n1, InVisibility=Hidden)
   n3 = For Each Loop(Array:Array of EUW Picker 2DSlider Object References=Make Array!(
     "[0]":EUW Picker 2DSlider Object Reference
     "[1]":EUW Picker 2DSlider Object Reference
     "[2]":EUW Picker 2DSlider Object Reference
     "[3]":EUW Picker 2DSlider Object Reference
    ) -> (Array:Array of EUW Picker 2DSlider Object References)) -> (Array Element:EUW Picker 2DSlider Object Reference)
   > n4 = Branch(Condition=Get visible())
   > SetVisibility(self=@n3, InVisibility=Visible)
   SetVisibility(execute=@n4.else, self=@n3, InVisibility=Hidden)
   Get CTRL_R_mouth_cornerPull() -> (CTRL_R_mouth_cornerPull:EUW Picker 1DSlider Object Reference)
   Get CTRL_L_mouth_cornerPull() -> (CTRL_L_mouth_cornerPull:EUW Picker 1DSlider Object Reference)
  ]}
 Update Sliders based on CR {nodes [
   Update Sliders based on CR() {is=title K2Node_FunctionEntry}
   > n1 = GetFocusedLevelSequence!(execute) -> (then, ReturnValue:Level Sequence Object Reference)
   > Set LS(LS:Level Sequence Object Reference=@n1) -> (Output_Get:Level Sequence Object Reference)
   > n2 = Get Current Local Time!(execute, TimeUnit:EMovieSceneTimeUnit Enum=DisplayRate) -> (then, ReturnValue:Movie Scene Sequence Playback Params Structure) {is=title K2Node_CallFunction}
   > Set Current Frame Number(Current Frame Number:Frame Number Structure=Make Frame Number!(Value:int=Break Movie Scene Sequence Playback Params!(MovieSceneSequencePlaybackParams:"Movie Scene Sequence Playback Params Structure (by ref)"=@n2) -> (
      Frame_FrameNumber_Value:int
      Frame_SubFrame:float=~
      Time:float=~
      MarkedFrame:String
      Timecode:Timecode Structure
      PositionType:EMovieScenePositionType Enum
      UpdateMethod:EUpdatePositionMethod Enum
      bHasJumped:bool=~
     ) {is=title K2Node_BreakStruct}) -> (FrameNumber:Frame Number Structure) {is=title K2Node_MakeStruct}) -> (Output_Get:Frame Number Structure)
   > n3 = Get!(execute:Exec) -> (then:Exec, else:Exec, Face CR:Control Rig Object Reference)
   > n4 = Set CR(CR:Control Rig Object Reference=@n3) -> (Output_Get:Control Rig Object Reference)
   n5 = For Each Loop(
    Exec=[@n11.then, @n4.then]
    Array:Array of EUW Picker 1DSlider Object References=Get All 1DSliders() -> (All 1DSliders:Array of EUW Picker 1DSlider Object References)
   ) -> (Array Element:EUW Picker 1DSlider Object Reference)
   n6 = For Each Loop(
    Exec=@n5.Completed
    Array:Array of EUW Picker 2DSlider Object References=Get All 2DSliders() -> (All 2DSliders:Array of EUW Picker 2DSlider Object References)
   ) -> (Array Element:EUW Picker 2DSlider Object Reference)
   Return Node!(execute=@n6.Completed)
   n7 = String To Name(InString=GetObjectName(Object=@n6))
   > n8 = GetLocalControlRigVector2D!(
    execute
    LevelSequence:Level Sequence Object Reference=Get LS() -> (LS:Level Sequence Object Reference)
    ControlRig:Control Rig Object Reference=Get CR() -> (CR:Control Rig Object Reference)
    ControlName:Name=@n7
    Frame:Frame Number Structure=Get Current Frame Number() -> (Current Frame Number:Frame Number Structure)
    TimeUnit:EMovieSceneTimeUnit Enum=DisplayRate
   ) -> (then, ReturnValue:Vector 2D Structure)
   > SetValue!(
    execute
    self:Synth 2DSlider Object Reference=Get Slider!(self:EUW Picker 2DSlider Object Reference=@n6) -> (Slider:Synth 2DSlider Object Reference)
    InValue:Vector 2D Structure=Convert to Slider 2D Relative Value!(self:EUW Picker 2DSlider Object Reference=@n6, Value:Vector 2D Structure=@n8) -> (ReturnValue:Vector 2D Structure)
   ) -> (then)
   > Update Selection Display!(
    execute:Exec
    self:Object Reference=@n6
    Selected:bool=IsControlSelected!(self:Control Rig Object Reference=Get CR(), InControlName:"Name (by ref)"=@n7) -> (ReturnValue:bool)
   ) -> (then:Exec, Out Selected:bool) {is=title K2Node_Message}
   n9 = String To Name(InString=GetObjectName(Object=@n5))
   n10 = GetLocalControlRigFloat!(
    execute=@n5.LoopBody
    LevelSequence:Level Sequence Object Reference=Get LS()
    ControlRig:Control Rig Object Reference=Get CR()
    ControlName:Name=@n9
    Frame:Frame Number Structure=Get Current Frame Number()
    TimeUnit:EMovieSceneTimeUnit Enum=DisplayRate
   ) -> (then, ReturnValue:float)
   > SetValue(
    self:Slider Object Reference=Get Slider(self:EUW Picker 1DSlider Object Reference=@n5) -> (Slider:Editor Utility Slider Object Reference)
    InValue:float=@n10
   )
   > Update Selection Display(self=@n5, Selected=IsControlSelected(self=Get CR(), InControlName=@n9))
   n11 = Set CR(
    execute=@n3.else
    CR=Break Control Rig Sequencer Binding Proxy!(ControlRigSequencerBindingProxy:"Control Rig Sequencer Binding Proxy Structure (by ref)"=Get Control Rig Proxy Binding() -> (Control Rig Proxy Binding:Control Rig Sequencer Binding Proxy Structure)) -> (
     Proxy:Movie Scene Binding Proxy Structure
     ControlRig:Control Rig Object Reference
     Track:Movie Scene Control Rig Parameter Track Object Reference
    ) {is=title K2Node_BreakStruct}.ControlRig
   )
  ]}
]
variables [
 All 1DSliders {type="array<EUW_Picker_1DSlider_C>", default []}
 All 2DSliders {type="array<EUW_Picker_2DSlider_C>", default []}
 Button Control Pairs {default {}}
 CurrentSelection {type="array<string>", default []}
 Face CR {type=ControlRig, default=null}
 LowerFace1DSliders {type="array<EUW_Picker_1DSlider_C>", default []}
 LowerFace2DSliders {type="array<EUW_Picker_2DSlider_C>", default []}
 LS {type=LevelSequence, default=null}
 Prefix {type=string, default=""}
 UpperFace1DSliders {type="array<EUW_Picker_1DSlider_C>", default []}
 UpperFace2DSliders {type="array<EUW_Picker_2DSlider_C>", default []}
]
widgets {
 info {parentClass=&/EUW_FN_Face_Base.EUW_FN_Face_Base_C, rootWidgetClass=&CanvasPanel, widgetCount=183}
 widgets [
  FN_MH_Face_CanvasPanel {
   widgetClassPath=&CanvasPanel
   bIsVariable=true
   widget=FN_MH_Face_CanvasPanel
   properties {renderTransform {angle=0, scale=<1 1>, shear=<0 0>, translation=<37 0>}, renderTransformPivot=<0 0>}
  }
  CTRL_C_eye {
   widgetClassPath=/AnimInEdTools/AnimationPicker/PickerElements/EUW_Picker_2DSlider
   bIsVariable=true
   parent=FN_MH_Face_CanvasPanel
   widget=CTRL_C_eye
   properties {bOverride_Cursor=true, max Value Y=-1, min Value Y=1}
   slot {
    class=CanvasPanelSlot
    properties {layoutData {alignment=<0.5 0.5>, anchors {maximum=<0.5 0.5>, minimum=<0.5 0.5>}, offsets=<ltrb -349.9999 -131.996 85.9253 82.406>}}
   }
  }
  CTRL_R_nose {
   widgetClassPath=&EUW_Picker_2DSlider
   bIsVariable=true
   parent=FN_MH_Face_CanvasPanel
   widget=CTRL_R_nose
   properties {
    bOverride_Cursor=true
    max Value X=-1
    max Value Y=-1
    min Value X=1
    min Value Y=1
    normal Style {
     barThickness=2
     backgroundImage {
      bIsDynamicallyLoaded=false
      drawAs=Image
      imageType=NoImage
      mirroring=NoMirror
      resourceName=None
      resourceObject=null
      tiling=NoTile
      imageSize=<32 32>
      margin=<ltrb 0 0 0 0>
      outlineSettings {
       bUseBrushTransparency=false
       roundingType=HalfHeightRadius
       width=0
       color {colorUseRule=UseColor_Specified, specifiedColor=<rgba 0 0 0 0>}
       cornerRadii=<0 0 0 0>
      }
      tintColor {colorUseRule=UseColor_Specified, specifiedColor=<rgba 0 0 0 0.1826>}
      uVRegion {bIsValid=false, max=<0 0>, min=<0 0>}
     }
     disabledBarImage {
      bIsDynamicallyLoaded=false
      drawAs=Image
      imageType=NoImage
      mirroring=NoMirror
      resourceName=None
      resourceObject=null
      tiling=NoTile
      imageSize=<32 32>
      margin=<ltrb 0 0 0 0>
      outlineSettings {
       bUseBrushTransparency=false
       roundingType=HalfHeightRadius
       width=0
       color {colorUseRule=UseColor_Specified, specifiedColor=<rgba 0 0 0 0>}
       cornerRadii=<0 0 0 0>
      }
      tintColor {colorUseRule=UseColor_Specified, specifiedColor=<rgba 1 1 1 1>}
      uVRegion {bIsValid=false, max=<0 0>, min=<0 0>}
     }
     disabledThumbImage {
      bIsDynamicallyLoaded=false
      drawAs=Image
      imageType=NoImage
      mirroring=NoMirror
      resourceName=None
      resourceObject=null
      tiling=NoTile
      imageSize=<32 32>
      margin=<ltrb 0 0 0 0>
      outlineSettings {
       bUseBrushTransparency=false
       roundingType=HalfHeightRadius
       width=0
       color {colorUseRule=UseColor_Specified, specifiedColor=<rgba 0 0 0 0>}
       cornerRadii=<0 0 0 0>
      }
      tintColor {colorUseRule=UseColor_Specified, specifiedColor=<rgba 1 1 1 1>}
      uVRegion {bIsValid=false, max=<0 0>, min=<0 0>}
     }
     normalBarImage {
      bIsDynamicallyLoaded=false
      drawAs=Image
      imageType=NoImage
      mirroring=NoMirror
      resourceName=None
      resourceObject=null
      tiling=NoTile
      imageSize=<32 32>
      margin=<ltrb 0 0 0 0>
      outlineSettings {
       bUseBrushTransparency=false
       roundingType=HalfHeightRadius
       width=0
       color {colorUseRule=UseColor_Specified, specifiedColor=<rgba 0 0 0 0>}
       cornerRadii=<0 0 0 0>
      }
      tintColor {colorUseRule=UseColor_Specified, specifiedColor=<rgba 1 1 1 1>}
      uVRegion {bIsValid=false, max=<0 0>, min=<0 0>}
     }
     normalThumbImage {
      bIsDynamicallyLoaded=false
      drawAs=Image
      imageType=FullColor
      mirroring=NoMirror
      resourceName=None
      resourceObject=/Game/Characters/Player/Common/SharedRigs/Graphics/Circle
      tiling=NoTile
      imageSize=<15 15>
      margin=<ltrb 0 0 0 0>
      outlineSettings {
       bUseBrushTransparency=false
       roundingType=HalfHeightRadius
       width=0
       color {colorUseRule=UseColor_Specified, specifiedColor=<rgba 0 0 0 0>}
       cornerRadii=<0 0 0 0>
      }
      tintColor {colorUseRule=UseColor_Specified, specifiedColor=<rgba 0.2604 0.2604 0.2604 1>}
      uVRegion {bIsValid=false, max=<0 0>, min=<0 0>}
     }
    }
    selected Style {
     barThickness=2
     backgroundImage {
      bIsDynamicallyLoaded=false
      drawAs=Image
      imageType=NoImage
      mirroring=NoMirror
      resourceName=None
      resourceObject=null
      tiling=NoTile
      imageSize=<32 32>
      margin=<ltrb 0 0 0 0>
      outlineSettings {
       bUseBrushTransparency=false
       roundingType=HalfHeightRadius
       width=0
       color {colorUseRule=UseColor_Specified, specifiedColor=<rgba 0 0 0 0>}
       cornerRadii=<0 0 0 0>
      }
      tintColor {colorUseRule=UseColor_Specified, specifiedColor=<rgba 0.1979 0.1979 0.1979 0.1826>}
      uVRegion {bIsValid=false, max=<0 0>, min=<0 0>}
     }
     disabledBarImage {
      bIsDynamicallyLoaded=false
      drawAs=Image
      imageType=NoImage
      mirroring=NoMirror
      resourceName=None
      resourceObject=null
      tiling=NoTile
      imageSize=<32 32>
      margin=<ltrb 0 0 0 0>
      outlineSettings {
       bUseBrushTransparency=false
       roundingType=HalfHeightRadius
       width=0
       color {colorUseRule=UseColor_Specified, specifiedColor=<rgba 0 0 0 0>}
       cornerRadii=<0 0 0 0>
      }
      tintColor {colorUseRule=UseColor_Specified, specifiedColor=<rgba 1 1 1 1>}
      uVRegion {bIsValid=false, max=<0 0>, min=<0 0>}
     }
     disabledThumbImage {
      bIsDynamicallyLoaded=false
      drawAs=Image
      imageType=NoImage
      mirroring=NoMirror
      resourceName=None
      resourceObject=null
      tiling=NoTile
      imageSize=<32 32>
      margin=<ltrb 0 0 0 0>
      outlineSettings {
       bUseBrushTransparency=false
       roundingType=HalfHeightRadius
       width=0
       color {colorUseRule=UseColor_Specified, specifiedColor=<rgba 0 0 0 0>}
       cornerRadii=<0 0 0 0>
      }
      tintColor {colorUseRule=UseColor_Specified, specifiedColor=<rgba 1 1 1 1>}
      uVRegion {bIsValid=false, max=<0 0>, min=<0 0>}
     }
     normalBarImage {
      bIsDynamicallyLoaded=false
      drawAs=Image
      imageType=NoImage
      mirroring=NoMirror
      resourceName=None
      resourceObject=null
      tiling=NoTile
      imageSize=<32 32>
      margin=<ltrb 0 0 0 0>
      outlineSettings {
       bUseBrushTransparency=false
       roundingType=HalfHeightRadius
       width=0
       color {colorUseRule=UseColor_Specified, specifiedColor=<rgba 0 0 0 0>}
       cornerRadii=<0 0 0 0>
      }
      tintColor {colorUseRule=UseColor_Specified, specifiedColor=<rgba 1 1 1 1>}
      uVRegion {bIsValid=false, max=<0 0>, min=<0 0>}
     }
     normalThumbImage {
      bIsDynamicallyLoaded=false
      drawAs=Image
      imageType=FullColor
      mirroring=NoMirror
      resourceName=None
      resourceObject=&Circle
      tiling=NoTile
      imageSize=<26 26>
      margin=<ltrb 0 0 0 0>
      outlineSettings {
       bUseBrushTransparency=false
       roundingType=HalfHeightRadius
       width=0
       color {colorUseRule=UseColor_Specified, specifiedColor=<rgba 0 0 0 0>}
       cornerRadii=<0 0 0 0>
      }
      tintColor {colorUseRule=UseColor_Specified, specifiedColor=<rgba 1 1 1 1>}
      uVRegion {bIsValid=false, max=<0 0>, min=<0 0>}
     }
    }
   }
   slot {
    class=CanvasPanelSlot
    properties {layoutData {alignment=<0.5 0.5>, anchors {maximum=<0.5 0.5>, minimum=<0.5 0.5>}, offsets=<ltrb -380 -58 55.0373 56.237>}}
   }
  }
  CTRL_C_nose {
   widgetClassPath=&EUW_Picker_2DSlider
   bIsVariable=true
   parent=FN_MH_Face_CanvasPanel
   widget=CTRL_C_nose
   properties {bOverride_Cursor=true, max Value Y=-1, min Value Y=1}
   slot {
    class=CanvasPanelSlot
    properties {layoutData {alignment=<0.5 0.5>, anchors {maximum=<0.5 0.5>, minimum=<0.5 0.5>}, offsets=<ltrb 540.5577 -72.422 64 64>}}
   }
  }
  CTRL_R_eye {
   widgetClassPath=&EUW_Picker_2DSlider
   bIsVariable=true
   parent=FN_MH_Face_CanvasPanel
   widget=CTRL_R_eye
   properties {
    bOverride_Cursor=true
    max Value Y=-1
    min Value Y=1
    normal Style {
     barThickness=2
     backgroundImage {
      bIsDynamicallyLoaded=false
      drawAs=Image
      imageType=NoImage
      mirroring=NoMirror
      resourceName=None
      resourceObject=null
      tiling=NoTile
      imageSize=<32 32>
      margin=<ltrb 0 0 0 0>
      outlineSettings {
       bUseBrushTransparency=false
       roundingType=HalfHeightRadius
       width=0
       color {colorUseRule=UseColor_Specified, specifiedColor=<rgba 0 0 0 0>}
       cornerRadii=<0 0 0 0>
      }
      tintColor {colorUseRule=UseColor_Specified, specifiedColor=<rgba 0 0 0 0.1826>}
      uVRegion {bIsValid=false, max=<0 0>, min=<0 0>}
     }
     disabledBarImage {
      bIsDynamicallyLoaded=false
      drawAs=Image
      imageType=NoImage
      mirroring=NoMirror
      resourceName=None
      resourceObject=null
      tiling=NoTile
      imageSize=<32 32>
      margin=<ltrb 0 0 0 0>
      outlineSettings {
       bUseBrushTransparency=false
       roundingType=HalfHeightRadius
       width=0
       color {colorUseRule=UseColor_Specified, specifiedColor=<rgba 0 0 0 0>}
       cornerRadii=<0 0 0 0>
      }
      tintColor {colorUseRule=UseColor_Specified, specifiedColor=<rgba 1 1 1 1>}
      uVRegion {bIsValid=false, max=<0 0>, min=<0 0>}
     }
     disabledThumbImage {
      bIsDynamicallyLoaded=false
      drawAs=Image
      imageType=NoImage
      mirroring=NoMirror
      resourceName=None
      resourceObject=null
      tiling=NoTile
      imageSize=<32 32>
      margin=<ltrb 0 0 0 0>
      outlineSettings {
       bUseBrushTransparency=false
       roundingType=HalfHeightRadius
       width=0
       color {colorUseRule=UseColor_Specified, specifiedColor=<rgba 0 0 0 0>}
       cornerRadii=<0 0 0 0>
      }
      tintColor {colorUseRule=UseColor_Specified, specifiedColor=<rgba 1 1 1 1>}
      uVRegion {bIsValid=false, max=<0 0>, min=<0 0>}
     }
     normalBarImage {
      bIsDynamicallyLoaded=false
      drawAs=Image
      imageType=NoImage
      mirroring=NoMirror
      resourceName=None
      resourceObject=null
      tiling=NoTile
      imageSize=<32 32>
      margin=<ltrb 0 0 0 0>
      outlineSettings {
       bUseBrushTransparency=false
       roundingType=HalfHeightRadius
       width=0
       color {colorUseRule=UseColor_Specified, specifiedColor=<rgba 0 0 0 0>}
       cornerRadii=<0 0 0 0>
      }
      tintColor {colorUseRule=UseColor_Specified, specifiedColor=<rgba 1 1 1 1>}
      uVRegion {bIsValid=false, max=<0 0>, min=<0 0>}
     }
     normalThumbImage {
      bIsDynamicallyLoaded=false
      drawAs=Image
      imageType=FullColor
      mirroring=NoMirror
      resourceName=None
      resourceObject=&Circle
      tiling=NoTile
      imageSize=<20 20>
      margin=<ltrb 0 0 0 0>
      outlineSettings {
       bUseBrushTransparency=false
       roundingType=HalfHeightRadius
       width=0
       color {colorUseRule=UseColor_Specified, specifiedColor=<rgba 0 0 0 0>}
       cornerRadii=<0 0 0 0>
      }
      tintColor {colorUseRule=UseColor_Specified, specifiedColor=<rgba 0.1458 0.1458 0.1458 1>}
      uVRegion {bIsValid=false, max=<0 0>, min=<0 0>}
     }
    }
    selected Style {
     barThickness=2
     backgroundImage {
      bIsDynamicallyLoaded=false
      drawAs=Image
      imageType=NoImage
      mirroring=NoMirror
      resourceName=None
      resourceObject=null
      tiling=NoTile
      imageSize=<32 32>
      margin=<ltrb 0 0 0 0>
      outlineSettings {
       bUseBrushTransparency=false
       roundingType=HalfHeightRadius
       width=0
       color {colorUseRule=UseColor_Specified, specifiedColor=<rgba 0 0 0 0>}
       cornerRadii=<0 0 0 0>
      }
      tintColor {colorUseRule=UseColor_Specified, specifiedColor=<rgba 0.1979 0.1979 0.1979 0.1826>}
      uVRegion {bIsValid=false, max=<0 0>, min=<0 0>}
     }
     disabledBarImage {
      bIsDynamicallyLoaded=false
      drawAs=Image
      imageType=NoImage
      mirroring=NoMirror
      resourceName=None
      resourceObject=null
      tiling=NoTile
      imageSize=<32 32>
      margin=<ltrb 0 0 0 0>
      outlineSettings {
       bUseBrushTransparency=false
       roundingType=HalfHeightRadius
       width=0
       color {colorUseRule=UseColor_Specified, specifiedColor=<rgba 0 0 0 0>}
       cornerRadii=<0 0 0 0>
      }
      tintColor {colorUseRule=UseColor_Specified, specifiedColor=<rgba 1 1 1 1>}
      uVRegion {bIsValid=false, max=<0 0>, min=<0 0>}
     }
     disabledThumbImage {
      bIsDynamicallyLoaded=false
      drawAs=Image
      imageType=NoImage
      mirroring=NoMirror
      resourceName=None
      resourceObject=null
      tiling=NoTile
      imageSize=<32 32>
      margin=<ltrb 0 0 0 0>
      outlineSettings {
       bUseBrushTransparency=false
       roundingType=HalfHeightRadius
       width=0
       color {colorUseRule=UseColor_Specified, specifiedColor=<rgba 0 0 0 0>}
       cornerRadii=<0 0 0 0>
      }
      tintColor {colorUseRule=UseColor_Specified, specifiedColor=<rgba 1 1 1 1>}
      uVRegion {bIsValid=false, max=<0 0>, min=<0 0>}
     }
     normalBarImage {
      bIsDynamicallyLoaded=false
      drawAs=Image
      imageType=NoImage
      mirroring=NoMirror
      resourceName=None
      resourceObject=null
      tiling=NoTile
      imageSize=<32 32>
      margin=<ltrb 0 0 0 0>
      outlineSettings {
       bUseBrushTransparency=false
       roundingType=HalfHeightRadius
       width=0
       color {colorUseRule=UseColor_Specified, specifiedColor=<rgba 0 0 0 0>}
       cornerRadii=<0 0 0 0>
      }
      tintColor {colorUseRule=UseColor_Specified, specifiedColor=<rgba 1 1 1 1>}
      uVRegion {bIsValid=false, max=<0 0>, min=<0 0>}
     }
     normalThumbImage {
      bIsDynamicallyLoaded=false
      drawAs=Image
      imageType=FullColor
      mirroring=NoMirror
      resourceName=None
      resourceObject=&Circle
      tiling=NoTile
      imageSize=<30 30>
      margin=<ltrb 0 0 0 0>
      outlineSettings {
       bUseBrushTransparency=false
       roundingType=HalfHeightRadius
       width=0
       color {colorUseRule=UseColor_Specified, specifiedColor=<rgba 0 0 0 0>}
       cornerRadii=<0 0 0 0>
      }
      tintColor {colorUseRule=UseColor_Specified, specifiedColor=<rgba 1 1 1 1>}
      uVRegion {bIsValid=false, max=<0 0>, min=<0 0>}
     }
    }
   }
   slot {
    class=CanvasPanelSlot
    properties {layoutData {alignment=<0.5 0.5>, anchors {maximum=<0.5 0.5>, minimum=<0.5 0.5>}, offsets=<ltrb -459 -144.422 55.0373 56.237>}}
   }
  }
  CTRL_R_eye_eyelidU {
   widgetClassPath=&EUW_Picker_2DSlider
   bIsVariable=true
   parent=FN_MH_Face_CanvasPanel
   widget=CTRL_R_eye_eyelidU
   properties {
    bOverride_Cursor=true
    normal Style {
     barThickness=2
     backgroundImage {
      bIsDynamicallyLoaded=false
      drawAs=Image
      imageType=NoImage
      mirroring=NoMirror
      resourceName=None
      resourceObject=null
      tiling=NoTile
      imageSize=<32 32>
      margin=<ltrb 0 0 0 0>
      outlineSettings {
       bUseBrushTransparency=false
       roundingType=HalfHeightRadius
       width=0
       color {colorUseRule=UseColor_Specified, specifiedColor=<rgba 0 0 0 0>}
       cornerRadii=<0 0 0 0>
      }
      tintColor {colorUseRule=UseColor_Specified, specifiedColor=<rgba 0 0 0 0.1826>}
      uVRegion {bIsValid=false, max=<0 0>, min=<0 0>}
     }
     disabledBarImage {
      bIsDynamicallyLoaded=false
      drawAs=Image
      imageType=NoImage
      mirroring=NoMirror
      resourceName=None
      resourceObject=null
      tiling=NoTile
      imageSize=<32 32>
      margin=<ltrb 0 0 0 0>
      outlineSettings {
       bUseBrushTransparency=false
       roundingType=HalfHeightRadius
       width=0
       color {colorUseRule=UseColor_Specified, specifiedColor=<rgba 0 0 0 0>}
       cornerRadii=<0 0 0 0>
      }
      tintColor {colorUseRule=UseColor_Specified, specifiedColor=<rgba 1 1 1 1>}
      uVRegion {bIsValid=false, max=<0 0>, min=<0 0>}
     }
     disabledThumbImage {
      bIsDynamicallyLoaded=false
      drawAs=Image
      imageType=NoImage
      mirroring=NoMirror
      resourceName=None
      resourceObject=null
      tiling=NoTile
      imageSize=<32 32>
      margin=<ltrb 0 0 0 0>
      outlineSettings {
       bUseBrushTransparency=false
       roundingType=HalfHeightRadius
       width=0
       color {colorUseRule=UseColor_Specified, specifiedColor=<rgba 0 0 0 0>}
       cornerRadii=<0 0 0 0>
      }
      tintColor {colorUseRule=UseColor_Specified, specifiedColor=<rgba 1 1 1 1>}
      uVRegion {bIsValid=false, max=<0 0>, min=<0 0>}
     }
     normalBarImage {
      bIsDynamicallyLoaded=false
      drawAs=Image
      imageType=NoImage
      mirroring=NoMirror
      resourceName=None
      resourceObject=null
      tiling=NoTile
      imageSize=<32 32>
      margin=<ltrb 0 0 0 0>
      outlineSettings {
       bUseBrushTransparency=false
       roundingType=HalfHeightRadius
       width=0
       color {colorUseRule=UseColor_Specified, specifiedColor=<rgba 0 0 0 0>}
       cornerRadii=<0 0 0 0>
      }
      tintColor {colorUseRule=UseColor_Specified, specifiedColor=<rgba 1 1 1 1>}
      uVRegion {bIsValid=false, max=<0 0>, min=<0 0>}
     }
     normalThumbImage {
      bIsDynamicallyLoaded=false
      drawAs=Image
      imageType=FullColor
      mirroring=NoMirror
      resourceName=None
      resourceObject=&Circle
      tiling=NoTile
      imageSize=<20 20>
      margin=<ltrb 0 0 0 0>
      outlineSettings {
       bUseBrushTransparency=false
       roundingType=HalfHeightRadius
       width=0
       color {colorUseRule=UseColor_Specified, specifiedColor=<rgba 0 0 0 0>}
       cornerRadii=<0 0 0 0>
      }
      tintColor {colorUseRule=UseColor_Specified, specifiedColor=<rgba 0.2604 0.2604 0.2604 1>}
      uVRegion {bIsValid=false, max=<0 0>, min=<0 0>}
     }
    }
    selected Style {
     barThickness=2
     backgroundImage {
      bIsDynamicallyLoaded=false
      drawAs=Image
      imageType=NoImage
      mirroring=NoMirror
      resourceName=None
      resourceObject=null
      tiling=NoTile
      imageSize=<32 32>
      margin=<ltrb 0 0 0 0>
      outlineSettings {
       bUseBrushTransparency=false
       roundingType=HalfHeightRadius
       width=0
       color {colorUseRule=UseColor_Specified, specifiedColor=<rgba 0 0 0 0>}
       cornerRadii=<0 0 0 0>
      }
      tintColor {colorUseRule=UseColor_Specified, specifiedColor=<rgba 0.1979 0.1979 0.1979 0.1826>}
      uVRegion {bIsValid=false, max=<0 0>, min=<0 0>}
     }
     disabledBarImage {
      bIsDynamicallyLoaded=false
      drawAs=Image
      imageType=NoImage
      mirroring=NoMirror
      resourceName=None
      resourceObject=null
      tiling=NoTile
      imageSize=<32 32>
      margin=<ltrb 0 0 0 0>
      outlineSettings {
       bUseBrushTransparency=false
       roundingType=HalfHeightRadius
       width=0
       color {colorUseRule=UseColor_Specified, specifiedColor=<rgba 0 0 0 0>}
       cornerRadii=<0 0 0 0>
      }
      tintColor {colorUseRule=UseColor_Specified, specifiedColor=<rgba 1 1 1 1>}
      uVRegion {bIsValid=false, max=<0 0>, min=<0 0>}
     }
     disabledThumbImage {
      bIsDynamicallyLoaded=false
      drawAs=Image
      imageType=NoImage
      mirroring=NoMirror
      resourceName=None
      resourceObject=null
      tiling=NoTile
      imageSize=<32 32>
      margin=<ltrb 0 0 0 0>
      outlineSettings {
       bUseBrushTransparency=false
       roundingType=HalfHeightRadius
       width=0
       color {colorUseRule=UseColor_Specified, specifiedColor=<rgba 0 0 0 0>}
       cornerRadii=<0 0 0 0>
      }
      tintColor {colorUseRule=UseColor_Specified, specifiedColor=<rgba 1 1 1 1>}
      uVRegion {bIsValid=false, max=<0 0>, min=<0 0>}
     }
     normalBarImage {
      bIsDynamicallyLoaded=false
      drawAs=Image
      imageType=NoImage
      mirroring=NoMirror
      resourceName=None
      resourceObject=null
      tiling=NoTile
      imageSize=<32 32>
      margin=<ltrb 0 0 0 0>
      outlineSettings {
       bUseBrushTransparency=false
       roundingType=HalfHeightRadius
       width=0
       color {colorUseRule=UseColor_Specified, specifiedColor=<rgba 0 0 0 0>}
       cornerRadii=<0 0 0 0>
      }
      tintColor {colorUseRule=UseColor_Specified, specifiedColor=<rgba 1 1 1 1>}
      uVRegion {bIsValid=false, max=<0 0>, min=<0 0>}
     }
     normalThumbImage {
      bIsDynamicallyLoaded=false
      drawAs=Image
      imageType=FullColor
      mirroring=NoMirror
      resourceName=None
      resourceObject=&Circle
      tiling=NoTile
      imageSize=<25 25>
      margin=<ltrb 0 0 0 0>
      outlineSettings {
       bUseBrushTransparency=false
       roundingType=HalfHeightRadius
       width=0
       color {colorUseRule=UseColor_Specified, specifiedColor=<rgba 0 0 0 0>}
       cornerRadii=<0 0 0 0>
      }
      tintColor {colorUseRule=UseColor_Specified, specifiedColor=<rgba 1 1 1 1>}
      uVRegion {bIsValid=false, max=<0 0>, min=<0 0>}
     }
    }
   }
   slot {
    class=CanvasPanelSlot
    properties {layoutData {alignment=<0.5 0.5>, anchors {maximum=<0.5 0.5>, minimum=<0.5 0.5>}, offsets=<ltrb 380.5577 -196 50 50>}}
   }
  }
  CTRL_R_eye_eyelidD {
   widgetClassPath=&EUW_Picker_2DSlider
   bIsVariable=true
   parent=FN_MH_Face_CanvasPanel
   widget=CTRL_R_eye_eyelidD
   properties {
    bOverride_Cursor=true
    max Value Y=-1
    min Value Y=1
    normal Style {
     barThickness=2
     backgroundImage {
      bIsDynamicallyLoaded=false
      drawAs=Image
      imageType=NoImage
      mirroring=NoMirror
      resourceName=None
      resourceObject=null
      tiling=NoTile
      imageSize=<32 32>
      margin=<ltrb 0 0 0 0>
      outlineSettings {
       bUseBrushTransparency=false
       roundingType=HalfHeightRadius
       width=0
       color {colorUseRule=UseColor_Specified, specifiedColor=<rgba 0 0 0 0>}
       cornerRadii=<0 0 0 0>
      }
      tintColor {colorUseRule=UseColor_Specified, specifiedColor=<rgba 0 0 0 0.1826>}
      uVRegion {bIsValid=false, max=<0 0>, min=<0 0>}
     }
     disabledBarImage {
      bIsDynamicallyLoaded=false
      drawAs=Image
      imageType=NoImage
      mirroring=NoMirror
      resourceName=None
      resourceObject=null
      tiling=NoTile
      imageSize=<32 32>
      margin=<ltrb 0 0 0 0>
      outlineSettings {
       bUseBrushTransparency=false
       roundingType=HalfHeightRadius
       width=0
       color {colorUseRule=UseColor_Specified, specifiedColor=<rgba 0 0 0 0>}
       cornerRadii=<0 0 0 0>
      }
      tintColor {colorUseRule=UseColor_Specified, specifiedColor=<rgba 1 1 1 1>}
      uVRegion {bIsValid=false, max=<0 0>, min=<0 0>}
     }
     disabledThumbImage {
      bIsDynamicallyLoaded=false
      drawAs=Image
      imageType=NoImage
      mirroring=NoMirror
      resourceName=None
      resourceObject=null
      tiling=NoTile
      imageSize=<32 32>
      margin=<ltrb 0 0 0 0>
      outlineSettings {
       bUseBrushTransparency=false
       roundingType=HalfHeightRadius
       width=0
       color {colorUseRule=UseColor_Specified, specifiedColor=<rgba 0 0 0 0>}
       cornerRadii=<0 0 0 0>
      }
      tintColor {colorUseRule=UseColor_Specified, specifiedColor=<rgba 1 1 1 1>}
      uVRegion {bIsValid=false, max=<0 0>, min=<0 0>}
     }
     normalBarImage {
      bIsDynamicallyLoaded=false
      drawAs=Image
      imageType=NoImage
      mirroring=NoMirror
      resourceName=None
      resourceObject=null
      tiling=NoTile
      imageSize=<32 32>
      margin=<ltrb 0 0 0 0>
      outlineSettings {
       bUseBrushTransparency=false
       roundingType=HalfHeightRadius
       width=0
       color {colorUseRule=UseColor_Specified, specifiedColor=<rgba 0 0 0 0>}
       cornerRadii=<0 0 0 0>
      }
      tintColor {colorUseRule=UseColor_Specified, specifiedColor=<rgba 1 1 1 1>}
      uVRegion {bIsValid=false, max=<0 0>, min=<0 0>}
     }
     normalThumbImage {
      bIsDynamicallyLoaded=false
      drawAs=Image
      imageType=FullColor
      mirroring=NoMirror
      resourceName=None
      resourceObject=&Circle
      tiling=NoTile
      imageSize=<20 20>
      margin=<ltrb 0 0 0 0>
      outlineSettings {
       bUseBrushTransparency=false
       roundingType=HalfHeightRadius
       width=0
       color {colorUseRule=UseColor_Specified, specifiedColor=<rgba 0 0 0 0>}
       cornerRadii=<0 0 0 0>
      }
      tintColor {colorUseRule=UseColor_Specified, specifiedColor=<rgba 0.2604 0.2604 0.2604 1>}
      uVRegion {bIsValid=false, max=<0 0>, min=<0 0>}
     }
    }
    selected Style {
     barThickness=2
     backgroundImage {
      bIsDynamicallyLoaded=false
      drawAs=Image
      imageType=NoImage
      mirroring=NoMirror
      resourceName=None
      resourceObject=null
      tiling=NoTile
      imageSize=<32 32>
      margin=<ltrb 0 0 0 0>
      outlineSettings {
       bUseBrushTransparency=false
       roundingType=HalfHeightRadius
       width=0
       color {colorUseRule=UseColor_Specified, specifiedColor=<rgba 0 0 0 0>}
       cornerRadii=<0 0 0 0>
      }
      tintColor {colorUseRule=UseColor_Specified, specifiedColor=<rgba 0.1979 0.1979 0.1979 0.1826>}
      uVRegion {bIsValid=false, max=<0 0>, min=<0 0>}
     }
     disabledBarImage {
      bIsDynamicallyLoaded=false
      drawAs=Image
      imageType=NoImage
      mirroring=NoMirror
      resourceName=None
      resourceObject=null
      tiling=NoTile
      imageSize=<32 32>
      margin=<ltrb 0 0 0 0>
      outlineSettings {
       bUseBrushTransparency=false
       roundingType=HalfHeightRadius
       width=0
       color {colorUseRule=UseColor_Specified, specifiedColor=<rgba 0 0 0 0>}
       cornerRadii=<0 0 0 0>
      }
      tintColor {colorUseRule=UseColor_Specified, specifiedColor=<rgba 1 1 1 1>}
      uVRegion {bIsValid=false, max=<0 0>, min=<0 0>}
     }
     disabledThumbImage {
      bIsDynamicallyLoaded=false
      drawAs=Image
      imageType=NoImage
      mirroring=NoMirror
      resourceName=None
      resourceObject=null
      tiling=NoTile
      imageSize=<32 32>
      margin=<ltrb 0 0 0 0>
      outlineSettings {
       bUseBrushTransparency=false
       roundingType=HalfHeightRadius
       width=0
       color {colorUseRule=UseColor_Specified, specifiedColor=<rgba 0 0 0 0>}
       cornerRadii=<0 0 0 0>
      }
      tintColor {colorUseRule=UseColor_Specified, specifiedColor=<rgba 1 1 1 1>}
      uVRegion {bIsValid=false, max=<0 0>, min=<0 0>}
     }
     normalBarImage {
      bIsDynamicallyLoaded=false
      drawAs=Image
      imageType=NoImage
      mirroring=NoMirror
      resourceName=None
      resourceObject=null
      tiling=NoTile
      imageSize=<32 32>
      margin=<ltrb 0 0 0 0>
      outlineSettings {
       bUseBrushTransparency=false
       roundingType=HalfHeightRadius
       width=0
       color {colorUseRule=UseColor_Specified, specifiedColor=<rgba 0 0 0 0>}
       cornerRadii=<0 0 0 0>
      }
      tintColor {colorUseRule=UseColor_Specified, specifiedColor=<rgba 1 1 1 1>}
      uVRegion {bIsValid=false, max=<0 0>, min=<0 0>}
     }
     normalThumbImage {
      bIsDynamicallyLoaded=false
      drawAs=Image
      imageType=FullColor
      mirroring=NoMirror
      resourceName=None
      resourceObject=&Circle
      tiling=NoTile
      imageSize=<25 25>
      margin=<ltrb 0 0 0 0>
      outlineSettings {
       bUseBrushTransparency=false
       roundingType=HalfHeightRadius
       width=0
       color {colorUseRule=UseColor_Specified, specifiedColor=<rgba 0 0 0 0>}
       cornerRadii=<0 0 0 0>
      }
      tintColor {colorUseRule=UseColor_Specified, specifiedColor=<rgba 1 1 1 1>}
      uVRegion {bIsValid=false, max=<0 0>, min=<0 0>}
     }
    }
   }
   slot {
    class=CanvasPanelSlot
    properties {layoutData {alignment=<0.5 0.5>, anchors {maximum=<0.5 0.5>, minimum=<0.5 0.5>}, offsets=<ltrb 380.5577 -140 50 50>}}
   }
  }
  CTRL_L_eye {
   widgetClassPath=&EUW_Picker_2DSlider
   bIsVariable=true
   parent=FN_MH_Face_CanvasPanel
   widget=CTRL_L_eye
   properties {
    bOverride_Cursor=true
    max Value Y=-1
    min Value Y=1
    normal Style {
     barThickness=2
     backgroundImage {
      bIsDynamicallyLoaded=false
      drawAs=Image
      imageType=NoImage
      mirroring=NoMirror
      resourceName=None
      resourceObject=null
      tiling=NoTile
      imageSize=<32 32>
      margin=<ltrb 0 0 0 0>
      outlineSettings {
       bUseBrushTransparency=false
       roundingType=HalfHeightRadius
       width=0
       color {colorUseRule=UseColor_Specified, specifiedColor=<rgba 0 0 0 0>}
       cornerRadii=<0 0 0 0>
      }
      tintColor {colorUseRule=UseColor_Specified, specifiedColor=<rgba 0 0 0 0.1826>}
      uVRegion {bIsValid=false, max=<0 0>, min=<0 0>}
     }
     disabledBarImage {
      bIsDynamicallyLoaded=false
      drawAs=Image
      imageType=NoImage
      mirroring=NoMirror
      resourceName=None
      resourceObject=null
      tiling=NoTile
      imageSize=<32 32>
      margin=<ltrb 0 0 0 0>
      outlineSettings {
       bUseBrushTransparency=false
       roundingType=HalfHeightRadius
       width=0
       color {colorUseRule=UseColor_Specified, specifiedColor=<rgba 0 0 0 0>}
       cornerRadii=<0 0 0 0>
      }
      tintColor {colorUseRule=UseColor_Specified, specifiedColor=<rgba 1 1 1 1>}
      uVRegion {bIsValid=false, max=<0 0>, min=<0 0>}
     }
     disabledThumbImage {
      bIsDynamicallyLoaded=false
      drawAs=Image
      imageType=NoImage
      mirroring=NoMirror
      resourceName=None
      resourceObject=null
      tiling=NoTile
      imageSize=<32 32>
      margin=<ltrb 0 0 0 0>
      outlineSettings {
       bUseBrushTransparency=false
       roundingType=HalfHeightRadius
       width=0
       color {colorUseRule=UseColor_Specified, specifiedColor=<rgba 0 0 0 0>}
       cornerRadii=<0 0 0 0>
      }
      tintColor {colorUseRule=UseColor_Specified, specifiedColor=<rgba 1 1 1 1>}
      uVRegion {bIsValid=false, max=<0 0>, min=<0 0>}
     }
     normalBarImage {
      bIsDynamicallyLoaded=false
      drawAs=Image
      imageType=NoImage
      mirroring=NoMirror
      resourceName=None
      resourceObject=null
      tiling=NoTile
      imageSize=<32 32>
      margin=<ltrb 0 0 0 0>
      outlineSettings {
       bUseBrushTransparency=false
       roundingType=HalfHeightRadius
       width=0
       color {colorUseRule=UseColor_Specified, specifiedColor=<rgba 0 0 0 0>}
       cornerRadii=<0 0 0 0>
      }
      tintColor {colorUseRule=UseColor_Specified, specifiedColor=<rgba 1 1 1 1>}
      uVRegion {bIsValid=false, max=<0 0>, min=<0 0>}
     }
     normalThumbImage {
      bIsDynamicallyLoaded=false
      drawAs=Image
      imageType=FullColor
      mirroring=NoMirror
      resourceName=None
      resourceObject=&Circle
      tiling=NoTile
      imageSize=<20 20>
      margin=<ltrb 0 0 0 0>
      outlineSettings {
       bUseBrushTransparency=false
       roundingType=HalfHeightRadius
       width=0
       color {colorUseRule=UseColor_Specified, specifiedColor=<rgba 0 0 0 0>}
       cornerRadii=<0 0 0 0>
      }
      tintColor {colorUseRule=UseColor_Specified, specifiedColor=<rgba 0.1458 0.1458 0.1458 1>}
      uVRegion {bIsValid=false, max=<0 0>, min=<0 0>}
     }
    }
    selected Style {
     barThickness=2
     backgroundImage {
      bIsDynamicallyLoaded=false
      drawAs=Image
      imageType=NoImage
      mirroring=NoMirror
      resourceName=None
      resourceObject=null
      tiling=NoTile
      imageSize=<32 32>
      margin=<ltrb 0 0 0 0>
      outlineSettings {
       bUseBrushTransparency=false
       roundingType=HalfHeightRadius
       width=0
       color {colorUseRule=UseColor_Specified, specifiedColor=<rgba 0 0 0 0>}
       cornerRadii=<0 0 0 0>
      }
      tintColor {colorUseRule=UseColor_Specified, specifiedColor=<rgba 0.1979 0.1979 0.1979 0.1826>}
      uVRegion {bIsValid=false, max=<0 0>, min=<0 0>}
     }
     disabledBarImage {
      bIsDynamicallyLoaded=false
      drawAs=Image
      imageType=NoImage
      mirroring=NoMirror
      resourceName=None
      resourceObject=null
      tiling=NoTile
      imageSize=<32 32>
      margin=<ltrb 0 0 0 0>
      outlineSettings {
       bUseBrushTransparency=false
       roundingType=HalfHeightRadius
       width=0
       color {colorUseRule=UseColor_Specified, specifiedColor=<rgba 0 0 0 0>}
       cornerRadii=<0 0 0 0>
      }
      tintColor {colorUseRule=UseColor_Specified, specifiedColor=<rgba 1 1 1 1>}
      uVRegion {bIsValid=false, max=<0 0>, min=<0 0>}
     }
     disabledThumbImage {
      bIsDynamicallyLoaded=false
      drawAs=Image
      imageType=NoImage
      mirroring=NoMirror
      resourceName=None
      resourceObject=null
      tiling=NoTile
      imageSize=<32 32>
      margin=<ltrb 0 0 0 0>
      outlineSettings {
       bUseBrushTransparency=false
       roundingType=HalfHeightRadius
       width=0
       color {colorUseRule=UseColor_Specified, specifiedColor=<rgba 0 0 0 0>}
       cornerRadii=<0 0 0 0>
      }
      tintColor {colorUseRule=UseColor_Specified, specifiedColor=<rgba 1 1 1 1>}
      uVRegion {bIsValid=false, max=<0 0>, min=<0 0>}
     }
     normalBarImage {
      bIsDynamicallyLoaded=false
      drawAs=Image
      imageType=NoImage
      mirroring=NoMirror
      resourceName=None
      resourceObject=null
      tiling=NoTile
      imageSize=<32 32>
      margin=<ltrb 0 0 0 0>
      outlineSettings {
       bUseBrushTransparency=false
       roundingType=HalfHeightRadius
       width=0
       color {colorUseRule=UseColor_Specified, specifiedColor=<rgba 0 0 0 0>}
       cornerRadii=<0 0 0 0>
      }
      tintColor {colorUseRule=UseColor_Specified, specifiedColor=<rgba 1 1 1 1>}
      uVRegion {bIsValid=false, max=<0 0>, min=<0 0>}
     }
     normalThumbImage {
      bIsDynamicallyLoaded=false
      drawAs=Image
      imageType=FullColor
      mirroring=NoMirror
      resourceName=None
      resourceObject=&Circle
      tiling=NoTile
      imageSize=<30 30>
      margin=<ltrb 0 0 0 0>
      outlineSettings {
       bUseBrushTransparency=false
       roundingType=HalfHeightRadius
       width=0
       color {colorUseRule=UseColor_Specified, specifiedColor=<rgba 0 0 0 0>}
       cornerRadii=<0 0 0 0>
      }
      tintColor {colorUseRule=UseColor_Specified, specifiedColor=<rgba 1 1 1 1>}
      uVRegion {bIsValid=false, max=<0 0>, min=<0 0>}
     }
    }
   }
   slot {
    class=CanvasPanelSlot
    properties {layoutData {alignment=<0.5 0.5>, anchors {maximum=<0.5 0.5>, minimum=<0.5 0.5>}, offsets=<ltrb -241 -144.422 55.0373 56.237>}}
   }
  }
  CTRL_L_eye_eyelidD {
   widgetClassPath=&EUW_Picker_2DSlider
   bIsVariable=true
   parent=FN_MH_Face_CanvasPanel
   widget=CTRL_L_eye_eyelidD
   properties {
    bOverride_Cursor=true
    max Value X=-1
    max Value Y=-1
    min Value X=1
    min Value Y=1
    normal Style {
     barThickness=2
     backgroundImage {
      bIsDynamicallyLoaded=false
      drawAs=Image
      imageType=NoImage
      mirroring=NoMirror
      resourceName=None
      resourceObject=null
      tiling=NoTile
      imageSize=<32 32>
      margin=<ltrb 0 0 0 0>
      outlineSettings {
       bUseBrushTransparency=false
       roundingType=HalfHeightRadius
       width=0
       color {colorUseRule=UseColor_Specified, specifiedColor=<rgba 0 0 0 0>}
       cornerRadii=<0 0 0 0>
      }
      tintColor {colorUseRule=UseColor_Specified, specifiedColor=<rgba 0 0 0 0.1826>}
      uVRegion {bIsValid=false, max=<0 0>, min=<0 0>}
     }
     disabledBarImage {
      bIsDynamicallyLoaded=false
      drawAs=Image
      imageType=NoImage
      mirroring=NoMirror
      resourceName=None
      resourceObject=null
      tiling=NoTile
      imageSize=<32 32>
      margin=<ltrb 0 0 0 0>
      outlineSettings {
       bUseBrushTransparency=false
       roundingType=HalfHeightRadius
       width=0
       color {colorUseRule=UseColor_Specified, specifiedColor=<rgba 0 0 0 0>}
       cornerRadii=<0 0 0 0>
      }
      tintColor {colorUseRule=UseColor_Specified, specifiedColor=<rgba 1 1 1 1>}
      uVRegion {bIsValid=false, max=<0 0>, min=<0 0>}
     }
     disabledThumbImage {
      bIsDynamicallyLoaded=false
      drawAs=Image
      imageType=NoImage
      mirroring=NoMirror
      resourceName=None
      resourceObject=null
      tiling=NoTile
      imageSize=<32 32>
      margin=<ltrb 0 0 0 0>
      outlineSettings {
       bUseBrushTransparency=false
       roundingType=HalfHeightRadius
       width=0
       color {colorUseRule=UseColor_Specified, specifiedColor=<rgba 0 0 0 0>}
       cornerRadii=<0 0 0 0>
      }
      tintColor {colorUseRule=UseColor_Specified, specifiedColor=<rgba 1 1 1 1>}
      uVRegion {bIsValid=false, max=<0 0>, min=<0 0>}
     }
     normalBarImage {
      bIsDynamicallyLoaded=false
      drawAs=Image
      imageType=NoImage
      mirroring=NoMirror
      resourceName=None
      resourceObject=null
      tiling=NoTile
      imageSize=<32 32>
      margin=<ltrb 0 0 0 0>
      outlineSettings {
       bUseBrushTransparency=false
       roundingType=HalfHeightRadius
       width=0
       color {colorUseRule=UseColor_Specified, specifiedColor=<rgba 0 0 0 0>}
       cornerRadii=<0 0 0 0>
      }
      tintColor {colorUseRule=UseColor_Specified, specifiedColor=<rgba 1 1 1 1>}
      uVRegion {bIsValid=false, max=<0 0>, min=<0 0>}
     }
     normalThumbImage {
      bIsDynamicallyLoaded=false
      drawAs=Image
      imageType=FullColor
      mirroring=NoMirror
      resourceName=None
      resourceObject=&Circle
      tiling=NoTile
      imageSize=<20 20>
      margin=<ltrb 0 0 0 0>
      outlineSettings {
       bUseBrushTransparency=false
       roundingType=HalfHeightRadius
       width=0
       color {colorUseRule=UseColor_Specified, specifiedColor=<rgba 0 0 0 0>}
       cornerRadii=<0 0 0 0>
      }
      tintColor {colorUseRule=UseColor_Specified, specifiedColor=<rgba 0.2604 0.2604 0.2604 1>}
      uVRegion {bIsValid=false, max=<0 0>, min=<0 0>}
     }
    }
    selected Style {
     barThickness=2
     backgroundImage {
      bIsDynamicallyLoaded=false
      drawAs=Image
      imageType=NoImage
      mirroring=NoMirror
      resourceName=None
      resourceObject=null
      tiling=NoTile
      imageSize=<32 32>
      margin=<ltrb 0 0 0 0>
      outlineSettings {
       bUseBrushTransparency=false
       roundingType=HalfHeightRadius
       width=0
       color {colorUseRule=UseColor_Specified, specifiedColor=<rgba 0 0 0 0>}
       cornerRadii=<0 0 0 0>
      }
      tintColor {colorUseRule=UseColor_Specified, specifiedColor=<rgba 0.1979 0.1979 0.1979 0.1826>}
      uVRegion {bIsValid=false, max=<0 0>, min=<0 0>}
     }
     disabledBarImage {
      bIsDynamicallyLoaded=false
      drawAs=Image
      imageType=NoImage
      mirroring=NoMirror
      resourceName=None
      resourceObject=null
      tiling=NoTile
      imageSize=<32 32>
      margin=<ltrb 0 0 0 0>
      outlineSettings {
       bUseBrushTransparency=false
       roundingType=HalfHeightRadius
       width=0
       color {colorUseRule=UseColor_Specified, specifiedColor=<rgba 0 0 0 0>}
       cornerRadii=<0 0 0 0>
      }
      tintColor {colorUseRule=UseColor_Specified, specifiedColor=<rgba 1 1 1 1>}
      uVRegion {bIsValid=false, max=<0 0>, min=<0 0>}
     }
     disabledThumbImage {
      bIsDynamicallyLoaded=false
      drawAs=Image
      imageType=NoImage
      mirroring=NoMirror
      resourceName=None
      resourceObject=null
      tiling=NoTile
      imageSize=<32 32>
      margin=<ltrb 0 0 0 0>
      outlineSettings {
       bUseBrushTransparency=false
       roundingType=HalfHeightRadius
       width=0
       color {colorUseRule=UseColor_Specified, specifiedColor=<rgba 0 0 0 0>}
       cornerRadii=<0 0 0 0>
      }
      tintColor {colorUseRule=UseColor_Specified, specifiedColor=<rgba 1 1 1 1>}
      uVRegion {bIsValid=false, max=<0 0>, min=<0 0>}
     }
     normalBarImage {
      bIsDynamicallyLoaded=false
      drawAs=Image
      imageType=NoImage
      mirroring=NoMirror
      resourceName=None
      resourceObject=null
      tiling=NoTile
      imageSize=<32 32>
      margin=<ltrb 0 0 0 0>
      outlineSettings {
       bUseBrushTransparency=false
       roundingType=HalfHeightRadius
       width=0
       color {colorUseRule=UseColor_Specified, specifiedColor=<rgba 0 0 0 0>}
       cornerRadii=<0 0 0 0>
      }
      tintColor {colorUseRule=UseColor_Specified, specifiedColor=<rgba 1 1 1 1>}
      uVRegion {bIsValid=false, max=<0 0>, min=<0 0>}
     }
     normalThumbImage {
      bIsDynamicallyLoaded=false
      drawAs=Image
      imageType=FullColor
      mirroring=NoMirror
      resourceName=None
      resourceObject=&Circle
      tiling=NoTile
      imageSize=<25 25>
      margin=<ltrb 0 0 0 0>
      outlineSettings {
       bUseBrushTransparency=false
       roundingType=HalfHeightRadius
       width=0
       color {colorUseRule=UseColor_Specified, specifiedColor=<rgba 0 0 0 0>}
       cornerRadii=<0 0 0 0>
      }
      tintColor {colorUseRule=UseColor_Specified, specifiedColor=<rgba 1 1 1 1>}
      uVRegion {bIsValid=false, max=<0 0>, min=<0 0>}
     }
    }
   }
   slot {
    class=CanvasPanelSlot
    properties {layoutData {alignment=<0.5 0.5>, anchors {maximum=<0.5 0.5>, minimum=<0.5 0.5>}, offsets=<ltrb 700.5577 -140 50 50>}}
   }
  }
  CTRL_L_eye_eyelidU {
   widgetClassPath=&EUW_Picker_2DSlider
   bIsVariable=true
   parent=FN_MH_Face_CanvasPanel
   widget=CTRL_L_eye_eyelidU
   properties {
    bOverride_Cursor=true
    max Value X=-1
    min Value X=1
    normal Style {
     barThickness=2
     backgroundImage {
      bIsDynamicallyLoaded=false
      drawAs=Image
      imageType=NoImage
      mirroring=NoMirror
      resourceName=None
      resourceObject=null
      tiling=NoTile
      imageSize=<32 32>
      margin=<ltrb 0 0 0 0>
      outlineSettings {
       bUseBrushTransparency=false
       roundingType=HalfHeightRadius
       width=0
       color {colorUseRule=UseColor_Specified, specifiedColor=<rgba 0 0 0 0>}
       cornerRadii=<0 0 0 0>
      }
      tintColor {colorUseRule=UseColor_Specified, specifiedColor=<rgba 0 0 0 0.1826>}
      uVRegion {bIsValid=false, max=<0 0>, min=<0 0>}
     }
     disabledBarImage {
      bIsDynamicallyLoaded=false
      drawAs=Image
      imageType=NoImage
      mirroring=NoMirror
      resourceName=None
      resourceObject=null
      tiling=NoTile
      imageSize=<32 32>
      margin=<ltrb 0 0 0 0>
      outlineSettings {
       bUseBrushTransparency=false
       roundingType=HalfHeightRadius
       width=0
       color {colorUseRule=UseColor_Specified, specifiedColor=<rgba 0 0 0 0>}
       cornerRadii=<0 0 0 0>
      }
      tintColor {colorUseRule=UseColor_Specified, specifiedColor=<rgba 1 1 1 1>}
      uVRegion {bIsValid=false, max=<0 0>, min=<0 0>}
     }
     disabledThumbImage {
      bIsDynamicallyLoaded=false
      drawAs=Image
      imageType=NoImage
      mirroring=NoMirror
      resourceName=None
      resourceObject=null
      tiling=NoTile
      imageSize=<32 32>
      margin=<ltrb 0 0 0 0>
      outlineSettings {
       bUseBrushTransparency=false
       roundingType=HalfHeightRadius
       width=0
       color {colorUseRule=UseColor_Specified, specifiedColor=<rgba 0 0 0 0>}
       cornerRadii=<0 0 0 0>
      }
      tintColor {colorUseRule=UseColor_Specified, specifiedColor=<rgba 1 1 1 1>}
      uVRegion {bIsValid=false, max=<0 0>, min=<0 0>}
     }
     normalBarImage {
      bIsDynamicallyLoaded=false
      drawAs=Image
      imageType=NoImage
      mirroring=NoMirror
      resourceName=None
      resourceObject=null
      tiling=NoTile
      imageSize=<32 32>
      margin=<ltrb 0 0 0 0>
      outlineSettings {
       bUseBrushTransparency=false
       roundingType=HalfHeightRadius
       width=0
       color {colorUseRule=UseColor_Specified, specifiedColor=<rgba 0 0 0 0>}
       cornerRadii=<0 0 0 0>
      }
      tintColor {colorUseRule=UseColor_Specified, specifiedColor=<rgba 1 1 1 1>}
      uVRegion {bIsValid=false, max=<0 0>, min=<0 0>}
     }
     normalThumbImage {
      bIsDynamicallyLoaded=false
      drawAs=Image
      imageType=FullColor
      mirroring=NoMirror
      resourceName=None
      resourceObject=&Circle
      tiling=NoTile
      imageSize=<20 20>
      margin=<ltrb 0 0 0 0>
      outlineSettings {
       bUseBrushTransparency=false
       roundingType=HalfHeightRadius
       width=0
       color {colorUseRule=UseColor_Specified, specifiedColor=<rgba 0 0 0 0>}
       cornerRadii=<0 0 0 0>
      }
      tintColor {colorUseRule=UseColor_Specified, specifiedColor=<rgba 0.2604 0.2604 0.2604 1>}
      uVRegion {bIsValid=false, max=<0 0>, min=<0 0>}
     }
    }
    selected Style {
     barThickness=2
     backgroundImage {
      bIsDynamicallyLoaded=false
      drawAs=Image
      imageType=NoImage
      mirroring=NoMirror
      resourceName=None
      resourceObject=null
      tiling=NoTile
      imageSize=<32 32>
      margin=<ltrb 0 0 0 0>
      outlineSettings {
       bUseBrushTransparency=false
       roundingType=HalfHeightRadius
       width=0
       color {colorUseRule=UseColor_Specified, specifiedColor=<rgba 0 0 0 0>}
       cornerRadii=<0 0 0 0>
      }
      tintColor {colorUseRule=UseColor_Specified, specifiedColor=<rgba 0.1979 0.1979 0.1979 0.1826>}
      uVRegion {bIsValid=false, max=<0 0>, min=<0 0>}
     }
     disabledBarImage {
      bIsDynamicallyLoaded=false
      drawAs=Image
      imageType=NoImage
      mirroring=NoMirror
      resourceName=None
      resourceObject=null
      tiling=NoTile
      imageSize=<32 32>
      margin=<ltrb 0 0 0 0>
      outlineSettings {
       bUseBrushTransparency=false
       roundingType=HalfHeightRadius
       width=0
       color {colorUseRule=UseColor_Specified, specifiedColor=<rgba 0 0 0 0>}
       cornerRadii=<0 0 0 0>
      }
      tintColor {colorUseRule=UseColor_Specified, specifiedColor=<rgba 1 1 1 1>}
      uVRegion {bIsValid=false, max=<0 0>, min=<0 0>}
     }
     disabledThumbImage {
      bIsDynamicallyLoaded=false
      drawAs=Image
      imageType=NoImage
      mirroring=NoMirror
      resourceName=None
      resourceObject=null
      tiling=NoTile
      imageSize=<32 32>
      margin=<ltrb 0 0 0 0>
      outlineSettings {
       bUseBrushTransparency=false
       roundingType=HalfHeightRadius
       width=0
       color {colorUseRule=UseColor_Specified, specifiedColor=<rgba 0 0 0 0>}
       cornerRadii=<0 0 0 0>
      }
      tintColor {colorUseRule=UseColor_Specified, specifiedColor=<rgba 1 1 1 1>}
      uVRegion {bIsValid=false, max=<0 0>, min=<0 0>}
     }
     normalBarImage {
      bIsDynamicallyLoaded=false
      drawAs=Image
      imageType=NoImage
      mirroring=NoMirror
      resourceName=None
      resourceObject=null
      tiling=NoTile
      imageSize=<32 32>
      margin=<ltrb 0 0 0 0>
      outlineSettings {
       bUseBrushTransparency=false
       roundingType=HalfHeightRadius
       width=0
       color {colorUseRule=UseColor_Specified, specifiedColor=<rgba 0 0 0 0>}
       cornerRadii=<0 0 0 0>
      }
      tintColor {colorUseRule=UseColor_Specified, specifiedColor=<rgba 1 1 1 1>}
      uVRegion {bIsValid=false, max=<0 0>, min=<0 0>}
     }
     normalThumbImage {
      bIsDynamicallyLoaded=false
      drawAs=Image
      imageType=FullColor
      mirroring=NoMirror
      resourceName=None
      resourceObject=&Circle
      tiling=NoTile
      imageSize=<25 25>
      margin=<ltrb 0 0 0 0>
      outlineSettings {
       bUseBrushTransparency=false
       roundingType=HalfHeightRadius
       width=0
       color {colorUseRule=UseColor_Specified, specifiedColor=<rgba 0 0 0 0>}
       cornerRadii=<0 0 0 0>
      }
      tintColor {colorUseRule=UseColor_Specified, specifiedColor=<rgba 1 1 1 1>}
      uVRegion {bIsValid=false, max=<0 0>, min=<0 0>}
     }
    }
   }
   slot {
    class=CanvasPanelSlot
    properties {layoutData {alignment=<0.5 0.5>, anchors {maximum=<0.5 0.5>, minimum=<0.5 0.5>}, offsets=<ltrb 700.5577 -196 50 50>}}
   }
  }
  CTRL_L_nose {
   widgetClassPath=&EUW_Picker_2DSlider
   bIsVariable=true
   parent=FN_MH_Face_CanvasPanel
   widget=CTRL_L_nose
   properties {
    bOverride_Cursor=true
    max Value Y=-1
    min Value Y=1
    normal Style {
     barThickness=2
     backgroundImage {
      bIsDynamicallyLoaded=false
      drawAs=Image
      imageType=NoImage
      mirroring=NoMirror
      resourceName=None
      resourceObject=null
      tiling=NoTile
      imageSize=<32 32>
      margin=<ltrb 0 0 0 0>
      outlineSettings {
       bUseBrushTransparency=false
       roundingType=HalfHeightRadius
       width=0
       color {colorUseRule=UseColor_Specified, specifiedColor=<rgba 0 0 0 0>}
       cornerRadii=<0 0 0 0>
      }
      tintColor {colorUseRule=UseColor_Specified, specifiedColor=<rgba 0 0 0 0.1826>}
      uVRegion {bIsValid=false, max=<0 0>, min=<0 0>}
     }
     disabledBarImage {
      bIsDynamicallyLoaded=false
      drawAs=Image
      imageType=NoImage
      mirroring=NoMirror
      resourceName=None
      resourceObject=null
      tiling=NoTile
      imageSize=<32 32>
      margin=<ltrb 0 0 0 0>
      outlineSettings {
       bUseBrushTransparency=false
       roundingType=HalfHeightRadius
       width=0
       color {colorUseRule=UseColor_Specified, specifiedColor=<rgba 0 0 0 0>}
       cornerRadii=<0 0 0 0>
      }
      tintColor {colorUseRule=UseColor_Specified, specifiedColor=<rgba 1 1 1 1>}
      uVRegion {bIsValid=false, max=<0 0>, min=<0 0>}
     }
     disabledThumbImage {
      bIsDynamicallyLoaded=false
      drawAs=Image
      imageType=NoImage
      mirroring=NoMirror
      resourceName=None
      resourceObject=null
      tiling=NoTile
      imageSize=<32 32>
      margin=<ltrb 0 0 0 0>
      outlineSettings {
       bUseBrushTransparency=false
       roundingType=HalfHeightRadius
       width=0
       color {colorUseRule=UseColor_Specified, specifiedColor=<rgba 0 0 0 0>}
       cornerRadii=<0 0 0 0>
      }
      tintColor {colorUseRule=UseColor_Specified, specifiedColor=<rgba 1 1 1 1>}
      uVRegion {bIsValid=false, max=<0 0>, min=<0 0>}
     }
     normalBarImage {
      bIsDynamicallyLoaded=false
      drawAs=Image
      imageType=NoImage
      mirroring=NoMirror
      resourceName=None
      resourceObject=null
      tiling=NoTile
      imageSize=<32 32>
      margin=<ltrb 0 0 0 0>
      outlineSettings {
       bUseBrushTransparency=false
       roundingType=HalfHeightRadius
       width=0
       color {colorUseRule=UseColor_Specified, specifiedColor=<rgba 0 0 0 0>}
       cornerRadii=<0 0 0 0>
      }
      tintColor {colorUseRule=UseColor_Specified, specifiedColor=<rgba 1 1 1 1>}
      uVRegion {bIsValid=false, max=<0 0>, min=<0 0>}
     }
     normalThumbImage {
      bIsDynamicallyLoaded=false
      drawAs=Image
      imageType=FullColor
      mirroring=NoMirror
      resourceName=None
      resourceObject=&Circle
      tiling=NoTile
      imageSize=<15 15>
      margin=<ltrb 0 0 0 0>
      outlineSettings {
       bUseBrushTransparency=false
       roundingType=HalfHeightRadius
       width=0
       color {colorUseRule=UseColor_Specified, specifiedColor=<rgba 0 0 0 0>}
       cornerRadii=<0 0 0 0>
      }
      tintColor {colorUseRule=UseColor_Specified, specifiedColor=<rgba 0.2604 0.2604 0.2604 1>}
      uVRegion {bIsValid=false, max=<0 0>, min=<0 0>}
     }
    }
    selected Style {
     barThickness=2
     backgroundImage {
      bIsDynamicallyLoaded=false
      drawAs=Image
      imageType=NoImage
      mirroring=NoMirror
      resourceName=None
      resourceObject=null
      tiling=NoTile
      imageSize=<32 32>
      margin=<ltrb 0 0 0 0>
      outlineSettings {
       bUseBrushTransparency=false
       roundingType=HalfHeightRadius
       width=0
       color {colorUseRule=UseColor_Specified, specifiedColor=<rgba 0 0 0 0>}
       cornerRadii=<0 0 0 0>
      }
      tintColor {colorUseRule=UseColor_Specified, specifiedColor=<rgba 0.1979 0.1979 0.1979 0.1826>}
      uVRegion {bIsValid=false, max=<0 0>, min=<0 0>}
     }
     disabledBarImage {
      bIsDynamicallyLoaded=false
      drawAs=Image
      imageType=NoImage
      mirroring=NoMirror
      resourceName=None
      resourceObject=null
      tiling=NoTile
      imageSize=<32 32>
      margin=<ltrb 0 0 0 0>
      outlineSettings {
       bUseBrushTransparency=false
       roundingType=HalfHeightRadius
       width=0
       color {colorUseRule=UseColor_Specified, specifiedColor=<rgba 0 0 0 0>}
       cornerRadii=<0 0 0 0>
      }
      tintColor {colorUseRule=UseColor_Specified, specifiedColor=<rgba 1 1 1 1>}
      uVRegion {bIsValid=false, max=<0 0>, min=<0 0>}
     }
     disabledThumbImage {
      bIsDynamicallyLoaded=false
      drawAs=Image
      imageType=NoImage
      mirroring=NoMirror
      resourceName=None
      resourceObject=null
      tiling=NoTile
      imageSize=<32 32>
      margin=<ltrb 0 0 0 0>
      outlineSettings {
       bUseBrushTransparency=false
       roundingType=HalfHeightRadius
       width=0
       color {colorUseRule=UseColor_Specified, specifiedColor=<rgba 0 0 0 0>}
       cornerRadii=<0 0 0 0>
      }
      tintColor {colorUseRule=UseColor_Specified, specifiedColor=<rgba 1 1 1 1>}
      uVRegion {bIsValid=false, max=<0 0>, min=<0 0>}
     }
     normalBarImage {
      bIsDynamicallyLoaded=false
      drawAs=Image
      imageType=NoImage
      mirroring=NoMirror
      resourceName=None
      resourceObject=null
      tiling=NoTile
      imageSize=<32 32>
      margin=<ltrb 0 0 0 0>
      outlineSettings {
       bUseBrushTransparency=false
       roundingType=HalfHeightRadius
       width=0
       color {colorUseRule=UseColor_Specified, specifiedColor=<rgba 0 0 0 0>}
       cornerRadii=<0 0 0 0>
      }
      tintColor {colorUseRule=UseColor_Specified, specifiedColor=<rgba 1 1 1 1>}
      uVRegion {bIsValid=false, max=<0 0>, min=<0 0>}
     }
     normalThumbImage {
      bIsDynamicallyLoaded=false
      drawAs=Image
      imageType=FullColor
      mirroring=NoMirror
      resourceName=None
      resourceObject=&Circle
      tiling=NoTile
      imageSize=<26 26>
      margin=<ltrb 0 0 0 0>
      outlineSettings {
       bUseBrushTransparency=false
       roundingType=HalfHeightRadius
       width=0
       color {colorUseRule=UseColor_Specified, specifiedColor=<rgba 0 0 0 0>}
       cornerRadii=<0 0 0 0>
      }
      tintColor {colorUseRule=UseColor_Specified, specifiedColor=<rgba 1 1 1 1>}
      uVRegion {bIsValid=false, max=<0 0>, min=<0 0>}
     }
    }
   }
   slot {
    class=CanvasPanelSlot
    properties {layoutData {alignment=<0.5 0.5>, anchors {maximum=<0.5 0.5>, minimum=<0.5 0.5>}, offsets=<ltrb -320 -58 55.0373 56.237>}}
   }
  }
  CTRL_C_mouth {
   widgetClassPath=&EUW_Picker_2DSlider
   bIsVariable=true
   parent=FN_MH_Face_CanvasPanel
   widget=CTRL_C_mouth
   properties {bOverride_Cursor=true, max Value Y=-1, min Value Y=1}
   slot {
    class=CanvasPanelSlot
    properties {layoutData {alignment=<0.5 0.5>, anchors {maximum=<0.5 0.5>, minimum=<0.5 0.5>}, offsets=<ltrb -350.0001 74.9812 79.3733 75.308>}}
   }
  }
  CTRL_C_jaw {
   widgetClassPath=&EUW_Picker_2DSlider
   bIsVariable=true
   parent=FN_MH_Face_CanvasPanel
   widget=CTRL_C_jaw
   properties {bOverride_Cursor=true, max Value X=-1, min Value X=1}
   slot {
    class=CanvasPanelSlot
    properties {layoutData {alignment=<0.5 0.5>, anchors {maximum=<0.5 0.5>, minimum=<0.5 0.5>}, offsets=<ltrb -350 465 120 120>}}
   }
  }
  CTRL_C_tongue {
   widgetClassPath=&EUW_Picker_2DSlider
   bIsVariable=true
   parent=FN_MH_Face_CanvasPanel
   widget=CTRL_C_tongue
   properties {
    bOverride_Cursor=true
    max Value Y=-1
    min Value Y=1
    normal Style {
     barThickness=2
     backgroundImage {
      bIsDynamicallyLoaded=false
      drawAs=Image
      imageType=NoImage
      mirroring=NoMirror
      resourceName=None
      resourceObject=null
      tiling=NoTile
      imageSize=<32 32>
      margin=<ltrb 0 0 0 0>
      outlineSettings {
       bUseBrushTransparency=false
       roundingType=HalfHeightRadius
       width=0
       color {colorUseRule=UseColor_Specified, specifiedColor=<rgba 0 0 0 0>}
       cornerRadii=<0 0 0 0>
      }
      tintColor {colorUseRule=UseColor_Specified, specifiedColor=<rgba 0 0 0 0.1826>}
      uVRegion {bIsValid=false, max=<0 0>, min=<0 0>}
     }
     disabledBarImage {
      bIsDynamicallyLoaded=false
      drawAs=Image
      imageType=NoImage
      mirroring=NoMirror
      resourceName=None
      resourceObject=null
      tiling=NoTile
      imageSize=<32 32>
      margin=<ltrb 0 0 0 0>
      outlineSettings {
       bUseBrushTransparency=false
       roundingType=HalfHeightRadius
       width=0
       color {colorUseRule=UseColor_Specified, specifiedColor=<rgba 0 0 0 0>}
       cornerRadii=<0 0 0 0>
      }
      tintColor {colorUseRule=UseColor_Specified, specifiedColor=<rgba 1 1 1 1>}
      uVRegion {bIsValid=false, max=<0 0>, min=<0 0>}
     }
     disabledThumbImage {
      bIsDynamicallyLoaded=false
      drawAs=Image
      imageType=NoImage
      mirroring=NoMirror
      resourceName=None
      resourceObject=null
      tiling=NoTile
      imageSize=<32 32>
      margin=<ltrb 0 0 0 0>
      outlineSettings {
       bUseBrushTransparency=false
       roundingType=HalfHeightRadius
       width=0
       color {colorUseRule=UseColor_Specified, specifiedColor=<rgba 0 0 0 0>}
       cornerRadii=<0 0 0 0>
      }
      tintColor {colorUseRule=UseColor_Specified, specifiedColor=<rgba 1 1 1 1>}
      uVRegion {bIsValid=false, max=<0 0>, min=<0 0>}
     }
     normalBarImage {
      bIsDynamicallyLoaded=false
      drawAs=Image
      imageType=NoImage
      mirroring=NoMirror
      resourceName=None
      resourceObject=null
      tiling=NoTile
      imageSize=<32 32>
      margin=<ltrb 0 0 0 0>
      outlineSettings {
       bUseBrushTransparency=false
       roundingType=HalfHeightRadius
       width=0
       color {colorUseRule=UseColor_Specified, specifiedColor=<rgba 0 0 0 0>}
       cornerRadii=<0 0 0 0>
      }
      tintColor {colorUseRule=UseColor_Specified, specifiedColor=<rgba 1 1 1 1>}
      uVRegion {bIsValid=false, max=<0 0>, min=<0 0>}
     }
     normalThumbImage {
      bIsDynamicallyLoaded=false
      drawAs=Image
      imageType=FullColor
      mirroring=NoMirror
      resourceName=None
      resourceObject=&Circle
      tiling=NoTile
      imageSize=<18 18>
      margin=<ltrb 0 0 0 0>
      outlineSettings {
       bUseBrushTransparency=false
       roundingType=HalfHeightRadius
       width=0
       color {colorUseRule=UseColor_Specified, specifiedColor=<rgba 0 0 0 0>}
       cornerRadii=<0 0 0 0>
      }
      tintColor {colorUseRule=UseColor_Specified, specifiedColor=<rgba 0.2604 0.2604 0.2604 1>}
      uVRegion {bIsValid=false, max=<0 0>, min=<0 0>}
     }
    }
    selected Style {
     barThickness=2
     backgroundImage {
      bIsDynamicallyLoaded=false
      drawAs=Image
      imageType=NoImage
      mirroring=NoMirror
      resourceName=None
      resourceObject=null
      tiling=NoTile
      imageSize=<32 32>
      margin=<ltrb 0 0 0 0>
      outlineSettings {
       bUseBrushTransparency=false
       roundingType=HalfHeightRadius
       width=0
       color {colorUseRule=UseColor_Specified, specifiedColor=<rgba 0 0 0 0>}
       cornerRadii=<0 0 0 0>
      }
      tintColor {colorUseRule=UseColor_Specified, specifiedColor=<rgba 0.1979 0.1979 0.1979 0.1826>}
      uVRegion {bIsValid=false, max=<0 0>, min=<0 0>}
     }
     disabledBarImage {
      bIsDynamicallyLoaded=false
      drawAs=Image
      imageType=NoImage
      mirroring=NoMirror
      resourceName=None
      resourceObject=null
      tiling=NoTile
      imageSize=<32 32>
      margin=<ltrb 0 0 0 0>
      outlineSettings {
       bUseBrushTransparency=false
       roundingType=HalfHeightRadius
       width=0
       color {colorUseRule=UseColor_Specified, specifiedColor=<rgba 0 0 0 0>}
       cornerRadii=<0 0 0 0>
      }
      tintColor {colorUseRule=UseColor_Specified, specifiedColor=<rgba 1 1 1 1>}
      uVRegion {bIsValid=false, max=<0 0>, min=<0 0>}
     }
     disabledThumbImage {
      bIsDynamicallyLoaded=false
      drawAs=Image
      imageType=NoImage
      mirroring=NoMirror
      resourceName=None
      resourceObject=null
      tiling=NoTile
      imageSize=<32 32>
      margin=<ltrb 0 0 0 0>
      outlineSettings {
       bUseBrushTransparency=false
       roundingType=HalfHeightRadius
       width=0
       color {colorUseRule=UseColor_Specified, specifiedColor=<rgba 0 0 0 0>}
       cornerRadii=<0 0 0 0>
      }
      tintColor {colorUseRule=UseColor_Specified, specifiedColor=<rgba 1 1 1 1>}
      uVRegion {bIsValid=false, max=<0 0>, min=<0 0>}
     }
     normalBarImage {
      bIsDynamicallyLoaded=false
      drawAs=Image
      imageType=NoImage
      mirroring=NoMirror
      resourceName=None
      resourceObject=null
      tiling=NoTile
      imageSize=<32 32>
      margin=<ltrb 0 0 0 0>
      outlineSettings {
       bUseBrushTransparency=false
       roundingType=HalfHeightRadius
       width=0
       color {colorUseRule=UseColor_Specified, specifiedColor=<rgba 0 0 0 0>}
       cornerRadii=<0 0 0 0>
      }
      tintColor {colorUseRule=UseColor_Specified, specifiedColor=<rgba 1 1 1 1>}
      uVRegion {bIsValid=false, max=<0 0>, min=<0 0>}
     }
     normalThumbImage {
      bIsDynamicallyLoaded=false
      drawAs=Image
      imageType=FullColor
      mirroring=NoMirror
      resourceName=None
      resourceObject=&Circle
      tiling=NoTile
      imageSize=<25 25>
      margin=<ltrb 0 0 0 0>
      outlineSettings {
       bUseBrushTransparency=false
       roundingType=HalfHeightRadius
       width=0
       color {colorUseRule=UseColor_Specified, specifiedColor=<rgba 0 0 0 0>}
       cornerRadii=<0 0 0 0>
      }
      tintColor {colorUseRule=UseColor_Specified, specifiedColor=<rgba 1 1 1 1>}
      uVRegion {bIsValid=false, max=<0 0>, min=<0 0>}
     }
    }
   }
   slot {
    class=CanvasPanelSlot
    properties {layoutData {alignment=<0.5 0.5>, anchors {maximum=<0.5 0.5>, minimum=<0.5 0.5>}, offsets=<ltrb -75.4609 154.9595 55 55>}}
   }
  }
  CTRL_C_tongue_roll {
   widgetClassPath=&EUW_Picker_2DSlider
   bIsVariable=true
   parent=FN_MH_Face_CanvasPanel
   widget=CTRL_C_tongue_roll
   properties {
    bOverride_Cursor=true
    max Value Y=-1
    min Value Y=1
    normal Style {
     barThickness=2
     backgroundImage {
      bIsDynamicallyLoaded=false
      drawAs=Image
      imageType=NoImage
      mirroring=NoMirror
      resourceName=None
      resourceObject=null
      tiling=NoTile
      imageSize=<32 32>
      margin=<ltrb 0 0 0 0>
      outlineSettings {
       bUseBrushTransparency=false
       roundingType=HalfHeightRadius
       width=0
       color {colorUseRule=UseColor_Specified, specifiedColor=<rgba 0 0 0 0>}
       cornerRadii=<0 0 0 0>
      }
      tintColor {colorUseRule=UseColor_Specified, specifiedColor=<rgba 0 0 0 0.1826>}
      uVRegion {bIsValid=false, max=<0 0>, min=<0 0>}
     }
     disabledBarImage {
      bIsDynamicallyLoaded=false
      drawAs=Image
      imageType=NoImage
      mirroring=NoMirror
      resourceName=None
      resourceObject=null
      tiling=NoTile
      imageSize=<32 32>
      margin=<ltrb 0 0 0 0>
      outlineSettings {
       bUseBrushTransparency=false
       roundingType=HalfHeightRadius
       width=0
       color {colorUseRule=UseColor_Specified, specifiedColor=<rgba 0 0 0 0>}
       cornerRadii=<0 0 0 0>
      }
      tintColor {colorUseRule=UseColor_Specified, specifiedColor=<rgba 1 1 1 1>}
      uVRegion {bIsValid=false, max=<0 0>, min=<0 0>}
     }
     disabledThumbImage {
      bIsDynamicallyLoaded=false
      drawAs=Image
      imageType=NoImage
      mirroring=NoMirror
      resourceName=None
      resourceObject=null
      tiling=NoTile
      imageSize=<32 32>
      margin=<ltrb 0 0 0 0>
      outlineSettings {
       bUseBrushTransparency=false
       roundingType=HalfHeightRadius
       width=0
       color {colorUseRule=UseColor_Specified, specifiedColor=<rgba 0 0 0 0>}
       cornerRadii=<0 0 0 0>
      }
      tintColor {colorUseRule=UseColor_Specified, specifiedColor=<rgba 1 1 1 1>}
      uVRegion {bIsValid=false, max=<0 0>, min=<0 0>}
     }
     normalBarImage {
      bIsDynamicallyLoaded=false
      drawAs=Image
      imageType=NoImage
      mirroring=NoMirror
      resourceName=None
      resourceObject=null
      tiling=NoTile
      imageSize=<32 32>
      margin=<ltrb 0 0 0 0>
      outlineSettings {
       bUseBrushTransparency=false
       roundingType=HalfHeightRadius
       width=0
       color {colorUseRule=UseColor_Specified, specifiedColor=<rgba 0 0 0 0>}
       cornerRadii=<0 0 0 0>
      }
      tintColor {colorUseRule=UseColor_Specified, specifiedColor=<rgba 1 1 1 1>}
      uVRegion {bIsValid=false, max=<0 0>, min=<0 0>}
     }
     normalThumbImage {
      bIsDynamicallyLoaded=false
      drawAs=Image
      imageType=FullColor
      mirroring=NoMirror
      resourceName=None
      resourceObject=&Circle
      tiling=NoTile
      imageSize=<18 18>
      margin=<ltrb 0 0 0 0>
      outlineSettings {
       bUseBrushTransparency=false
       roundingType=HalfHeightRadius
       width=0
       color {colorUseRule=UseColor_Specified, specifiedColor=<rgba 0 0 0 0>}
       cornerRadii=<0 0 0 0>
      }
      tintColor {colorUseRule=UseColor_Specified, specifiedColor=<rgba 0.2604 0.2604 0.2604 1>}
      uVRegion {bIsValid=false, max=<0 0>, min=<0 0>}
     }
    }
    selected Style {
     barThickness=2
     backgroundImage {
      bIsDynamicallyLoaded=false
      drawAs=Image
      imageType=NoImage
      mirroring=NoMirror
      resourceName=None
      resourceObject=null
      tiling=NoTile
      imageSize=<32 32>
      margin=<ltrb 0 0 0 0>
      outlineSettings {
       bUseBrushTransparency=false
       roundingType=HalfHeightRadius
       width=0
       color {colorUseRule=UseColor_Specified, specifiedColor=<rgba 0 0 0 0>}
       cornerRadii=<0 0 0 0>
      }
      tintColor {colorUseRule=UseColor_Specified, specifiedColor=<rgba 0.1979 0.1979 0.1979 0.1826>}
      uVRegion {bIsValid=false, max=<0 0>, min=<0 0>}
     }
     disabledBarImage {
      bIsDynamicallyLoaded=false
      drawAs=Image
      imageType=NoImage
      mirroring=NoMirror
      resourceName=None
      resourceObject=null
      tiling=NoTile
      imageSize=<32 32>
      margin=<ltrb 0 0 0 0>
      outlineSettings {
       bUseBrushTransparency=false
       roundingType=HalfHeightRadius
       width=0
       color {colorUseRule=UseColor_Specified, specifiedColor=<rgba 0 0 0 0>}
       cornerRadii=<0 0 0 0>
      }
      tintColor {colorUseRule=UseColor_Specified, specifiedColor=<rgba 1 1 1 1>}
      uVRegion {bIsValid=false, max=<0 0>, min=<0 0>}
     }
     disabledThumbImage {
      bIsDynamicallyLoaded=false
      drawAs=Image
      imageType=NoImage
      mirroring=NoMirror
      resourceName=None
      resourceObject=null
      tiling=NoTile
      imageSize=<32 32>
      margin=<ltrb 0 0 0 0>
      outlineSettings {
       bUseBrushTransparency=false
       roundingType=HalfHeightRadius
       width=0
       color {colorUseRule=UseColor_Specified, specifiedColor=<rgba 0 0 0 0>}
       cornerRadii=<0 0 0 0>
      }
      tintColor {colorUseRule=UseColor_Specified, specifiedColor=<rgba 1 1 1 1>}
      uVRegion {bIsValid=false, max=<0 0>, min=<0 0>}
     }
     normalBarImage {
      bIsDynamicallyLoaded=false
      drawAs=Image
      imageType=NoImage
      mirroring=NoMirror
      resourceName=None
      resourceObject=null
      tiling=NoTile
      imageSize=<32 32>
      margin=<ltrb 0 0 0 0>
      outlineSettings {
       bUseBrushTransparency=false
       roundingType=HalfHeightRadius
       width=0
       color {colorUseRule=UseColor_Specified, specifiedColor=<rgba 0 0 0 0>}
       cornerRadii=<0 0 0 0>
      }
      tintColor {colorUseRule=UseColor_Specified, specifiedColor=<rgba 1 1 1 1>}
      uVRegion {bIsValid=false, max=<0 0>, min=<0 0>}
     }
     normalThumbImage {
      bIsDynamicallyLoaded=false
      drawAs=Image
      imageType=FullColor
      mirroring=NoMirror
      resourceName=None
      resourceObject=&Circle
      tiling=NoTile
      imageSize=<25 25>
      margin=<ltrb 0 0 0 0>
      outlineSettings {
       bUseBrushTransparency=false
       roundingType=HalfHeightRadius
       width=0
       color {colorUseRule=UseColor_Specified, specifiedColor=<rgba 0 0 0 0>}
       cornerRadii=<0 0 0 0>
      }
      tintColor {colorUseRule=UseColor_Specified, specifiedColor=<rgba 1 1 1 1>}
      uVRegion {bIsValid=false, max=<0 0>, min=<0 0>}
     }
    }
   }
   slot {
    class=CanvasPanelSlot
    properties {layoutData {alignment=<0.5 0.5>, anchors {maximum=<0.5 0.5>, minimum=<0.5 0.5>}, offsets=<ltrb -11.4609 154.9595 55 55>}}
   }
  }
  CTRL_R_eye_blink {
   widgetClassPath=/AnimInEdTools/AnimationPicker/PickerElements/EUW_Picker_1DSlider
   bIsVariable=true
   parent=FN_MH_Face_CanvasPanel
   widget=CTRL_R_eye_blink
   properties {bOverride_Cursor=true, renderTransform {angle=86, scale=<1 1>, shear=<0 0>, translation=<0 0>}}
   slot {
    class=CanvasPanelSlot
    properties {layoutData {alignment=<0.5 0.5>, anchors {maximum=<0.5 0.5>, minimum=<0.5 0.5>}, offsets=<ltrb -512.9609 -140.5405 100 10>}}
   }
  }
  CTRL_L_eye_blink {
   widgetClassPath=&EUW_Picker_1DSlider
   bIsVariable=true
   parent=FN_MH_Face_CanvasPanel
   widget=CTRL_L_eye_blink
   properties {bOverride_Cursor=true, renderTransform {angle=94, scale=<1 1>, shear=<0 0>, translation=<0 0>}}
   slot {
    class=CanvasPanelSlot
    properties {layoutData {alignment=<0.5 0.5>, anchors {maximum=<0.5 0.5>, minimum=<0.5 0.5>}, offsets=<ltrb -187.0391 -140.5405 100 10>}}
   }
  }
  CTRL_R_brow_lateral {
   widgetClassPath=&EUW_Picker_1DSlider
   bIsVariable=true
   parent=FN_MH_Face_CanvasPanel
   widget=CTRL_R_brow_lateral
   properties {bOverride_Cursor=true, renderTransform {angle=17, scale=<1 1>, shear=<0 0>, translation=<0 0>}}
   slot {
    class=CanvasPanelSlot
    properties {layoutData {alignment=<0.5 0.5>, anchors {maximum=<0.5 0.5>, minimum=<0.5 0.5>}, offsets=<ltrb -476.9609 -275 100 10>}}
   }
  }
  CTRL_C_eyesConvergence {
   widgetClassPath=&EUW_Picker_1DSlider
   bIsVariable=true
   parent=FN_MH_Face_CanvasPanel
   widget=CTRL_C_eyesConvergence
   properties {min Value=-1}
   slot {
    class=CanvasPanelSlot
    properties {layoutData {alignment=<0.5 0.5>, anchors {maximum=<0.5 0.5>, minimum=<0.5 0.5>}, offsets=<ltrb -353.7455 -200.3483 77.5916 10>}}
   }
  }
  CTRL_L_eye_pupil {
   widgetClassPath=&EUW_Picker_1DSlider
   bIsVariable=true
   parent=FN_MH_Face_CanvasPanel
   widget=CTRL_L_eye_pupil
   properties {min Value=-1}
   slot {
    class=CanvasPanelSlot
    properties {layoutData {alignment=<0.5 0.5>, anchors {maximum=<0.5 0.5>, minimum=<0.5 0.5>}, offsets=<ltrb -241 -183.5405 55 10>}}
   }
  }
  CTRL_R_eye_pupil {
   widgetClassPath=&EUW_Picker_1DSlider
   bIsVariable=true
   parent=FN_MH_Face_CanvasPanel
   widget=CTRL_R_eye_pupil
   properties {min Value=-1}
   slot {
    class=CanvasPanelSlot
    properties {layoutData {alignment=<0.5 0.5>, anchors {maximum=<0.5 0.5>, minimum=<0.5 0.5>}, offsets=<ltrb -459 -183.5405 55 10>}}
   }
  }
  CTRL_lookAtSwitch {
   widgetClassPath=&EUW_Picker_1DSlider
   bIsVariable=true
   parent=FN_MH_Face_CanvasPanel
   widget=CTRL_lookAtSwitch
   properties {bOverride_Cursor=true}
   slot {
    class=CanvasPanelSlot
    properties {layoutData {alignment=<0.5 0.5>, anchors {maximum=<0.5 0.5>, minimum=<0.5 0.5>}, offsets=<ltrb -53.1531 587.2598 115.6156 15.6006>}}
   }
  }
  CTRL_rigLogicSwitch {
   widgetClassPath=&EUW_Picker_1DSlider
   bIsVariable=true
   parent=FN_MH_Face_CanvasPanel
   widget=CTRL_rigLogicSwitch
   properties {bOverride_Cursor=true, colorAndOpacity=<rgba 1 1 1 0.2>}
   slot {
    class=CanvasPanelSlot
    properties {layoutData {alignment=<0.5 0.5>, anchors {maximum=<0.5 0.5>, minimum=<0.5 0.5>}, offsets=<ltrb -53.1531 547.2598 115.6156 15.6006>}}
   }
  }
  CTRL_faceGUIfollowHead {
   widgetClassPath=&EUW_Picker_1DSlider
   bIsVariable=true
   parent=FN_MH_Face_CanvasPanel
   widget=CTRL_faceGUIfollowHead
   properties {bOverride_Cursor=true}
   slot {
    class=CanvasPanelSlot
    properties {layoutData {alignment=<0.5 0.5>, anchors {maximum=<0.5 0.5>, minimum=<0.5 0.5>}, offsets=<ltrb 514.8469 435.2598 115.6156 15.6006>}}
   }
  }
  CTRL_eyesAimFollowHead {
   widgetClassPath=&EUW_Picker_1DSlider
   bIsVariable=true
   parent=FN_MH_Face_CanvasPanel
   widget=CTRL_eyesAimFollowHead
   properties {bOverride_Cursor=true}
   slot {
    class=CanvasPanelSlot
    properties {layoutData {alignment=<0.5 0.5>, anchors {maximum=<0.5 0.5>, minimum=<0.5 0.5>}, offsets=<ltrb 514.8469 475.2598 115.6156 15.6006>}}
   }
  }
  CTRL_R_nose_wrinkleUpper {
   widgetClassPath=&EUW_Picker_1DSlider
   bIsVariable=true
   parent=FN_MH_Face_CanvasPanel
   widget=CTRL_R_nose_wrinkleUpper
   properties {bOverride_Cursor=true, renderTransform {angle=90, scale=<1 1>, shear=<0 0>, translation=<0 0>}}
   slot {
    class=CanvasPanelSlot
    properties {layoutData {alignment=<0.5 0.5>, anchors {maximum=<0.5 0.5>, minimum=<0.5 0.5>}, offsets=<ltrb -414.9577 -195.3989 47.4474 10>}}
   }
  }
  CTRL_L_nose_wrinkleUpper {
   widgetClassPath=&EUW_Picker_1DSlider
   bIsVariable=true
   parent=FN_MH_Face_CanvasPanel
   widget=CTRL_L_nose_wrinkleUpper
   properties {bOverride_Cursor=true, renderTransform {angle=90, scale=<1 1>, shear=<0 0>, translation=<0 0>}}
   slot {
    class=CanvasPanelSlot
    properties {layoutData {alignment=<0.5 0.5>, anchors {maximum=<0.5 0.5>, minimum=<0.5 0.5>}, offsets=<ltrb -285.0423 -195.3989 47.4474 10>}}
   }
  }
  CTRL_L_brow_lateral {
   widgetClassPath=&EUW_Picker_1DSlider
   bIsVariable=true
   parent=FN_MH_Face_CanvasPanel
   widget=CTRL_L_brow_lateral
   properties {bOverride_Cursor=true, renderTransform {angle=163, scale=<1 1>, shear=<0 0>, translation=<0 0>}}
   slot {
    class=CanvasPanelSlot
    properties {layoutData {alignment=<0.5 0.5>, anchors {maximum=<0.5 0.5>, minimum=<0.5 0.5>}, offsets=<ltrb -223.0391 -275 100 10>}}
   }
  }
  CTRL_R_brow_down {
   widgetClassPath=&EUW_Picker_1DSlider
   bIsVariable=true
   parent=FN_MH_Face_CanvasPanel
   widget=CTRL_R_brow_down
   properties {bOverride_Cursor=true, renderTransform {angle=66, scale=<1 1>, shear=<0 0>, translation=<0 0>}}
   slot {
    class=CanvasPanelSlot
    properties {layoutData {alignment=<0 0>, anchors {maximum=<0.5 0.5>, minimum=<0.5 0.5>}, offsets=<ltrb -442 -300 90 10>}}
   }
  }
  CTRL_L_brow_down {
   widgetClassPath=&EUW_Picker_1DSlider
   bIsVariable=true
   parent=FN_MH_Face_CanvasPanel
   widget=CTRL_L_brow_down
   properties {bOverride_Cursor=true, renderTransform {angle=114, scale=<1 1>, shear=<0 0>, translation=<0 0>}}
   slot {
    class=CanvasPanelSlot
    properties {layoutData {alignment=<0 0>, anchors {maximum=<0.5 0.5>, minimum=<0.5 0.5>}, offsets=<ltrb -359 -300 90 10>}}
   }
  }
  CTRL_R_eye_squintInner {
   widgetClassPath=&EUW_Picker_1DSlider
   bIsVariable=true
   parent=FN_MH_Face_CanvasPanel
   widget=CTRL_R_eye_squintInner
   properties {bOverride_Cursor=true, renderTransform {angle=-32, scale=<1 1>, shear=<0 0>, translation=<0 0>}}
   slot {
    class=CanvasPanelSlot
    properties {layoutData {alignment=<0.5 0.5>, anchors {maximum=<0.5 0.5>, minimum=<0.5 0.5>}, offsets=<ltrb -465.6969 -78.5465 82.528 10>}}
   }
  }
  CTRL_L_eye_squintInner {
   widgetClassPath=&EUW_Picker_1DSlider
   bIsVariable=true
   parent=FN_MH_Face_CanvasPanel
   widget=CTRL_L_eye_squintInner
   properties {bOverride_Cursor=true, renderTransform {angle=-148, scale=<1 1>, shear=<0 0>, translation=<0 0>}}
   slot {
    class=CanvasPanelSlot
    properties {layoutData {alignment=<0.5 0.5>, anchors {maximum=<0.5 0.5>, minimum=<0.5 0.5>}, offsets=<ltrb -234.3031 -78.5465 82.528 10>}}
   }
  }
  CTRL_L_brow_outer {
   widgetClassPath=&EUW_Picker_1DSlider
   bIsVariable=true
   parent=FN_MH_Face_CanvasPanel
   widget=CTRL_L_brow_outer
   properties {bOverride_Cursor=true, min Value=-1, renderTransform {angle=-90, scale=<1 1>, shear=<0 0>, translation=<0 0>}}
   slot {
    class=CanvasPanelSlot
    properties {layoutData {alignment=<0.5 0.5>, anchors {maximum=<0.5 0.5>, minimum=<0.5 0.5>}, offsets=<ltrb 742.3031 -326.5465 82.528 10>}}
   }
  }
  CTRL_L_brow_inner {
   widgetClassPath=&EUW_Picker_1DSlider
   bIsVariable=true
   parent=FN_MH_Face_CanvasPanel
   widget=CTRL_L_brow_inner
   properties {bOverride_Cursor=true, min Value=-1, renderTransform {angle=-90, scale=<1 1>, shear=<0 0>, translation=<0 0>}}
   slot {
    class=CanvasPanelSlot
    properties {layoutData {alignment=<0.5 0.5>, anchors {maximum=<0.5 0.5>, minimum=<0.5 0.5>}, offsets=<ltrb 580.303 -339.5405 82.528 10>}}
   }
  }
  CTRL_R_brow_inner {
   widgetClassPath=&EUW_Picker_1DSlider
   bIsVariable=true
   parent=FN_MH_Face_CanvasPanel
   widget=CTRL_R_brow_inner
   properties {bOverride_Cursor=true, min Value=-1, renderTransform {angle=-90, scale=<1 1>, shear=<0 0>, translation=<0 0>}}
   slot {
    class=CanvasPanelSlot
    properties {layoutData {alignment=<0.5 0.5>, anchors {maximum=<0.5 0.5>, minimum=<0.5 0.5>}, offsets=<ltrb 496.303 -339.5405 82.528 10>}}
   }
  }
  CTRL_R_brow_outer {
   widgetClassPath=&EUW_Picker_1DSlider
   bIsVariable=true
   parent=FN_MH_Face_CanvasPanel
   widget=CTRL_R_brow_outer
   properties {bOverride_Cursor=true, min Value=-1, renderTransform {angle=-90, scale=<1 1>, shear=<0 0>, translation=<0 0>}}
   slot {
    class=CanvasPanelSlot
    properties {layoutData {alignment=<0.5 0.5>, anchors {maximum=<0.5 0.5>, minimum=<0.5 0.5>}, offsets=<ltrb 338.303 -326.5465 82.528 10>}}
   }
  }
  CTRL_L_eye_faceScrunch {
   widgetClassPath=&EUW_Picker_1DSlider
   bIsVariable=true
   parent=FN_MH_Face_CanvasPanel
   widget=CTRL_L_eye_faceScrunch
   properties {bOverride_Cursor=true, renderTransform {angle=90, scale=<1 1>, shear=<0 0>, translation=<0 0>}}
   slot {
    class=CanvasPanelSlot
    properties {layoutData {alignment=<0.5 0.5>, anchors {maximum=<0.5 0.5>, minimum=<0.5 0.5>}, offsets=<ltrb 643.5391 -219.5405 65 10>}}
   }
  }
  CTRL_R_eye_faceScrunch {
   widgetClassPath=&EUW_Picker_1DSlider
   bIsVariable=true
   parent=FN_MH_Face_CanvasPanel
   widget=CTRL_R_eye_faceScrunch
   properties {bOverride_Cursor=true, renderTransform {angle=90, scale=<1 1>, shear=<0 0>, translation=<0 0>}}
   slot {
    class=CanvasPanelSlot
    properties {layoutData {alignment=<0.5 0.5>, anchors {maximum=<0.5 0.5>, minimum=<0.5 0.5>}, offsets=<ltrb 439.5391 -219.5405 65 10>}}
   }
  }
  CTRL_L_nose_nasolabialDeepen {
   widgetClassPath=&EUW_Picker_1DSlider
   bIsVariable=true
   parent=FN_MH_Face_CanvasPanel
   widget=CTRL_L_nose_nasolabialDeepen
   properties {bOverride_Cursor=true, renderTransform {angle=-80, scale=<1 1>, shear=<0 0>, translation=<0 0>}}
   slot {
    class=CanvasPanelSlot
    properties {layoutData {alignment=<0.5 0.5>, anchors {maximum=<0.5 0.5>, minimum=<0.5 0.5>}, offsets=<ltrb 580.0391 -140 60 10>}}
   }
  }
  CTRL_R_nose_nasolabialDeepen {
   widgetClassPath=&EUW_Picker_1DSlider
   bIsVariable=true
   parent=FN_MH_Face_CanvasPanel
   widget=CTRL_R_nose_nasolabialDeepen
   properties {bOverride_Cursor=true, renderTransform {angle=-100, scale=<1 1>, shear=<0 0>, translation=<0 0>}}
   slot {
    class=CanvasPanelSlot
    properties {layoutData {alignment=<0.5 0.5>, anchors {maximum=<0.5 0.5>, minimum=<0.5 0.5>}, offsets=<ltrb 504.0391 -140 60 10>}}
   }
  }
  CTRL_R_brow_raiseOut {
   widgetClassPath=&EUW_Picker_1DSlider
   bIsVariable=true
   parent=FN_MH_Face_CanvasPanel
   widget=CTRL_R_brow_raiseOut
   properties {bOverride_Cursor=true, renderTransform {angle=-114, scale=<1 1>, shear=<0 0>, translation=<0 0>}}
   slot {
    class=CanvasPanelSlot
    properties {layoutData {alignment=<0.5 0.5>, anchors {maximum=<0.5 0.5>, minimum=<0.5 0.5>}, offsets=<ltrb -522.248 -410 100 10>}}
   }
  }
  CTRL_L_brow_raiseOut {
   widgetClassPath=&EUW_Picker_1DSlider
   bIsVariable=true
   parent=FN_MH_Face_CanvasPanel
   widget=CTRL_L_brow_raiseOut
   properties {bOverride_Cursor=true, renderTransform {angle=-66, scale=<1 1>, shear=<0 0>, translation=<0 0>}}
   slot {
    class=CanvasPanelSlot
    properties {layoutData {alignment=<0.5 0.5>, anchors {maximum=<0.5 0.5>, minimum=<0.5 0.5>}, offsets=<ltrb -177.752 -410 100 10>}}
   }
  }
  CTRL_R_brow_raiseIn {
   widgetClassPath=&EUW_Picker_1DSlider
   bIsVariable=true
   parent=FN_MH_Face_CanvasPanel
   widget=CTRL_R_brow_raiseIn
   properties {bOverride_Cursor=true, renderTransform {angle=-100, scale=<1 1>, shear=<0 0>, translation=<0 0>}}
   slot {
    class=CanvasPanelSlot
    properties {layoutData {alignment=<0.5 0.5>, anchors {maximum=<0.5 0.5>, minimum=<0.5 0.5>}, offsets=<ltrb -394 -415 90 10>}}
   }
  }
  CTRL_L_brow_raiseIn {
   widgetClassPath=&EUW_Picker_1DSlider
   bIsVariable=true
   parent=FN_MH_Face_CanvasPanel
   widget=CTRL_L_brow_raiseIn
   properties {bOverride_Cursor=true, renderTransform {angle=-80, scale=<1 1>, shear=<0 0>, translation=<0 0>}}
   slot {
    class=CanvasPanelSlot
    properties {layoutData {alignment=<0.5 0.5>, anchors {maximum=<0.5 0.5>, minimum=<0.5 0.5>}, offsets=<ltrb -306 -415 90 10>}}
   }
  }
  CTRL_R_mouth_upperLipRaise {
   widgetClassPath=&EUW_Picker_1DSlider
   bIsVariable=true
   parent=FN_MH_Face_CanvasPanel
   widget=CTRL_R_mouth_upperLipRaise
   properties {bOverride_Cursor=true, renderTransform {angle=-122, scale=<1 1>, shear=<0 0>, translation=<0 0>}}
   slot {
    class=CanvasPanelSlot
    properties {layoutData {alignment=<0.5 0.5>, anchors {maximum=<0.5 0.5>, minimum=<0.5 0.5>}, offsets=<ltrb -396 0 55.8346 10>}}
   }
  }
  CTRL_R_mouth_sharpCornerPull {
   widgetClassPath=&EUW_Picker_1DSlider
   bIsVariable=true
   parent=FN_MH_Face_CanvasPanel
   widget=CTRL_R_mouth_sharpCornerPull
   properties {bOverride_Cursor=true, renderTransform {angle=-130, scale=<1 1>, shear=<0 0>, translation=<0 0>}}
   slot {
    class=CanvasPanelSlot
    properties {layoutData {alignment=<0.5 0.5>, anchors {maximum=<0.5 0.5>, minimum=<0.5 0.5>}, offsets=<ltrb -424 12 58 10>}}
   }
  }
  CTRL_L_mouth_upperLipRaise {
   widgetClassPath=&EUW_Picker_1DSlider
   bIsVariable=true
   parent=FN_MH_Face_CanvasPanel
   widget=CTRL_L_mouth_upperLipRaise
   properties {bOverride_Cursor=true, renderTransform {angle=-58, scale=<1 1>, shear=<0 0>, translation=<0 0>}}
   slot {
    class=CanvasPanelSlot
    properties {layoutData {alignment=<0.5 0.5>, anchors {maximum=<0.5 0.5>, minimum=<0.5 0.5>}, offsets=<ltrb -300 0 55.8346 10>}}
   }
  }
  CTRL_L_mouth_sharpCornerPull {
   widgetClassPath=&EUW_Picker_1DSlider
   bIsVariable=true
   parent=FN_MH_Face_CanvasPanel
   widget=CTRL_L_mouth_sharpCornerPull
   properties {bOverride_Cursor=true, renderTransform {angle=-50, scale=<1 1>, shear=<0 0>, translation=<0 0>}}
   slot {
    class=CanvasPanelSlot
    properties {layoutData {alignment=<0.5 0.5>, anchors {maximum=<0.5 0.5>, minimum=<0.5 0.5>}, offsets=<ltrb -268 12 58 10>}}
   }
  }
  CTRL_R_mouth_cornerPull {
   widgetClassPath=&EUW_Picker_1DSlider
   bIsVariable=true
   parent=FN_MH_Face_CanvasPanel
   widget=CTRL_R_mouth_cornerPull
   properties {bOverride_Cursor=true, renderTransform {angle=-136, scale=<1 1>, shear=<0 0>, translation=<0 0>}}
   slot {
    class=CanvasPanelSlot
    properties {layoutData {alignment=<0.5 0.5>, anchors {maximum=<0.5 0.5>, minimum=<0.5 0.5>}, offsets=<ltrb -456 28 69.8486 10>}}
   }
  }
  CTRL_L_mouth_cornerPull {
   widgetClassPath=&EUW_Picker_1DSlider
   bIsVariable=true
   parent=FN_MH_Face_CanvasPanel
   widget=CTRL_L_mouth_cornerPull
   properties {bOverride_Cursor=true, renderTransform {angle=-44, scale=<1 1>, shear=<0 0>, translation=<0 0>}}
   slot {
    class=CanvasPanelSlot
    properties {layoutData {alignment=<0.5 0.5>, anchors {maximum=<0.5 0.5>, minimum=<0.5 0.5>}, offsets=<ltrb -240 28 69.8486 10>}}
   }
  }
  CTRL_R_mouth_dimple {
   widgetClassPath=&EUW_Picker_1DSlider
   bIsVariable=true
   parent=FN_MH_Face_CanvasPanel
   widget=CTRL_R_mouth_dimple
   properties {bOverride_Cursor=true, renderTransform {angle=-166, scale=<1 1>, shear=<0 0>, translation=<0 0>}}
   slot {
    class=CanvasPanelSlot
    properties {layoutData {alignment=<0.5 0.5>, anchors {maximum=<0.5 0.5>, minimum=<0.5 0.5>}, offsets=<ltrb -482.5311 61.4535 80.8596 10>}}
   }
  }
  CTRL_L_mouth_dimple {
   widgetClassPath=&EUW_Picker_1DSlider
   bIsVariable=true
   parent=FN_MH_Face_CanvasPanel
   widget=CTRL_L_mouth_dimple
   properties {bOverride_Cursor=true, renderTransform {angle=-14, scale=<1 1>, shear=<0 0>, translation=<0 0>}}
   slot {
    class=CanvasPanelSlot
    properties {layoutData {alignment=<0.5 0.5>, anchors {maximum=<0.5 0.5>, minimum=<0.5 0.5>}, offsets=<ltrb -217.4689 61.4535 80.8596 10>}}
   }
  }
  CTRL_R_mouth_cornerDepress {
   widgetClassPath=&EUW_Picker_1DSlider
   bIsVariable=true
   parent=FN_MH_Face_CanvasPanel
   widget=CTRL_R_mouth_cornerDepress
   properties {bOverride_Cursor=true, renderTransform {angle=130, scale=<1 1>, shear=<0 0>, translation=<0 0>}}
   slot {
    class=CanvasPanelSlot
    properties {layoutData {alignment=<0.5 0.5>, anchors {maximum=<0.5 0.5>, minimum=<0.5 0.5>}, offsets=<ltrb -478.5311 125.4535 80.8596 10>}}
   }
  }
  CTRL_L_mouth_cornerDepress {
   widgetClassPath=&EUW_Picker_1DSlider
   bIsVariable=true
   parent=FN_MH_Face_CanvasPanel
   widget=CTRL_L_mouth_cornerDepress
   properties {bOverride_Cursor=true, renderTransform {angle=42, scale=<1 1>, shear=<0 0>, translation=<0 0>}}
   slot {
    class=CanvasPanelSlot
    properties {layoutData {alignment=<0.5 0.5>, anchors {maximum=<0.5 0.5>, minimum=<0.5 0.5>}, offsets=<ltrb -221.4689 125.4535 80.8596 10>}}
   }
  }
  CTRL_R_mouth_stretch {
   widgetClassPath=&EUW_Picker_1DSlider
   bIsVariable=true
   parent=FN_MH_Face_CanvasPanel
   widget=CTRL_R_mouth_stretch
   properties {bOverride_Cursor=true, renderTransform {angle=145, scale=<1 1>, shear=<0 0>, translation=<0 0>}}
   slot {
    class=CanvasPanelSlot
    properties {layoutData {alignment=<0.5 0.5>, anchors {maximum=<0.5 0.5>, minimum=<0.5 0.5>}, offsets=<ltrb -438.5311 145.4535 80.8596 10>}}
   }
  }
  CTRL_L_mouth_stretch {
   widgetClassPath=&EUW_Picker_1DSlider
   bIsVariable=true
   parent=FN_MH_Face_CanvasPanel
   widget=CTRL_L_mouth_stretch
   properties {bOverride_Cursor=true, renderTransform {angle=35, scale=<1 1>, shear=<0 0>, translation=<0 0>}}
   slot {
    class=CanvasPanelSlot
    properties {layoutData {alignment=<0.5 0.5>, anchors {maximum=<0.5 0.5>, minimum=<0.5 0.5>}, offsets=<ltrb -261.4689 145.4535 80.8596 10>}}
   }
  }
  CTRL_C_jaw_fwdBack {
   widgetClassPath=&EUW_Picker_1DSlider
   bIsVariable=true
   parent=FN_MH_Face_CanvasPanel
   widget=CTRL_C_jaw_fwdBack
   properties {bOverride_Cursor=true, renderTransform {angle=-90, scale=<1 1>, shear=<0 0>, translation=<0 0>}}
   slot {
    class=CanvasPanelSlot
    properties {layoutData {alignment=<0.5 0.5>, anchors {maximum=<0.5 0.5>, minimum=<0.5 0.5>}, offsets=<ltrb -268.5413 463.6517 120 10>}}
   }
  }
  CTRL_C_jaw_openExtreme {
   widgetClassPath=&EUW_Picker_1DSlider
   bIsVariable=true
   parent=FN_MH_Face_CanvasPanel
   widget=CTRL_C_jaw_openExtreme
   properties {bOverride_Cursor=true}
   slot {
    class=CanvasPanelSlot
    properties {layoutData {alignment=<0.5 0.5>, anchors {maximum=<0.5 0.5>, minimum=<0.5 0.5>}, offsets=<ltrb -348.5413 543.6517 120 10>}}
   }
  }
  CTRL_R_mouth_lowerLipDepress {
   widgetClassPath=&EUW_Picker_1DSlider
   bIsVariable=true
   parent=FN_MH_Face_CanvasPanel
   widget=CTRL_R_mouth_lowerLipDepress
   properties {bOverride_Cursor=true, renderTransform {angle=113, scale=<1 1>, shear=<0 0>, translation=<0 0>}}
   slot {
    class=CanvasPanelSlot
    properties {layoutData {alignment=<0.5 0.5>, anchors {maximum=<0.5 0.5>, minimum=<0.5 0.5>}, offsets=<ltrb -390.7693 173.4535 104.3832 10>}}
   }
  }
  CTRL_L_mouth_lowerLipDepress {
   widgetClassPath=&EUW_Picker_1DSlider
   bIsVariable=true
   parent=FN_MH_Face_CanvasPanel
   widget=CTRL_L_mouth_lowerLipDepress
   properties {bOverride_Cursor=true, renderTransform {angle=67, scale=<1 1>, shear=<0 0>, translation=<0 0>}}
   slot {
    class=CanvasPanelSlot
    properties {layoutData {alignment=<0.5 0.5>, anchors {maximum=<0.5 0.5>, minimum=<0.5 0.5>}, offsets=<ltrb -309.2307 173.4535 104.3832 10>}}
   }
  }
  CTRL_R_eye_cheekRaise {
   widgetClassPath=&EUW_Picker_1DSlider
   bIsVariable=true
   parent=FN_MH_Face_CanvasPanel
   widget=CTRL_R_eye_cheekRaise
   properties {bOverride_Cursor=true, renderTransform {angle=-90, scale=<1 1>, shear=<0 0>, translation=<0 0>}}
   slot {
    class=CanvasPanelSlot
    properties {layoutData {alignment=<0.5 0.5>, anchors {maximum=<0.5 0.5>, minimum=<0.5 0.5>}, offsets=<ltrb -570.8499 -119.9189 72.2222 10>}}
   }
  }
  CTRL_L_eye_cheekRaise {
   widgetClassPath=&EUW_Picker_1DSlider
   bIsVariable=true
   parent=FN_MH_Face_CanvasPanel
   widget=CTRL_L_eye_cheekRaise
   properties {bOverride_Cursor=true, renderTransform {angle=-90, scale=<1 1>, shear=<0 0>, translation=<0 0>}}
   slot {
    class=CanvasPanelSlot
    properties {layoutData {alignment=<0.5 0.5>, anchors {maximum=<0.5 0.5>, minimum=<0.5 0.5>}, offsets=<ltrb -129.1502 -119.9189 72.2222 10>}}
   }
  }
  CTRL_R_mouth_suckBlow {
   widgetClassPath=&EUW_Picker_1DSlider
   bIsVariable=true
   parent=FN_MH_Face_CanvasPanel
   widget=CTRL_R_mouth_suckBlow
   properties {bOverride_Cursor=true, min Value=-1, renderTransform {angle=180, scale=<1 1>, shear=<0 0>, translation=<0 0>}}
   slot {
    class=CanvasPanelSlot
    properties {layoutData {alignment=<0.5 0.5>, anchors {maximum=<0.5 0.5>, minimum=<0.5 0.5>}, offsets=<ltrb -650.3848 40.6246 145.1523 10>}}
   }
  }
  CTRL_L_mouth_suckBlow {
   widgetClassPath=&EUW_Picker_1DSlider
   bIsVariable=true
   parent=FN_MH_Face_CanvasPanel
   widget=CTRL_L_mouth_suckBlow
   properties {bOverride_Cursor=true, min Value=-1}
   slot {
    class=CanvasPanelSlot
    properties {layoutData {alignment=<0.5 0.5>, anchors {maximum=<0.5 0.5>, minimum=<0.5 0.5>}, offsets=<ltrb -49.6152 40.6246 145.1523 10>}}
   }
  }
  CTRL_C_eyesAim {
   widgetClassPath=/AnimInEdTools/AnimationPicker/EUW_Picker_Button
   bIsVariable=true
   parent=FN_MH_Face_CanvasPanel
   widget=CTRL_C_eyesAim
   properties {
    label Text=Aim
    label Color {colorUseRule=UseColor_Specified, specifiedColor=<rgba 0 0 0 1>}
    normal Tint {colorUseRule=UseColor_Specified, specifiedColor=<rgba 0.1185 0.6771 0.3237 1>}
   }
   slot {
    class=CanvasPanelSlot
    properties {layoutData {alignment=<0.5 0.5>, anchors {maximum=<0.5 0.5>, minimum=<0.5 0.5>}, offsets=<ltrb 538.3047 -230.3528 53.692 21.991>}}
   }
  }
  CTRL_C_teeth_fwdBackD {
   widgetClassPath=&EUW_Picker_1DSlider
   bIsVariable=true
   parent=FN_MH_Face_CanvasPanel
   widget=CTRL_C_teeth_fwdBackD
   properties {bOverride_Cursor=true, min Value=-1, renderTransform {angle=-90, scale=<1 1>, shear=<0 0>, translation=<0 0>}}
   slot {
    class=CanvasPanelSlot
    properties {layoutData {alignment=<0.5 0.5>, anchors {maximum=<0.5 0.5>, minimum=<0.5 0.5>}, offsets=<ltrb 424.0391 39.4595 50 8>}}
   }
  }
  CTRL_C_teeth_rotateD {
   widgetClassPath=&EUW_Picker_1DSlider
   bIsVariable=true
   parent=FN_MH_Face_CanvasPanel
   widget=CTRL_C_teeth_rotateD
   properties {bOverride_Cursor=true, min Value=-1, renderTransform {angle=-90, scale=<1 1>, shear=<0 0>, translation=<0 0>}}
   slot {
    class=CanvasPanelSlot
    properties {layoutData {alignment=<0.5 0.5>, anchors {maximum=<0.5 0.5>, minimum=<0.5 0.5>}, offsets=<ltrb 348 39.4595 50 8>}}
   }
  }
  CTRL_C_teeth_rotateU {
   widgetClassPath=&EUW_Picker_1DSlider
   bIsVariable=true
   parent=FN_MH_Face_CanvasPanel
   widget=CTRL_C_teeth_rotateU
   properties {bOverride_Cursor=true, min Value=-1, renderTransform {angle=-90, scale=<1 1>, shear=<0 0>, translation=<0 0>}}
   slot {
    class=CanvasPanelSlot
    properties {layoutData {alignment=<0.5 0.5>, anchors {maximum=<0.5 0.5>, minimum=<0.5 0.5>}, offsets=<ltrb 424 -24 50 8>}}
   }
  }
  CTRL_C_teeth_fwdBackU {
   widgetClassPath=&EUW_Picker_1DSlider
   bIsVariable=true
   parent=FN_MH_Face_CanvasPanel
   widget=CTRL_C_teeth_fwdBackU
   properties {bOverride_Cursor=true, min Value=-1, renderTransform {angle=-90, scale=<1 1>, shear=<0 0>, translation=<0 0>}}
   slot {
    class=CanvasPanelSlot
    properties {layoutData {alignment=<0.5 0.5>, anchors {maximum=<0.5 0.5>, minimum=<0.5 0.5>}, offsets=<ltrb 348 -24 50 8>}}
   }
  }
  CTRL_C_teethU {
   widgetClassPath=&EUW_Picker_2DSlider
   bIsVariable=true
   parent=FN_MH_Face_CanvasPanel
   widget=CTRL_C_teethU
   properties {
    bOverride_Cursor=true
    max Value Y=-1
    min Value Y=1
    normal Style {
     barThickness=2
     backgroundImage {
      bIsDynamicallyLoaded=false
      drawAs=Image
      imageType=NoImage
      mirroring=NoMirror
      resourceName=None
      resourceObject=null
      tiling=NoTile
      imageSize=<32 32>
      margin=<ltrb 0 0 0 0>
      outlineSettings {
       bUseBrushTransparency=false
       roundingType=HalfHeightRadius
       width=0
       color {colorUseRule=UseColor_Specified, specifiedColor=<rgba 0 0 0 0>}
       cornerRadii=<0 0 0 0>
      }
      tintColor {colorUseRule=UseColor_Specified, specifiedColor=<rgba 0 0 0 0.1826>}
      uVRegion {bIsValid=false, max=<0 0>, min=<0 0>}
     }
     disabledBarImage {
      bIsDynamicallyLoaded=false
      drawAs=Image
      imageType=NoImage
      mirroring=NoMirror
      resourceName=None
      resourceObject=null
      tiling=NoTile
      imageSize=<32 32>
      margin=<ltrb 0 0 0 0>
      outlineSettings {
       bUseBrushTransparency=false
       roundingType=HalfHeightRadius
       width=0
       color {colorUseRule=UseColor_Specified, specifiedColor=<rgba 0 0 0 0>}
       cornerRadii=<0 0 0 0>
      }
      tintColor {colorUseRule=UseColor_Specified, specifiedColor=<rgba 1 1 1 1>}
      uVRegion {bIsValid=false, max=<0 0>, min=<0 0>}
     }
     disabledThumbImage {
      bIsDynamicallyLoaded=false
      drawAs=Image
      imageType=NoImage
      mirroring=NoMirror
      resourceName=None
      resourceObject=null
      tiling=NoTile
      imageSize=<32 32>
      margin=<ltrb 0 0 0 0>
      outlineSettings {
       bUseBrushTransparency=false
       roundingType=HalfHeightRadius
       width=0
       color {colorUseRule=UseColor_Specified, specifiedColor=<rgba 0 0 0 0>}
       cornerRadii=<0 0 0 0>
      }
      tintColor {colorUseRule=UseColor_Specified, specifiedColor=<rgba 1 1 1 1>}
      uVRegion {bIsValid=false, max=<0 0>, min=<0 0>}
     }
     normalBarImage {
      bIsDynamicallyLoaded=false
      drawAs=Image
      imageType=NoImage
      mirroring=NoMirror
      resourceName=None
      resourceObject=null
      tiling=NoTile
      imageSize=<32 32>
      margin=<ltrb 0 0 0 0>
      outlineSettings {
       bUseBrushTransparency=false
       roundingType=HalfHeightRadius
       width=0
       color {colorUseRule=UseColor_Specified, specifiedColor=<rgba 0 0 0 0>}
       cornerRadii=<0 0 0 0>
      }
      tintColor {colorUseRule=UseColor_Specified, specifiedColor=<rgba 1 1 1 1>}
      uVRegion {bIsValid=false, max=<0 0>, min=<0 0>}
     }
     normalThumbImage {
      bIsDynamicallyLoaded=false
      drawAs=Image
      imageType=FullColor
      mirroring=NoMirror
      resourceName=None
      resourceObject=&Circle
      tiling=NoTile
      imageSize=<20 20>
      margin=<ltrb 0 0 0 0>
      outlineSettings {
       bUseBrushTransparency=false
       roundingType=HalfHeightRadius
       width=0
       color {colorUseRule=UseColor_Specified, specifiedColor=<rgba 0 0 0 0>}
       cornerRadii=<0 0 0 0>
      }
      tintColor {colorUseRule=UseColor_Specified, specifiedColor=<rgba 0.2604 0.2604 0.2604 1>}
      uVRegion {bIsValid=false, max=<0 0>, min=<0 0>}
     }
    }
    selected Style {
     barThickness=2
     backgroundImage {
      bIsDynamicallyLoaded=false
      drawAs=Image
      imageType=NoImage
      mirroring=NoMirror
      resourceName=None
      resourceObject=null
      tiling=NoTile
      imageSize=<32 32>
      margin=<ltrb 0 0 0 0>
      outlineSettings {
       bUseBrushTransparency=false
       roundingType=HalfHeightRadius
       width=0
       color {colorUseRule=UseColor_Specified, specifiedColor=<rgba 0 0 0 0>}
       cornerRadii=<0 0 0 0>
      }
      tintColor {colorUseRule=UseColor_Specified, specifiedColor=<rgba 0.1979 0.1979 0.1979 0.1826>}
      uVRegion {bIsValid=false, max=<0 0>, min=<0 0>}
     }
     disabledBarImage {
      bIsDynamicallyLoaded=false
      drawAs=Image
      imageType=NoImage
      mirroring=NoMirror
      resourceName=None
      resourceObject=null
      tiling=NoTile
      imageSize=<32 32>
      margin=<ltrb 0 0 0 0>
      outlineSettings {
       bUseBrushTransparency=false
       roundingType=HalfHeightRadius
       width=0
       color {colorUseRule=UseColor_Specified, specifiedColor=<rgba 0 0 0 0>}
       cornerRadii=<0 0 0 0>
      }
      tintColor {colorUseRule=UseColor_Specified, specifiedColor=<rgba 1 1 1 1>}
      uVRegion {bIsValid=false, max=<0 0>, min=<0 0>}
     }
     disabledThumbImage {
      bIsDynamicallyLoaded=false
      drawAs=Image
      imageType=NoImage
      mirroring=NoMirror
      resourceName=None
      resourceObject=null
      tiling=NoTile
      imageSize=<32 32>
      margin=<ltrb 0 0 0 0>
      outlineSettings {
       bUseBrushTransparency=false
       roundingType=HalfHeightRadius
       width=0
       color {colorUseRule=UseColor_Specified, specifiedColor=<rgba 0 0 0 0>}
       cornerRadii=<0 0 0 0>
      }
      tintColor {colorUseRule=UseColor_Specified, specifiedColor=<rgba 1 1 1 1>}
      uVRegion {bIsValid=false, max=<0 0>, min=<0 0>}
     }
     normalBarImage {
      bIsDynamicallyLoaded=false
      drawAs=Image
      imageType=NoImage
      mirroring=NoMirror
      resourceName=None
      resourceObject=null
      tiling=NoTile
      imageSize=<32 32>
      margin=<ltrb 0 0 0 0>
      outlineSettings {
       bUseBrushTransparency=false
       roundingType=HalfHeightRadius
       width=0
       color {colorUseRule=UseColor_Specified, specifiedColor=<rgba 0 0 0 0>}
       cornerRadii=<0 0 0 0>
      }
      tintColor {colorUseRule=UseColor_Specified, specifiedColor=<rgba 1 1 1 1>}
      uVRegion {bIsValid=false, max=<0 0>, min=<0 0>}
     }
     normalThumbImage {
      bIsDynamicallyLoaded=false
      drawAs=Image
      imageType=FullColor
      mirroring=NoMirror
      resourceName=None
      resourceObject=&Circle
      tiling=NoTile
      imageSize=<25 25>
      margin=<ltrb 0 0 0 0>
      outlineSettings {
       bUseBrushTransparency=false
       roundingType=HalfHeightRadius
       width=0
       color {colorUseRule=UseColor_Specified, specifiedColor=<rgba 0 0 0 0>}
       cornerRadii=<0 0 0 0>
      }
      tintColor {colorUseRule=UseColor_Specified, specifiedColor=<rgba 1 1 1 1>}
      uVRegion {bIsValid=false, max=<0 0>, min=<0 0>}
     }
    }
   }
   slot {
    class=CanvasPanelSlot
    properties {layoutData {alignment=<0.5 0.5>, anchors {maximum=<0.5 0.5>, minimum=<0.5 0.5>}, offsets=<ltrb 388.0391 -23.5405 50 50>}}
   }
  }
  CTRL_C_teethD {
   widgetClassPath=&EUW_Picker_2DSlider
   bIsVariable=true
   parent=FN_MH_Face_CanvasPanel
   widget=CTRL_C_teethD
   properties {
    bOverride_Cursor=true
    max Value Y=-1
    min Value Y=1
    normal Style {
     barThickness=2
     backgroundImage {
      bIsDynamicallyLoaded=false
      drawAs=Image
      imageType=NoImage
      mirroring=NoMirror
      resourceName=None
      resourceObject=null
      tiling=NoTile
      imageSize=<32 32>
      margin=<ltrb 0 0 0 0>
      outlineSettings {
       bUseBrushTransparency=false
       roundingType=HalfHeightRadius
       width=0
       color {colorUseRule=UseColor_Specified, specifiedColor=<rgba 0 0 0 0>}
       cornerRadii=<0 0 0 0>
      }
      tintColor {colorUseRule=UseColor_Specified, specifiedColor=<rgba 0 0 0 0.1826>}
      uVRegion {bIsValid=false, max=<0 0>, min=<0 0>}
     }
     disabledBarImage {
      bIsDynamicallyLoaded=false
      drawAs=Image
      imageType=NoImage
      mirroring=NoMirror
      resourceName=None
      resourceObject=null
      tiling=NoTile
      imageSize=<32 32>
      margin=<ltrb 0 0 0 0>
      outlineSettings {
       bUseBrushTransparency=false
       roundingType=HalfHeightRadius
       width=0
       color {colorUseRule=UseColor_Specified, specifiedColor=<rgba 0 0 0 0>}
       cornerRadii=<0 0 0 0>
      }
      tintColor {colorUseRule=UseColor_Specified, specifiedColor=<rgba 1 1 1 1>}
      uVRegion {bIsValid=false, max=<0 0>, min=<0 0>}
     }
     disabledThumbImage {
      bIsDynamicallyLoaded=false
      drawAs=Image
      imageType=NoImage
      mirroring=NoMirror
      resourceName=None
      resourceObject=null
      tiling=NoTile
      imageSize=<32 32>
      margin=<ltrb 0 0 0 0>
      outlineSettings {
       bUseBrushTransparency=false
       roundingType=HalfHeightRadius
       width=0
       color {colorUseRule=UseColor_Specified, specifiedColor=<rgba 0 0 0 0>}
       cornerRadii=<0 0 0 0>
      }
      tintColor {colorUseRule=UseColor_Specified, specifiedColor=<rgba 1 1 1 1>}
      uVRegion {bIsValid=false, max=<0 0>, min=<0 0>}
     }
     normalBarImage {
      bIsDynamicallyLoaded=false
      drawAs=Image
      imageType=NoImage
      mirroring=NoMirror
      resourceName=None
      resourceObject=null
      tiling=NoTile
      imageSize=<32 32>
      margin=<ltrb 0 0 0 0>
      outlineSettings {
       bUseBrushTransparency=false
       roundingType=HalfHeightRadius
       width=0
       color {colorUseRule=UseColor_Specified, specifiedColor=<rgba 0 0 0 0>}
       cornerRadii=<0 0 0 0>
      }
      tintColor {colorUseRule=UseColor_Specified, specifiedColor=<rgba 1 1 1 1>}
      uVRegion {bIsValid=false, max=<0 0>, min=<0 0>}
     }
     normalThumbImage {
      bIsDynamicallyLoaded=false
      drawAs=Image
      imageType=FullColor
      mirroring=NoMirror
      resourceName=None
      resourceObject=&Circle
      tiling=NoTile
      imageSize=<20 20>
      margin=<ltrb 0 0 0 0>
      outlineSettings {
       bUseBrushTransparency=false
       roundingType=HalfHeightRadius
       width=0
       color {colorUseRule=UseColor_Specified, specifiedColor=<rgba 0 0 0 0>}
       cornerRadii=<0 0 0 0>
      }
      tintColor {colorUseRule=UseColor_Specified, specifiedColor=<rgba 0.2604 0.2604 0.2604 1>}
      uVRegion {bIsValid=false, max=<0 0>, min=<0 0>}
     }
    }
    selected Style {
     barThickness=2
     backgroundImage {
      bIsDynamicallyLoaded=false
      drawAs=Image
      imageType=NoImage
      mirroring=NoMirror
      resourceName=None
      resourceObject=null
      tiling=NoTile
      imageSize=<32 32>
      margin=<ltrb 0 0 0 0>
      outlineSettings {
       bUseBrushTransparency=false
       roundingType=HalfHeightRadius
       width=0
       color {colorUseRule=UseColor_Specified, specifiedColor=<rgba 0 0 0 0>}
       cornerRadii=<0 0 0 0>
      }
      tintColor {colorUseRule=UseColor_Specified, specifiedColor=<rgba 0.1979 0.1979 0.1979 0.1826>}
      uVRegion {bIsValid=false, max=<0 0>, min=<0 0>}
     }
     disabledBarImage {
      bIsDynamicallyLoaded=false
      drawAs=Image
      imageType=NoImage
      mirroring=NoMirror
      resourceName=None
      resourceObject=null
      tiling=NoTile
      imageSize=<32 32>
      margin=<ltrb 0 0 0 0>
      outlineSettings {
       bUseBrushTransparency=false
       roundingType=HalfHeightRadius
       width=0
       color {colorUseRule=UseColor_Specified, specifiedColor=<rgba 0 0 0 0>}
       cornerRadii=<0 0 0 0>
      }
      tintColor {colorUseRule=UseColor_Specified, specifiedColor=<rgba 1 1 1 1>}
      uVRegion {bIsValid=false, max=<0 0>, min=<0 0>}
     }
     disabledThumbImage {
      bIsDynamicallyLoaded=false
      drawAs=Image
      imageType=NoImage
      mirroring=NoMirror
      resourceName=None
      resourceObject=null
      tiling=NoTile
      imageSize=<32 32>
      margin=<ltrb 0 0 0 0>
      outlineSettings {
       bUseBrushTransparency=false
       roundingType=HalfHeightRadius
       width=0
       color {colorUseRule=UseColor_Specified, specifiedColor=<rgba 0 0 0 0>}
       cornerRadii=<0 0 0 0>
      }
      tintColor {colorUseRule=UseColor_Specified, specifiedColor=<rgba 1 1 1 1>}
      uVRegion {bIsValid=false, max=<0 0>, min=<0 0>}
     }
     normalBarImage {
      bIsDynamicallyLoaded=false
      drawAs=Image
      imageType=NoImage
      mirroring=NoMirror
      resourceName=None
      resourceObject=null
      tiling=NoTile
      imageSize=<32 32>
      margin=<ltrb 0 0 0 0>
      outlineSettings {
       bUseBrushTransparency=false
       roundingType=HalfHeightRadius
       width=0
       color {colorUseRule=UseColor_Specified, specifiedColor=<rgba 0 0 0 0>}
       cornerRadii=<0 0 0 0>
      }
      tintColor {colorUseRule=UseColor_Specified, specifiedColor=<rgba 1 1 1 1>}
      uVRegion {bIsValid=false, max=<0 0>, min=<0 0>}
     }
     normalThumbImage {
      bIsDynamicallyLoaded=false
      drawAs=Image
      imageType=FullColor
      mirroring=NoMirror
      resourceName=None
      resourceObject=&Circle
      tiling=NoTile
      imageSize=<25 25>
      margin=<ltrb 0 0 0 0>
      outlineSettings {
       bUseBrushTransparency=false
       roundingType=HalfHeightRadius
       width=0
       color {colorUseRule=UseColor_Specified, specifiedColor=<rgba 0 0 0 0>}
       cornerRadii=<0 0 0 0>
      }
      tintColor {colorUseRule=UseColor_Specified, specifiedColor=<rgba 1 1 1 1>}
      uVRegion {bIsValid=false, max=<0 0>, min=<0 0>}
     }
    }
   }
   slot {
    class=CanvasPanelSlot
    properties {layoutData {alignment=<0.5 0.5>, anchors {maximum=<0.5 0.5>, minimum=<0.5 0.5>}, offsets=<ltrb 388 40.4595 50 50>}}
   }
  }
  CTRL_C_tongue_inOut {
   widgetClassPath=&EUW_Picker_1DSlider
   bIsVariable=true
   parent=FN_MH_Face_CanvasPanel
   widget=CTRL_C_tongue_inOut
   properties {bOverride_Cursor=true, min Value=-1, renderTransform {angle=-90, scale=<1 1>, shear=<0 0>, translation=<0 0>}}
   slot {
    class=CanvasPanelSlot
    properties {layoutData {alignment=<0.5 0.5>, anchors {maximum=<0.5 0.5>, minimum=<0.5 0.5>}, offsets=<ltrb -123.0413 141.9247 75 10>}}
   }
  }
  CTRL_C_tongue_narrowWide {
   widgetClassPath=&EUW_Picker_1DSlider
   bIsVariable=true
   parent=FN_MH_Face_CanvasPanel
   widget=CTRL_C_tongue_narrowWide
   properties {bOverride_Cursor=true, min Value=-1}
   slot {
    class=CanvasPanelSlot
    properties {layoutData {alignment=<0.5 0.5>, anchors {maximum=<0.5 0.5>, minimum=<0.5 0.5>}, offsets=<ltrb -61.4427 109.8208 91.0365 10>}}
   }
  }
  CTRL_R_mouth_purseU {
   widgetClassPath=&EUW_Picker_1DSlider
   bIsVariable=true
   parent=FN_MH_Face_CanvasPanel
   widget=CTRL_R_mouth_purseU
   properties {bOverride_Cursor=true, renderTransform {angle=45, scale=<1 1>, shear=<0 0>, translation=<0 0>}}
   slot {
    class=CanvasPanelSlot
    properties {layoutData {alignment=<0.5 0.5>, anchors {maximum=<0.5 0.5>, minimum=<0.5 0.5>}, offsets=<ltrb -685.9609 319.4595 50 8>}}
   }
  }
  CTRL_R_mouth_lipsPressU {
   widgetClassPath=&EUW_Picker_1DSlider
   bIsVariable=true
   parent=FN_MH_Face_CanvasPanel
   widget=CTRL_R_mouth_lipsPressU
   properties {bOverride_Cursor=true, renderTransform {angle=45, scale=<1 1>, shear=<0 0>, translation=<0 0>}}
   slot {
    class=CanvasPanelSlot
    properties {layoutData {alignment=<0.5 0.5>, anchors {maximum=<0.5 0.5>, minimum=<0.5 0.5>}, offsets=<ltrb -687.9609 499.4595 50 8>}}
   }
  }
  CTRL_R_mouth_purseD {
   widgetClassPath=&EUW_Picker_1DSlider
   bIsVariable=true
   parent=FN_MH_Face_CanvasPanel
   widget=CTRL_R_mouth_purseD
   properties {bOverride_Cursor=true, renderTransform {angle=-45, scale=<1 1>, shear=<0 0>, translation=<0 0>}}
   slot {
    class=CanvasPanelSlot
    properties {layoutData {alignment=<0.5 0.5>, anchors {maximum=<0.5 0.5>, minimum=<0.5 0.5>}, offsets=<ltrb -685.9609 363.4595 50 8>}}
   }
  }
  CTRL_R_mouth_lipsPressD {
   widgetClassPath=&EUW_Picker_1DSlider
   bIsVariable=true
   parent=FN_MH_Face_CanvasPanel
   widget=CTRL_R_mouth_lipsPressD
   properties {
    bIsEnabled=false
    bOverride_Cursor=true
    colorAndOpacity=<rgba 0.2292 0.2292 0.2292 1>
    renderTransform {angle=-45, scale=<1 1>, shear=<0 0>, translation=<0 0>}
   }
   slot {
    class=CanvasPanelSlot
    properties {layoutData {alignment=<0.5 0.5>, anchors {maximum=<0.5 0.5>, minimum=<0.5 0.5>}, offsets=<ltrb -683.9609 543.4595 50 8>}}
   }
  }
  CTRL_L_mouth_purseU {
   widgetClassPath=&EUW_Picker_1DSlider
   bIsVariable=true
   parent=FN_MH_Face_CanvasPanel
   widget=CTRL_L_mouth_purseU
   properties {bOverride_Cursor=true, renderTransform {angle=135, scale=<1 1>, shear=<0 0>, translation=<0 0>}}
   slot {
    class=CanvasPanelSlot
    properties {layoutData {alignment=<0.5 0.5>, anchors {maximum=<0.5 0.5>, minimum=<0.5 0.5>}, offsets=<ltrb -641.9609 319.4595 50 8>}}
   }
  }
  CTRL_L_mouth_lipsPressU {
   widgetClassPath=&EUW_Picker_1DSlider
   bIsVariable=true
   parent=FN_MH_Face_CanvasPanel
   widget=CTRL_L_mouth_lipsPressU
   properties {bOverride_Cursor=true, renderTransform {angle=135, scale=<1 1>, shear=<0 0>, translation=<0 0>}}
   slot {
    class=CanvasPanelSlot
    properties {layoutData {alignment=<0.5 0.5>, anchors {maximum=<0.5 0.5>, minimum=<0.5 0.5>}, offsets=<ltrb -643.9609 499.4595 50 8>}}
   }
  }
  CTRL_L_mouth_purseD {
   widgetClassPath=&EUW_Picker_1DSlider
   bIsVariable=true
   parent=FN_MH_Face_CanvasPanel
   widget=CTRL_L_mouth_purseD
   properties {bOverride_Cursor=true, renderTransform {angle=-135, scale=<1 1>, shear=<0 0>, translation=<0 0>}}
   slot {
    class=CanvasPanelSlot
    properties {layoutData {alignment=<0.5 0.5>, anchors {maximum=<0.5 0.5>, minimum=<0.5 0.5>}, offsets=<ltrb -641.9609 363.4595 50 8>}}
   }
  }
  CTRL_L_mouth_lipsPressD {
   widgetClassPath=&EUW_Picker_1DSlider
   bIsVariable=true
   parent=FN_MH_Face_CanvasPanel
   widget=CTRL_L_mouth_lipsPressD
   properties {
    bIsEnabled=false
    bOverride_Cursor=true
    colorAndOpacity=<rgba 0.2292 0.2292 0.2292 1>
    renderTransform {angle=-135, scale=<1 1>, shear=<0 0>, translation=<0 0>}
   }
   slot {
    class=CanvasPanelSlot
    properties {layoutData {alignment=<0.5 0.5>, anchors {maximum=<0.5 0.5>, minimum=<0.5 0.5>}, offsets=<ltrb -643.9609 543.4595 50 8>}}
   }
  }
  CTRL_R_mouth_towardsU {
   widgetClassPath=&EUW_Picker_1DSlider
   bIsVariable=true
   parent=FN_MH_Face_CanvasPanel
   widget=CTRL_R_mouth_towardsU
   properties {bOverride_Cursor=true, renderTransform {angle=45, scale=<1 1>, shear=<0 0>, translation=<0 0>}}
   slot {
    class=CanvasPanelSlot
    properties {layoutData {alignment=<0.5 0.5>, anchors {maximum=<0.5 0.5>, minimum=<0.5 0.5>}, offsets=<ltrb -793.9609 315.4595 50 8>}}
   }
  }
  CTRL_R_mouth_funnelD {
   widgetClassPath=&EUW_Picker_1DSlider
   bIsVariable=true
   parent=FN_MH_Face_CanvasPanel
   widget=CTRL_R_mouth_funnelD
   properties {bOverride_Cursor=true, renderTransform {angle=-45, scale=<1 1>, shear=<0 0>, translation=<0 0>}}
   slot {
    class=CanvasPanelSlot
    properties {layoutData {alignment=<0.5 0.5>, anchors {maximum=<0.5 0.5>, minimum=<0.5 0.5>}, offsets=<ltrb -573.9609 363.4595 50 8>}}
   }
  }
  CTRL_R_mouth_tightenD {
   widgetClassPath=&EUW_Picker_1DSlider
   bIsVariable=true
   parent=FN_MH_Face_CanvasPanel
   widget=CTRL_R_mouth_tightenD
   properties {bOverride_Cursor=true, renderTransform {angle=-45, scale=<1 1>, shear=<0 0>, translation=<0 0>}}
   slot {
    class=CanvasPanelSlot
    properties {layoutData {alignment=<0.5 0.5>, anchors {maximum=<0.5 0.5>, minimum=<0.5 0.5>}, offsets=<ltrb -575.9609 543.4595 50 8>}}
   }
  }
  CTRL_R_mouth_funnelU {
   widgetClassPath=&EUW_Picker_1DSlider
   bIsVariable=true
   parent=FN_MH_Face_CanvasPanel
   widget=CTRL_R_mouth_funnelU
   properties {bOverride_Cursor=true, renderTransform {angle=45, scale=<1 1>, shear=<0 0>, translation=<0 0>}}
   slot {
    class=CanvasPanelSlot
    properties {layoutData {alignment=<0.5 0.5>, anchors {maximum=<0.5 0.5>, minimum=<0.5 0.5>}, offsets=<ltrb -575.4072 315.9681 50 8>}}
   }
  }
  CTRL_R_mouth_tightenU {
   widgetClassPath=&EUW_Picker_1DSlider
   bIsVariable=true
   parent=FN_MH_Face_CanvasPanel
   widget=CTRL_R_mouth_tightenU
   properties {bOverride_Cursor=true, renderTransform {angle=45, scale=<1 1>, shear=<0 0>, translation=<0 0>}}
   slot {
    class=CanvasPanelSlot
    properties {layoutData {alignment=<0.5 0.5>, anchors {maximum=<0.5 0.5>, minimum=<0.5 0.5>}, offsets=<ltrb -575.9609 499.4595 50 8>}}
   }
  }
  CTRL_L_mouth_funnelU {
   widgetClassPath=&EUW_Picker_1DSlider
   bIsVariable=true
   parent=FN_MH_Face_CanvasPanel
   widget=CTRL_L_mouth_funnelU
   properties {bOverride_Cursor=true, renderTransform {angle=135, scale=<1 1>, shear=<0 0>, translation=<0 0>}}
   slot {
    class=CanvasPanelSlot
    properties {layoutData {alignment=<0.5 0.5>, anchors {maximum=<0.5 0.5>, minimum=<0.5 0.5>}, offsets=<ltrb -525.9609 315.4595 50 8>}}
   }
  }
  CTRL_L_mouth_tightenU {
   widgetClassPath=&EUW_Picker_1DSlider
   bIsVariable=true
   parent=FN_MH_Face_CanvasPanel
   widget=CTRL_L_mouth_tightenU
   properties {bOverride_Cursor=true, renderTransform {angle=135, scale=<1 1>, shear=<0 0>, translation=<0 0>}}
   slot {
    class=CanvasPanelSlot
    properties {layoutData {alignment=<0.5 0.5>, anchors {maximum=<0.5 0.5>, minimum=<0.5 0.5>}, offsets=<ltrb -531.9609 499.4595 50 8>}}
   }
  }
  CTRL_L_mouth_funnelD {
   widgetClassPath=&EUW_Picker_1DSlider
   bIsVariable=true
   parent=FN_MH_Face_CanvasPanel
   widget=CTRL_L_mouth_funnelD
   properties {bOverride_Cursor=true, renderTransform {angle=-135, scale=<1 1>, shear=<0 0>, translation=<0 0>}}
   slot {
    class=CanvasPanelSlot
    properties {layoutData {alignment=<0.5 0.5>, anchors {maximum=<0.5 0.5>, minimum=<0.5 0.5>}, offsets=<ltrb -525.9609 363.4595 50 8>}}
   }
  }
  CTRL_L_mouth_tightenD {
   widgetClassPath=&EUW_Picker_1DSlider
   bIsVariable=true
   parent=FN_MH_Face_CanvasPanel
   widget=CTRL_L_mouth_tightenD
   properties {bOverride_Cursor=true, renderTransform {angle=-135, scale=<1 1>, shear=<0 0>, translation=<0 0>}}
   slot {
    class=CanvasPanelSlot
    properties {layoutData {alignment=<0.5 0.5>, anchors {maximum=<0.5 0.5>, minimum=<0.5 0.5>}, offsets=<ltrb -531.9609 543.4595 50 8>}}
   }
  }
  CTRL_R_mouth_towardsD {
   widgetClassPath=&EUW_Picker_1DSlider
   bIsVariable=true
   parent=FN_MH_Face_CanvasPanel
   widget=CTRL_R_mouth_towardsD
   properties {bOverride_Cursor=true, renderTransform {angle=-45, scale=<1 1>, shear=<0 0>, translation=<0 0>}}
   slot {
    class=CanvasPanelSlot
    properties {layoutData {alignment=<0.5 0.5>, anchors {maximum=<0.5 0.5>, minimum=<0.5 0.5>}, offsets=<ltrb -793.9609 363.4595 50 8>}}
   }
  }
  CTRL_L_mouth_towardsU {
   widgetClassPath=&EUW_Picker_1DSlider
   bIsVariable=true
   parent=FN_MH_Face_CanvasPanel
   widget=CTRL_L_mouth_towardsU
   properties {bOverride_Cursor=true, renderTransform {angle=135, scale=<1 1>, shear=<0 0>, translation=<0 0>}}
   slot {
    class=CanvasPanelSlot
    properties {layoutData {alignment=<0.5 0.5>, anchors {maximum=<0.5 0.5>, minimum=<0.5 0.5>}, offsets=<ltrb -745.9609 315.4595 50 8>}}
   }
  }
  CTRL_L_mouth_towardsD {
   widgetClassPath=&EUW_Picker_1DSlider
   bIsVariable=true
   parent=FN_MH_Face_CanvasPanel
   widget=CTRL_L_mouth_towardsD
   properties {bOverride_Cursor=true, renderTransform {angle=-135, scale=<1 1>, shear=<0 0>, translation=<0 0>}}
   slot {
    class=CanvasPanelSlot
    properties {layoutData {alignment=<0.5 0.5>, anchors {maximum=<0.5 0.5>, minimum=<0.5 0.5>}, offsets=<ltrb -745.9609 363.4595 50 8>}}
   }
  }
  CTRL_R_mouth_lipsBlow {
   widgetClassPath=&EUW_Picker_1DSlider
   bIsVariable=true
   parent=FN_MH_Face_CanvasPanel
   widget=CTRL_R_mouth_lipsBlow
   properties {bOverride_Cursor=true, renderTransform {angle=135, scale=<1 1>, shear=<0 0>, translation=<0 0>}}
   slot {
    class=CanvasPanelSlot
    properties {layoutData {alignment=<0.5 0.5>, anchors {maximum=<0.5 0.5>, minimum=<0.5 0.5>}, offsets=<ltrb -100.9609 256.4595 60 10>}}
   }
  }
  CTRL_L_mouth_lipsBlow {
   widgetClassPath=&EUW_Picker_1DSlider
   bIsVariable=true
   parent=FN_MH_Face_CanvasPanel
   widget=CTRL_L_mouth_lipsBlow
   properties {bOverride_Cursor=true, renderTransform {angle=45, scale=<1 1>, shear=<0 0>, translation=<0 0>}}
   slot {
    class=CanvasPanelSlot
    properties {layoutData {alignment=<0.5 0.5>, anchors {maximum=<0.5 0.5>, minimum=<0.5 0.5>}, offsets=<ltrb -36.9609 256.4595 60 10>}}
   }
  }
  CTRL_R_mouth_cornerSharpnessU {
   widgetClassPath=&EUW_Picker_1DSlider
   bIsVariable=true
   parent=FN_MH_Face_CanvasPanel
   widget=CTRL_R_mouth_cornerSharpnessU
   properties {bOverride_Cursor=true, min Value=-1, renderTransform {angle=90, scale=<1 1>, shear=<0 0>, translation=<0 0>}}
   slot {
    class=CanvasPanelSlot
    properties {layoutData {alignment=<0.5 0.5>, anchors {maximum=<0.5 0.5>, minimum=<0.5 0.5>}, offsets=<ltrb 557.0391 147.4595 60 8>}}
   }
  }
  CTRL_R_mouth_pushPullU {
   widgetClassPath=&EUW_Picker_1DSlider
   bIsVariable=true
   parent=FN_MH_Face_CanvasPanel
   widget=CTRL_R_mouth_pushPullU
   properties {bOverride_Cursor=true, min Value=-1, renderTransform {angle=-135, scale=<1 1>, shear=<0 0>, translation=<20.328 -2.772>}}
   slot {
    class=CanvasPanelSlot
    properties {layoutData {alignment=<0.5 0.5>, anchors {maximum=<0.5 0.5>, minimum=<0.5 0.5>}, offsets=<ltrb 652 148 60 8>}}
   }
  }
  CTRL_R_mouth_lipsRollU {
   widgetClassPath=&EUW_Picker_1DSlider
   bIsVariable=true
   parent=FN_MH_Face_CanvasPanel
   widget=CTRL_R_mouth_lipsRollU
   properties {bOverride_Cursor=true, min Value=-1, renderTransform {angle=90, scale=<1 1>, shear=<0 0>, translation=<0 0>}}
   slot {
    class=CanvasPanelSlot
    properties {layoutData {alignment=<0.5 0.5>, anchors {maximum=<0.5 0.5>, minimum=<0.5 0.5>}, offsets=<ltrb 480 147.4595 60 8>}}
   }
  }
  CTRL_R_mouth_thicknessU {
   widgetClassPath=&EUW_Picker_1DSlider
   bIsVariable=true
   parent=FN_MH_Face_CanvasPanel
   widget=CTRL_R_mouth_thicknessU
   properties {bOverride_Cursor=true, min Value=-1, renderTransform {angle=90, scale=<1 1>, shear=<0 0>, translation=<0 0>}}
   slot {
    class=CanvasPanelSlot
    properties {layoutData {alignment=<0.5 0.5>, anchors {maximum=<0.5 0.5>, minimum=<0.5 0.5>}, offsets=<ltrb 368 147.4595 60 8>}}
   }
  }
  CTRL_R_mouth_thicknessInwardU {
   widgetClassPath=&EUW_Picker_1DSlider
   bIsVariable=true
   parent=FN_MH_Face_CanvasPanel
   widget=CTRL_R_mouth_thicknessInwardU
   properties {bOverride_Cursor=true, min Value=-1, renderTransform {angle=90, scale=<1 1>, shear=<0 0>, translation=<-0.462 0.462>}}
   slot {
    class=CanvasPanelSlot
    properties {layoutData {alignment=<0.5 0.5>, anchors {maximum=<0.5 0.5>, minimum=<0.5 0.5>}, offsets=<ltrb 408 147.4595 60 8>}}
   }
  }
  CTRL_L_mouth_cornerSharpnessU {
   widgetClassPath=&EUW_Picker_1DSlider
   bIsVariable=true
   parent=FN_MH_Face_CanvasPanel
   widget=CTRL_L_mouth_cornerSharpnessU
   properties {bOverride_Cursor=true, min Value=-1, renderTransform {angle=90, scale=<1 1>, shear=<0 0>, translation=<0 0>}}
   slot {
    class=CanvasPanelSlot
    properties {layoutData {alignment=<0.5 0.5>, anchors {maximum=<0.5 0.5>, minimum=<0.5 0.5>}, offsets=<ltrb 576 148 60 8>}}
   }
  }
  CTRL_L_mouth_pushPullU {
   widgetClassPath=&EUW_Picker_1DSlider
   bIsVariable=true
   parent=FN_MH_Face_CanvasPanel
   widget=CTRL_L_mouth_pushPullU
   properties {bOverride_Cursor=true, min Value=-1, renderTransform {angle=-45, scale=<1 1>, shear=<0 0>, translation=<49.4341 -5.544>}}
   slot {
    class=CanvasPanelSlot
    properties {layoutData {alignment=<0.5 0.5>, anchors {maximum=<0.5 0.5>, minimum=<0.5 0.5>}, offsets=<ltrb 676 148 60 8>}}
   }
  }
  CTRL_L_mouth_lipsRollU {
   widgetClassPath=&EUW_Picker_1DSlider
   bIsVariable=true
   parent=FN_MH_Face_CanvasPanel
   widget=CTRL_L_mouth_lipsRollU
   properties {bOverride_Cursor=true, min Value=-1, renderTransform {angle=90, scale=<1 1>, shear=<0 0>, translation=<0 0>}}
   slot {
    class=CanvasPanelSlot
    properties {layoutData {alignment=<0.5 0.5>, anchors {maximum=<0.5 0.5>, minimum=<0.5 0.5>}, offsets=<ltrb 500 147.4595 60 8>}}
   }
  }
  CTRL_L_mouth_thicknessU {
   widgetClassPath=&EUW_Picker_1DSlider
   bIsVariable=true
   parent=FN_MH_Face_CanvasPanel
   widget=CTRL_L_mouth_thicknessU
   properties {bOverride_Cursor=true, min Value=-1, renderTransform {angle=90, scale=<1 1>, shear=<0 0>, translation=<0 0>}}
   slot {
    class=CanvasPanelSlot
    properties {layoutData {alignment=<0.5 0.5>, anchors {maximum=<0.5 0.5>, minimum=<0.5 0.5>}, offsets=<ltrb 384 147.4595 60 8>}}
   }
  }
  CTRL_L_mouth_thicknessInwardU {
   widgetClassPath=&EUW_Picker_1DSlider
   bIsVariable=true
   parent=FN_MH_Face_CanvasPanel
   widget=CTRL_L_mouth_thicknessInwardU
   properties {bOverride_Cursor=true, min Value=-1, renderTransform {angle=90, scale=<1 1>, shear=<0 0>, translation=<0 0>}}
   slot {
    class=CanvasPanelSlot
    properties {layoutData {alignment=<0.5 0.5>, anchors {maximum=<0.5 0.5>, minimum=<0.5 0.5>}, offsets=<ltrb 424 147.4595 60 8>}}
   }
  }
  CTRL_L_mouth_cornerSharpnessD {
   widgetClassPath=&EUW_Picker_1DSlider
   bIsVariable=true
   parent=FN_MH_Face_CanvasPanel
   widget=CTRL_L_mouth_cornerSharpnessD
   properties {bOverride_Cursor=true, min Value=-1, renderTransform {angle=-90, scale=<1 1>, shear=<0 0>, translation=<0 0>}}
   slot {
    class=CanvasPanelSlot
    properties {layoutData {alignment=<0.5 0.5>, anchors {maximum=<0.5 0.5>, minimum=<0.5 0.5>}, offsets=<ltrb 576 212 60 8>}}
   }
  }
  CTRL_L_mouth_pushPullD {
   widgetClassPath=&EUW_Picker_1DSlider
   bIsVariable=true
   parent=FN_MH_Face_CanvasPanel
   widget=CTRL_L_mouth_pushPullD
   properties {bOverride_Cursor=true, min Value=-1, renderTransform {angle=45, scale=<1 1>, shear=<0 0>, translation=<49.8961 -14.322>}}
   slot {
    class=CanvasPanelSlot
    properties {layoutData {alignment=<0.5 0.5>, anchors {maximum=<0.5 0.5>, minimum=<0.5 0.5>}, offsets=<ltrb 676 212 60 8>}}
   }
  }
  CTRL_L_mouth_lipsRollD {
   widgetClassPath=&EUW_Picker_1DSlider
   bIsVariable=true
   parent=FN_MH_Face_CanvasPanel
   widget=CTRL_L_mouth_lipsRollD
   properties {bOverride_Cursor=true, min Value=-1, renderTransform {angle=-90, scale=<1 1>, shear=<0 0>, translation=<0 0>}}
   slot {
    class=CanvasPanelSlot
    properties {layoutData {alignment=<0.5 0.5>, anchors {maximum=<0.5 0.5>, minimum=<0.5 0.5>}, offsets=<ltrb 500 212 60 8>}}
   }
  }
  CTRL_L_mouth_thicknessD {
   widgetClassPath=&EUW_Picker_1DSlider
   bIsVariable=true
   parent=FN_MH_Face_CanvasPanel
   widget=CTRL_L_mouth_thicknessD
   properties {bOverride_Cursor=true, min Value=-1, renderTransform {angle=-90, scale=<1 1>, shear=<0 0>, translation=<0 0>}}
   slot {
    class=CanvasPanelSlot
    properties {layoutData {alignment=<0.5 0.5>, anchors {maximum=<0.5 0.5>, minimum=<0.5 0.5>}, offsets=<ltrb 384 212 60 8>}}
   }
  }
  CTRL_L_mouth_thicknessInwardD {
   widgetClassPath=&EUW_Picker_1DSlider
   bIsVariable=true
   parent=FN_MH_Face_CanvasPanel
   widget=CTRL_L_mouth_thicknessInwardD
   properties {bOverride_Cursor=true, min Value=-1, renderTransform {angle=-90, scale=<1 1>, shear=<0 0>, translation=<0 0>}}
   slot {
    class=CanvasPanelSlot
    properties {layoutData {alignment=<0.5 0.5>, anchors {maximum=<0.5 0.5>, minimum=<0.5 0.5>}, offsets=<ltrb 424 212 60 8>}}
   }
  }
  CTRL_R_mouth_cornerSharpnessD {
   widgetClassPath=&EUW_Picker_1DSlider
   bIsVariable=true
   parent=FN_MH_Face_CanvasPanel
   widget=CTRL_R_mouth_cornerSharpnessD
   properties {bOverride_Cursor=true, min Value=-1, renderTransform {angle=-90, scale=<1 1>, shear=<0 0>, translation=<0 0>}}
   slot {
    class=CanvasPanelSlot
    properties {layoutData {alignment=<0.5 0.5>, anchors {maximum=<0.5 0.5>, minimum=<0.5 0.5>}, offsets=<ltrb 556 212 60 8>}}
   }
  }
  CTRL_R_mouth_pushPullD {
   widgetClassPath=&EUW_Picker_1DSlider
   bIsVariable=true
   parent=FN_MH_Face_CanvasPanel
   widget=CTRL_R_mouth_pushPullD
   properties {bOverride_Cursor=true, min Value=-1, renderTransform {angle=135, scale=<1 1>, shear=<0 0>, translation=<19.866 -13.86>}}
   slot {
    class=CanvasPanelSlot
    properties {layoutData {alignment=<0.5 0.5>, anchors {maximum=<0.5 0.5>, minimum=<0.5 0.5>}, offsets=<ltrb 652 212 60 8>}}
   }
  }
  CTRL_R_mouth_lipsRollD {
   widgetClassPath=&EUW_Picker_1DSlider
   bIsVariable=true
   parent=FN_MH_Face_CanvasPanel
   widget=CTRL_R_mouth_lipsRollD
   properties {bOverride_Cursor=true, min Value=-1, renderTransform {angle=-90, scale=<1 1>, shear=<0 0>, translation=<0 0>}}
   slot {
    class=CanvasPanelSlot
    properties {layoutData {alignment=<0.5 0.5>, anchors {maximum=<0.5 0.5>, minimum=<0.5 0.5>}, offsets=<ltrb 480 212 60 8>}}
   }
  }
  CTRL_R_mouth_thicknessD {
   widgetClassPath=&EUW_Picker_1DSlider
   bIsVariable=true
   parent=FN_MH_Face_CanvasPanel
   widget=CTRL_R_mouth_thicknessD
   properties {bOverride_Cursor=true, min Value=-1, renderTransform {angle=-90, scale=<1 1>, shear=<0 0>, translation=<0 0>}}
   slot {
    class=CanvasPanelSlot
    properties {layoutData {alignment=<0.5 0.5>, anchors {maximum=<0.5 0.5>, minimum=<0.5 0.5>}, offsets=<ltrb 368 212 60 8>}}
   }
  }
  CTRL_R_mouth_thicknessInwardD {
   widgetClassPath=&EUW_Picker_1DSlider
   bIsVariable=true
   parent=FN_MH_Face_CanvasPanel
   widget=CTRL_R_mouth_thicknessInwardD
   properties {bOverride_Cursor=true, min Value=-1, renderTransform {angle=-90, scale=<1 1>, shear=<0 0>, translation=<0 0.462>}}
   slot {
    class=CanvasPanelSlot
    properties {layoutData {alignment=<0.5 0.5>, anchors {maximum=<0.5 0.5>, minimum=<0.5 0.5>}, offsets=<ltrb 408 212 60 8>}}
   }
  }
  CTRL_L_mouth_lipsTogetherU {
   widgetClassPath=&EUW_Picker_1DSlider
   bIsVariable=true
   parent=FN_MH_Face_CanvasPanel
   widget=CTRL_L_mouth_lipsTogetherU
   properties {bOverride_Cursor=true, renderTransform {angle=135, scale=<1 1>, shear=<0 0>, translation=<0 0>}}
   slot {
    class=CanvasPanelSlot
    properties {layoutData {alignment=<0.5 0.5>, anchors {maximum=<0.5 0.5>, minimum=<0.5 0.5>}, offsets=<ltrb -96.9609 371.4595 60 8>}}
   }
  }
  CTRL_R_mouth_lipsTogetherU {
   widgetClassPath=&EUW_Picker_1DSlider
   bIsVariable=true
   parent=FN_MH_Face_CanvasPanel
   widget=CTRL_R_mouth_lipsTogetherU
   properties {bOverride_Cursor=true, renderTransform {angle=45, scale=<1 1>, shear=<0 0>, translation=<0 0>}}
   slot {
    class=CanvasPanelSlot
    properties {layoutData {alignment=<0.5 0.5>, anchors {maximum=<0.5 0.5>, minimum=<0.5 0.5>}, offsets=<ltrb -156.9609 371.4595 60 8>}}
   }
  }
  CTRL_R_mouth_lipsTogetherD {
   widgetClassPath=&EUW_Picker_1DSlider
   bIsVariable=true
   parent=FN_MH_Face_CanvasPanel
   widget=CTRL_R_mouth_lipsTogetherD
   properties {bOverride_Cursor=true, renderTransform {angle=-45, scale=<1 1>, shear=<0 0>, translation=<0 0>}}
   slot {
    class=CanvasPanelSlot
    properties {layoutData {alignment=<0.5 0.5>, anchors {maximum=<0.5 0.5>, minimum=<0.5 0.5>}, offsets=<ltrb -156.9609 431.4595 60 8>}}
   }
  }
  CTRL_L_mouth_lipsTogetherD {
   widgetClassPath=&EUW_Picker_1DSlider
   bIsVariable=true
   parent=FN_MH_Face_CanvasPanel
   widget=CTRL_L_mouth_lipsTogetherD
   properties {bOverride_Cursor=true, renderTransform {angle=-135, scale=<1 1>, shear=<0 0>, translation=<0 0>}}
   slot {
    class=CanvasPanelSlot
    properties {layoutData {alignment=<0.5 0.5>, anchors {maximum=<0.5 0.5>, minimum=<0.5 0.5>}, offsets=<ltrb -92.9609 431.4595 60 8>}}
   }
  }
  CTRL_R_mouth_lipBiteD {
   widgetClassPath=&EUW_Picker_1DSlider
   bIsVariable=true
   parent=FN_MH_Face_CanvasPanel
   widget=CTRL_R_mouth_lipBiteD
   properties {bOverride_Cursor=true, renderTransform {angle=-90, scale=<1 1>, shear=<0 0>, translation=<0 0>}}
   slot {
    class=CanvasPanelSlot
    properties {layoutData {alignment=<0.5 0.5>, anchors {maximum=<0.5 0.5>, minimum=<0.5 0.5>}, offsets=<ltrb -800 544 51.1376 11.0544>}}
   }
  }
  CTRL_L_mouth_lipBiteD {
   widgetClassPath=&EUW_Picker_1DSlider
   bIsVariable=true
   parent=FN_MH_Face_CanvasPanel
   widget=CTRL_L_mouth_lipBiteD
   properties {bOverride_Cursor=true, renderTransform {angle=-90, scale=<1 1>, shear=<0 0>, translation=<0 0>}}
   slot {
    class=CanvasPanelSlot
    properties {layoutData {alignment=<0.5 0.5>, anchors {maximum=<0.5 0.5>, minimum=<0.5 0.5>}, offsets=<ltrb -744 544 51.1376 11.0544>}}
   }
  }
  CTRL_R_mouth_lipBiteU {
   widgetClassPath=&EUW_Picker_1DSlider
   bIsVariable=true
   parent=FN_MH_Face_CanvasPanel
   widget=CTRL_R_mouth_lipBiteU
   properties {bOverride_Cursor=true, renderTransform {angle=90, scale=<1 1>, shear=<0 0>, translation=<0 0>}}
   slot {
    class=CanvasPanelSlot
    properties {layoutData {alignment=<0.5 0.5>, anchors {maximum=<0.5 0.5>, minimum=<0.5 0.5>}, offsets=<ltrb -800 492 51.1376 11.0544>}}
   }
  }
  CTRL_L_mouth_lipBiteU {
   widgetClassPath=&EUW_Picker_1DSlider
   bIsVariable=true
   parent=FN_MH_Face_CanvasPanel
   widget=CTRL_L_mouth_lipBiteU
   properties {bOverride_Cursor=true, renderTransform {angle=90, scale=<1 1>, shear=<0 0>, translation=<0 0>}}
   slot {
    class=CanvasPanelSlot
    properties {layoutData {alignment=<0.5 0.5>, anchors {maximum=<0.5 0.5>, minimum=<0.5 0.5>}, offsets=<ltrb -744 492 51.1376 11.0544>}}
   }
  }
  CTRL_L_jaw_ChinRaiseU {
   widgetClassPath=&EUW_Picker_1DSlider
   bIsVariable=true
   parent=FN_MH_Face_CanvasPanel
   widget=CTRL_L_jaw_ChinRaiseU
   properties {bOverride_Cursor=true, renderTransform {angle=-90, scale=<1 1>, shear=<0 0>, translation=<0 0>}}
   slot {
    class=CanvasPanelSlot
    properties {layoutData {alignment=<0.5 0.5>, anchors {maximum=<0.5 0.5>, minimum=<0.5 0.5>}, offsets=<ltrb -325 284.9867 51.1376 11.0544>}}
   }
  }
  CTRL_R_jaw_ChinRaiseU {
   widgetClassPath=&EUW_Picker_1DSlider
   bIsVariable=true
   parent=FN_MH_Face_CanvasPanel
   widget=CTRL_R_jaw_ChinRaiseU
   properties {bOverride_Cursor=true, renderTransform {angle=-90, scale=<1 1>, shear=<0 0>, translation=<0 0>}}
   slot {
    class=CanvasPanelSlot
    properties {layoutData {alignment=<0.5 0.5>, anchors {maximum=<0.5 0.5>, minimum=<0.5 0.5>}, offsets=<ltrb -375 284.1789 51.1376 11.0544>}}
   }
  }
  CTRL_R_jaw_ChinRaiseD {
   widgetClassPath=&EUW_Picker_1DSlider
   bIsVariable=true
   parent=FN_MH_Face_CanvasPanel
   widget=CTRL_R_jaw_ChinRaiseD
   properties {bOverride_Cursor=true, renderTransform {angle=-90, scale=<1 1>, shear=<0 0>, translation=<0 0>}}
   slot {
    class=CanvasPanelSlot
    properties {layoutData {alignment=<0.5 0.5>, anchors {maximum=<0.5 0.5>, minimum=<0.5 0.5>}, offsets=<ltrb -375 345 51.1376 11.0544>}}
   }
  }
  CTRL_L_jaw_ChinRaiseD {
   widgetClassPath=&EUW_Picker_1DSlider
   bIsVariable=true
   parent=FN_MH_Face_CanvasPanel
   widget=CTRL_L_jaw_ChinRaiseD
   properties {bOverride_Cursor=true, renderTransform {angle=-90, scale=<1 1>, shear=<0 0>, translation=<0 0>}}
   slot {
    class=CanvasPanelSlot
    properties {layoutData {alignment=<0.5 0.5>, anchors {maximum=<0.5 0.5>, minimum=<0.5 0.5>}, offsets=<ltrb -325 345 51.1376 11.0544>}}
   }
  }
  CTRL_L_mouth_pressD {
   widgetClassPath=&EUW_Picker_1DSlider
   bIsVariable=true
   parent=FN_MH_Face_CanvasPanel
   widget=CTRL_L_mouth_pressD
   properties {bOverride_Cursor=true, renderTransform {angle=-135, scale=<1 1>, shear=<0 0>, translation=<0 0>}}
   slot {
    class=CanvasPanelSlot
    properties {layoutData {alignment=<0.5 0.5>, anchors {maximum=<0.5 0.5>, minimum=<0.5 0.5>}, offsets=<ltrb 58.6079 432.9867 51.1376 11.0544>}}
   }
  }
  CTRL_R_mouth_pressD {
   widgetClassPath=&EUW_Picker_1DSlider
   bIsVariable=true
   parent=FN_MH_Face_CanvasPanel
   widget=CTRL_R_mouth_pressD
   properties {bOverride_Cursor=true, renderTransform {angle=-45, scale=<1 1>, shear=<0 0>, translation=<0 0>}}
   slot {
    class=CanvasPanelSlot
    properties {layoutData {alignment=<0.5 0.5>, anchors {maximum=<0.5 0.5>, minimum=<0.5 0.5>}, offsets=<ltrb -17.3921 432.9867 51.1376 11.0544>}}
   }
  }
  CTRL_L_mouth_pressU {
   widgetClassPath=&EUW_Picker_1DSlider
   bIsVariable=true
   parent=FN_MH_Face_CanvasPanel
   widget=CTRL_L_mouth_pressU
   properties {bOverride_Cursor=true, renderTransform {angle=135, scale=<1 1>, shear=<0 0>, translation=<0 0>}}
   slot {
    class=CanvasPanelSlot
    properties {layoutData {alignment=<0.5 0.5>, anchors {maximum=<0.5 0.5>, minimum=<0.5 0.5>}, offsets=<ltrb 58.6079 364.9867 51.1376 11.0544>}}
   }
  }
  CTRL_R_mouth_pressU {
   widgetClassPath=&EUW_Picker_1DSlider
   bIsVariable=true
   parent=FN_MH_Face_CanvasPanel
   widget=CTRL_R_mouth_pressU
   properties {bOverride_Cursor=true, renderTransform {angle=45, scale=<1 1>, shear=<0 0>, translation=<0 0>}}
   slot {
    class=CanvasPanelSlot
    properties {layoutData {alignment=<0.5 0.5>, anchors {maximum=<0.5 0.5>, minimum=<0.5 0.5>}, offsets=<ltrb -17.3921 364.9867 51.1376 11.0544>}}
   }
  }
  CTRL_R_mouth_corner {
   widgetClassPath=&EUW_Picker_2DSlider
   bIsVariable=true
   parent=FN_MH_Face_CanvasPanel
   widget=CTRL_R_mouth_corner
   properties {
    bOverride_Cursor=true
    max Value X=-1
    max Value Y=-1
    min Value X=1
    min Value Y=1
    normal Style {
     barThickness=2
     backgroundImage {
      bIsDynamicallyLoaded=false
      drawAs=Image
      imageType=NoImage
      mirroring=NoMirror
      resourceName=None
      resourceObject=null
      tiling=NoTile
      imageSize=<32 32>
      margin=<ltrb 0 0 0 0>
      outlineSettings {
       bUseBrushTransparency=false
       roundingType=HalfHeightRadius
       width=0
       color {colorUseRule=UseColor_Specified, specifiedColor=<rgba 0 0 0 0>}
       cornerRadii=<0 0 0 0>
      }
      tintColor {colorUseRule=UseColor_Specified, specifiedColor=<rgba 0 0 0 0.1826>}
      uVRegion {bIsValid=false, max=<0 0>, min=<0 0>}
     }
     disabledBarImage {
      bIsDynamicallyLoaded=false
      drawAs=Image
      imageType=NoImage
      mirroring=NoMirror
      resourceName=None
      resourceObject=null
      tiling=NoTile
      imageSize=<32 32>
      margin=<ltrb 0 0 0 0>
      outlineSettings {
       bUseBrushTransparency=false
       roundingType=HalfHeightRadius
       width=0
       color {colorUseRule=UseColor_Specified, specifiedColor=<rgba 0 0 0 0>}
       cornerRadii=<0 0 0 0>
      }
      tintColor {colorUseRule=UseColor_Specified, specifiedColor=<rgba 1 1 1 1>}
      uVRegion {bIsValid=false, max=<0 0>, min=<0 0>}
     }
     disabledThumbImage {
      bIsDynamicallyLoaded=false
      drawAs=Image
      imageType=NoImage
      mirroring=NoMirror
      resourceName=None
      resourceObject=null
      tiling=NoTile
      imageSize=<32 32>
      margin=<ltrb 0 0 0 0>
      outlineSettings {
       bUseBrushTransparency=false
       roundingType=HalfHeightRadius
       width=0
       color {colorUseRule=UseColor_Specified, specifiedColor=<rgba 0 0 0 0>}
       cornerRadii=<0 0 0 0>
      }
      tintColor {colorUseRule=UseColor_Specified, specifiedColor=<rgba 1 1 1 1>}
      uVRegion {bIsValid=false, max=<0 0>, min=<0 0>}
     }
     normalBarImage {
      bIsDynamicallyLoaded=false
      drawAs=Image
      imageType=NoImage
      mirroring=NoMirror
      resourceName=None
      resourceObject=null
      tiling=NoTile
      imageSize=<32 32>
      margin=<ltrb 0 0 0 0>
      outlineSettings {
       bUseBrushTransparency=false
       roundingType=HalfHeightRadius
       width=0
       color {colorUseRule=UseColor_Specified, specifiedColor=<rgba 0 0 0 0>}
       cornerRadii=<0 0 0 0>
      }
      tintColor {colorUseRule=UseColor_Specified, specifiedColor=<rgba 1 1 1 1>}
      uVRegion {bIsValid=false, max=<0 0>, min=<0 0>}
     }
     normalThumbImage {
      bIsDynamicallyLoaded=false
      drawAs=Image
      imageType=FullColor
      mirroring=NoMirror
      resourceName=None
      resourceObject=&Circle
      tiling=NoTile
      imageSize=<20 20>
      margin=<ltrb 0 0 0 0>
      outlineSettings {
       bUseBrushTransparency=false
       roundingType=HalfHeightRadius
       width=0
       color {colorUseRule=UseColor_Specified, specifiedColor=<rgba 0 0 0 0>}
       cornerRadii=<0 0 0 0>
      }
      tintColor {colorUseRule=UseColor_Specified, specifiedColor=<rgba 0.2604 0.2604 0.2604 1>}
      uVRegion {bIsValid=false, max=<0 0>, min=<0 0>}
     }
    }
    selected Style {
     barThickness=2
     backgroundImage {
      bIsDynamicallyLoaded=false
      drawAs=Image
      imageType=NoImage
      mirroring=NoMirror
      resourceName=None
      resourceObject=null
      tiling=NoTile
      imageSize=<32 32>
      margin=<ltrb 0 0 0 0>
      outlineSettings {
       bUseBrushTransparency=false
       roundingType=HalfHeightRadius
       width=0
       color {colorUseRule=UseColor_Specified, specifiedColor=<rgba 0 0 0 0>}
       cornerRadii=<0 0 0 0>
      }
      tintColor {colorUseRule=UseColor_Specified, specifiedColor=<rgba 0.1979 0.1979 0.1979 0.1826>}
      uVRegion {bIsValid=false, max=<0 0>, min=<0 0>}
     }
     disabledBarImage {
      bIsDynamicallyLoaded=false
      drawAs=Image
      imageType=NoImage
      mirroring=NoMirror
      resourceName=None
      resourceObject=null
      tiling=NoTile
      imageSize=<32 32>
      margin=<ltrb 0 0 0 0>
      outlineSettings {
       bUseBrushTransparency=false
       roundingType=HalfHeightRadius
       width=0
       color {colorUseRule=UseColor_Specified, specifiedColor=<rgba 0 0 0 0>}
       cornerRadii=<0 0 0 0>
      }
      tintColor {colorUseRule=UseColor_Specified, specifiedColor=<rgba 1 1 1 1>}
      uVRegion {bIsValid=false, max=<0 0>, min=<0 0>}
     }
     disabledThumbImage {
      bIsDynamicallyLoaded=false
      drawAs=Image
      imageType=NoImage
      mirroring=NoMirror
      resourceName=None
      resourceObject=null
      tiling=NoTile
      imageSize=<32 32>
      margin=<ltrb 0 0 0 0>
      outlineSettings {
       bUseBrushTransparency=false
       roundingType=HalfHeightRadius
       width=0
       color {colorUseRule=UseColor_Specified, specifiedColor=<rgba 0 0 0 0>}
       cornerRadii=<0 0 0 0>
      }
      tintColor {colorUseRule=UseColor_Specified, specifiedColor=<rgba 1 1 1 1>}
      uVRegion {bIsValid=false, max=<0 0>, min=<0 0>}
     }
     normalBarImage {
      bIsDynamicallyLoaded=false
      drawAs=Image
      imageType=NoImage
      mirroring=NoMirror
      resourceName=None
      resourceObject=null
      tiling=NoTile
      imageSize=<32 32>
      margin=<ltrb 0 0 0 0>
      outlineSettings {
       bUseBrushTransparency=false
       roundingType=HalfHeightRadius
       width=0
       color {colorUseRule=UseColor_Specified, specifiedColor=<rgba 0 0 0 0>}
       cornerRadii=<0 0 0 0>
      }
      tintColor {colorUseRule=UseColor_Specified, specifiedColor=<rgba 1 1 1 1>}
      uVRegion {bIsValid=false, max=<0 0>, min=<0 0>}
     }
     normalThumbImage {
      bIsDynamicallyLoaded=false
      drawAs=Image
      imageType=FullColor
      mirroring=NoMirror
      resourceName=None
      resourceObject=&Circle
      tiling=NoTile
      imageSize=<25 25>
      margin=<ltrb 0 0 0 0>
      outlineSettings {
       bUseBrushTransparency=false
       roundingType=HalfHeightRadius
       width=0
       color {colorUseRule=UseColor_Specified, specifiedColor=<rgba 0 0 0 0>}
       cornerRadii=<0 0 0 0>
      }
      tintColor {colorUseRule=UseColor_Specified, specifiedColor=<rgba 1 1 1 1>}
      uVRegion {bIsValid=false, max=<0 0>, min=<0 0>}
     }
    }
   }
   slot {
    class=CanvasPanelSlot
    properties {layoutData {alignment=<0.5 0.5>, anchors {maximum=<0.5 0.5>, minimum=<0.5 0.5>}, offsets=<ltrb 648 -20 55 55>}}
   }
  }
  CTRL_L_mouth_corner {
   widgetClassPath=&EUW_Picker_2DSlider
   bIsVariable=true
   parent=FN_MH_Face_CanvasPanel
   widget=CTRL_L_mouth_corner
   properties {
    bOverride_Cursor=true
    max Value Y=-1
    min Value Y=1
    normal Style {
     barThickness=2
     backgroundImage {
      bIsDynamicallyLoaded=false
      drawAs=Image
      imageType=NoImage
      mirroring=NoMirror
      resourceName=None
      resourceObject=null
      tiling=NoTile
      imageSize=<32 32>
      margin=<ltrb 0 0 0 0>
      outlineSettings {
       bUseBrushTransparency=false
       roundingType=HalfHeightRadius
       width=0
       color {colorUseRule=UseColor_Specified, specifiedColor=<rgba 0 0 0 0>}
       cornerRadii=<0 0 0 0>
      }
      tintColor {colorUseRule=UseColor_Specified, specifiedColor=<rgba 0 0 0 0.1826>}
      uVRegion {bIsValid=false, max=<0 0>, min=<0 0>}
     }
     disabledBarImage {
      bIsDynamicallyLoaded=false
      drawAs=Image
      imageType=NoImage
      mirroring=NoMirror
      resourceName=None
      resourceObject=null
      tiling=NoTile
      imageSize=<32 32>
      margin=<ltrb 0 0 0 0>
      outlineSettings {
       bUseBrushTransparency=false
       roundingType=HalfHeightRadius
       width=0
       color {colorUseRule=UseColor_Specified, specifiedColor=<rgba 0 0 0 0>}
       cornerRadii=<0 0 0 0>
      }
      tintColor {colorUseRule=UseColor_Specified, specifiedColor=<rgba 1 1 1 1>}
      uVRegion {bIsValid=false, max=<0 0>, min=<0 0>}
     }
     disabledThumbImage {
      bIsDynamicallyLoaded=false
      drawAs=Image
      imageType=NoImage
      mirroring=NoMirror
      resourceName=None
      resourceObject=null
      tiling=NoTile
      imageSize=<32 32>
      margin=<ltrb 0 0 0 0>
      outlineSettings {
       bUseBrushTransparency=false
       roundingType=HalfHeightRadius
       width=0
       color {colorUseRule=UseColor_Specified, specifiedColor=<rgba 0 0 0 0>}
       cornerRadii=<0 0 0 0>
      }
      tintColor {colorUseRule=UseColor_Specified, specifiedColor=<rgba 1 1 1 1>}
      uVRegion {bIsValid=false, max=<0 0>, min=<0 0>}
     }
     normalBarImage {
      bIsDynamicallyLoaded=false
      drawAs=Image
      imageType=NoImage
      mirroring=NoMirror
      resourceName=None
      resourceObject=null
      tiling=NoTile
      imageSize=<32 32>
      margin=<ltrb 0 0 0 0>
      outlineSettings {
       bUseBrushTransparency=false
       roundingType=HalfHeightRadius
       width=0
       color {colorUseRule=UseColor_Specified, specifiedColor=<rgba 0 0 0 0>}
       cornerRadii=<0 0 0 0>
      }
      tintColor {colorUseRule=UseColor_Specified, specifiedColor=<rgba 1 1 1 1>}
      uVRegion {bIsValid=false, max=<0 0>, min=<0 0>}
     }
     normalThumbImage {
      bIsDynamicallyLoaded=false
      drawAs=Image
      imageType=FullColor
      mirroring=NoMirror
      resourceName=None
      resourceObject=&Circle
      tiling=NoTile
      imageSize=<20 20>
      margin=<ltrb 0 0 0 0>
      outlineSettings {
       bUseBrushTransparency=false
       roundingType=HalfHeightRadius
       width=0
       color {colorUseRule=UseColor_Specified, specifiedColor=<rgba 0 0 0 0>}
       cornerRadii=<0 0 0 0>
      }
      tintColor {colorUseRule=UseColor_Specified, specifiedColor=<rgba 0.2604 0.2604 0.2604 1>}
      uVRegion {bIsValid=false, max=<0 0>, min=<0 0>}
     }
    }
    selected Style {
     barThickness=2
     backgroundImage {
      bIsDynamicallyLoaded=false
      drawAs=Image
      imageType=NoImage
      mirroring=NoMirror
      resourceName=None
      resourceObject=null
      tiling=NoTile
      imageSize=<32 32>
      margin=<ltrb 0 0 0 0>
      outlineSettings {
       bUseBrushTransparency=false
       roundingType=HalfHeightRadius
       width=0
       color {colorUseRule=UseColor_Specified, specifiedColor=<rgba 0 0 0 0>}
       cornerRadii=<0 0 0 0>
      }
      tintColor {colorUseRule=UseColor_Specified, specifiedColor=<rgba 0.1979 0.1979 0.1979 0.1826>}
      uVRegion {bIsValid=false, max=<0 0>, min=<0 0>}
     }
     disabledBarImage {
      bIsDynamicallyLoaded=false
      drawAs=Image
      imageType=NoImage
      mirroring=NoMirror
      resourceName=None
      resourceObject=null
      tiling=NoTile
      imageSize=<32 32>
      margin=<ltrb 0 0 0 0>
      outlineSettings {
       bUseBrushTransparency=false
       roundingType=HalfHeightRadius
       width=0
       color {colorUseRule=UseColor_Specified, specifiedColor=<rgba 0 0 0 0>}
       cornerRadii=<0 0 0 0>
      }
      tintColor {colorUseRule=UseColor_Specified, specifiedColor=<rgba 1 1 1 1>}
      uVRegion {bIsValid=false, max=<0 0>, min=<0 0>}
     }
     disabledThumbImage {
      bIsDynamicallyLoaded=false
      drawAs=Image
      imageType=NoImage
      mirroring=NoMirror
      resourceName=None
      resourceObject=null
      tiling=NoTile
      imageSize=<32 32>
      margin=<ltrb 0 0 0 0>
      outlineSettings {
       bUseBrushTransparency=false
       roundingType=HalfHeightRadius
       width=0
       color {colorUseRule=UseColor_Specified, specifiedColor=<rgba 0 0 0 0>}
       cornerRadii=<0 0 0 0>
      }
      tintColor {colorUseRule=UseColor_Specified, specifiedColor=<rgba 1 1 1 1>}
      uVRegion {bIsValid=false, max=<0 0>, min=<0 0>}
     }
     normalBarImage {
      bIsDynamicallyLoaded=false
      drawAs=Image
      imageType=NoImage
      mirroring=NoMirror
      resourceName=None
      resourceObject=null
      tiling=NoTile
      imageSize=<32 32>
      margin=<ltrb 0 0 0 0>
      outlineSettings {
       bUseBrushTransparency=false
       roundingType=HalfHeightRadius
       width=0
       color {colorUseRule=UseColor_Specified, specifiedColor=<rgba 0 0 0 0>}
       cornerRadii=<0 0 0 0>
      }
      tintColor {colorUseRule=UseColor_Specified, specifiedColor=<rgba 1 1 1 1>}
      uVRegion {bIsValid=false, max=<0 0>, min=<0 0>}
     }
     normalThumbImage {
      bIsDynamicallyLoaded=false
      drawAs=Image
      imageType=FullColor
      mirroring=NoMirror
      resourceName=None
      resourceObject=&Circle
      tiling=NoTile
      imageSize=<25 25>
      margin=<ltrb 0 0 0 0>
      outlineSettings {
       bUseBrushTransparency=false
       roundingType=HalfHeightRadius
       width=0
       color {colorUseRule=UseColor_Specified, specifiedColor=<rgba 0 0 0 0>}
       cornerRadii=<0 0 0 0>
      }
      tintColor {colorUseRule=UseColor_Specified, specifiedColor=<rgba 1 1 1 1>}
      uVRegion {bIsValid=false, max=<0 0>, min=<0 0>}
     }
    }
   }
   slot {
    class=CanvasPanelSlot
    properties {layoutData {alignment=<0.5 0.5>, anchors {maximum=<0.5 0.5>, minimum=<0.5 0.5>}, offsets=<ltrb 724 -20 55 55>}}
   }
  }
  CTRL_R_mouth_lipsTowardsTeethD {
   widgetClassPath=&EUW_Picker_1DSlider
   bIsVariable=true
   parent=FN_MH_Face_CanvasPanel
   widget=CTRL_R_mouth_lipsTowardsTeethD
   properties {bOverride_Cursor=true, renderTransform {angle=-70, scale=<1 1>, shear=<0 0>, translation=<0 0>}}
   slot {
    class=CanvasPanelSlot
    properties {layoutData {alignment=<0.5 0.5>, anchors {maximum=<0.5 0.5>, minimum=<0.5 0.5>}, offsets=<ltrb 240 28.9867 51.1376 11.0544>}}
   }
  }
  CTRL_L_mouth_lipsTowardsTeethD {
   widgetClassPath=&EUW_Picker_1DSlider
   bIsVariable=true
   parent=FN_MH_Face_CanvasPanel
   widget=CTRL_L_mouth_lipsTowardsTeethD
   properties {bOverride_Cursor=true, renderTransform {angle=-110, scale=<1 1>, shear=<0 0>, translation=<0 0>}}
   slot {
    class=CanvasPanelSlot
    properties {layoutData {alignment=<0.5 0.5>, anchors {maximum=<0.5 0.5>, minimum=<0.5 0.5>}, offsets=<ltrb 280.6079 28 51.1376 11.0544>}}
   }
  }
  CTRL_L_mouth_lipsTowardsTeethU {
   widgetClassPath=&EUW_Picker_1DSlider
   bIsVariable=true
   parent=FN_MH_Face_CanvasPanel
   widget=CTRL_L_mouth_lipsTowardsTeethU
   properties {bOverride_Cursor=true, renderTransform {angle=110, scale=<1 1>, shear=<0 0>, translation=<0 0>}}
   slot {
    class=CanvasPanelSlot
    properties {layoutData {alignment=<0.5 0.5>, anchors {maximum=<0.5 0.5>, minimum=<0.5 0.5>}, offsets=<ltrb 280.6079 -27.0133 51.1376 11.0544>}}
   }
  }
  CTRL_R_mouth_lipsTowardsTeethU {
   widgetClassPath=&EUW_Picker_1DSlider
   bIsVariable=true
   parent=FN_MH_Face_CanvasPanel
   widget=CTRL_R_mouth_lipsTowardsTeethU
   properties {bOverride_Cursor=true, renderTransform {angle=70, scale=<1 1>, shear=<0 0>, translation=<0 0>}}
   slot {
    class=CanvasPanelSlot
    properties {layoutData {alignment=<0.5 0.5>, anchors {maximum=<0.5 0.5>, minimum=<0.5 0.5>}, offsets=<ltrb 240.6079 -27.0133 51.1376 11.0544>}}
   }
  }
  Graphics {
   widgetClassPath=&CanvasPanel
   parent=FN_MH_Face_CanvasPanel
   widget=Graphics
   slot {class=CanvasPanelSlot, properties {layoutData {alignment=<0.5 0.5>, anchors {maximum=<1 1>, minimum=<0 0>}, offsets=<ltrb 0 0 0 0>}}}
  }
  Image_42 {
   widgetClassPath=&Image
   bIsVariable=true
   parent=Graphics
   widget=Image_42
   properties {
    brush {
     bIsDynamicallyLoaded=false
     drawAs=Image
     imageType=FullColor
     mirroring=NoMirror
     resourceName=None
     resourceObject=/Game/Characters/Player/Common/SharedRigs/Graphics/JONESY_HEAD
     tiling=NoTile
     imageSize=<666 847>
     margin=<ltrb 0 0 0 0>
     outlineSettings {
      bUseBrushTransparency=false
      roundingType=HalfHeightRadius
      width=0
      color {colorUseRule=UseColor_Specified, specifiedColor=<rgba 0 0 0 0>}
      cornerRadii=<0 0 0 0>
     }
     tintColor {colorUseRule=UseColor_Specified, specifiedColor=<rgba 1 1 1 1>}
     uVRegion {bIsValid=false, max=<0 0>, min=<0 0>}
    }
    colorAndOpacity=<rgba 1 1 1 0.4>
   }
   slot {
    class=CanvasPanelSlot
    properties {layoutData {alignment=<0.5 0.5>, anchors {maximum=<0.5 0.5>, minimum=<0.5 0.5>}, offsets=<ltrb -365.6156 -52.8803 681.4449 839.1203>}}
   }
  }
  Image_13 {
   widgetClassPath=&Image
   bIsVariable=true
   parent=Graphics
   widget=Image_13
   properties {
    brush {
     bIsDynamicallyLoaded=false
     drawAs=Image
     imageType=FullColor
     mirroring=NoMirror
     resourceName=None
     resourceObject=&JONESY_HEAD
     tiling=NoTile
     imageSize=<666 847>
     margin=<ltrb 0 0 0 0>
     outlineSettings {
      bUseBrushTransparency=false
      roundingType=HalfHeightRadius
      width=0
      color {colorUseRule=UseColor_Specified, specifiedColor=<rgba 0 0 0 0>}
      cornerRadii=<0 0 0 0>
     }
     tintColor {colorUseRule=UseColor_Specified, specifiedColor=<rgba 1 1 1 1>}
     uVRegion {bIsValid=false, max=<0 0>, min=<0 0>}
    }
    colorAndOpacity=<rgba 1 1 1 0.4>
   }
   slot {
    class=CanvasPanelSlot
    properties {layoutData {alignment=<0.5 0.5>, anchors {maximum=<0.5 0.5>, minimum=<0.5 0.5>}, offsets=<ltrb 525.7615 -52.8804 681.4449 839.1203>}}
   }
  }
  Image {
   widgetClassPath=&Image
   bIsVariable=true
   parent=Graphics
   widget=Image
   properties {
    brush {
     bIsDynamicallyLoaded=false
     drawAs=Image
     imageType=FullColor
     mirroring=NoMirror
     resourceName=None
     resourceObject=/Game/Characters/Player/Common/SharedRigs/Graphics/_3L_towards
     tiling=NoTile
     imageSize=<75 75>
     margin=<ltrb 0 0 0 0>
     outlineSettings {
      bUseBrushTransparency=false
      roundingType=HalfHeightRadius
      width=0
      color {colorUseRule=UseColor_Specified, specifiedColor=<rgba 0 0 0 0>}
      cornerRadii=<0 0 0 0>
     }
     tintColor {colorUseRule=UseColor_Specified, specifiedColor=<rgba 1 1 1 1>}
     uVRegion {bIsValid=false, max=<0 0>, min=<0 0>}
    }
    colorAndOpacity=<rgba 1 1 1 0.4>
   }
   slot {
    class=CanvasPanelSlot
    properties {layoutData {alignment=<0.5 0.5>, anchors {maximum=<0.5 0.5>, minimum=<0.5 0.5>}, offsets=<ltrb -769.4609 340 75 75>}}
   }
  }
  Image_6 {
   widgetClassPath=&Image
   bIsVariable=true
   parent=Graphics
   widget=Image_6
   properties {
    brush {
     bIsDynamicallyLoaded=false
     drawAs=Image
     imageType=FullColor
     mirroring=NoMirror
     resourceName=None
     resourceObject=/Game/Characters/Player/Common/SharedRigs/Graphics/_3L_bite
     tiling=NoTile
     imageSize=<134 54>
     margin=<ltrb 0 0 0 0>
     outlineSettings {
      bUseBrushTransparency=false
      roundingType=HalfHeightRadius
      width=0
      color {colorUseRule=UseColor_Specified, specifiedColor=<rgba 0 0 0 0>}
      cornerRadii=<0 0 0 0>
     }
     tintColor {colorUseRule=UseColor_Specified, specifiedColor=<rgba 1 1 1 1>}
     uVRegion {bIsValid=false, max=<0 0>, min=<0 0>}
    }
    colorAndOpacity=<rgba 1 1 1 0.4>
   }
   slot {
    class=CanvasPanelSlot
    properties {layoutData {alignment=<0.5 0.5>, anchors {maximum=<0.5 0.5>, minimum=<0.5 0.5>}, offsets=<ltrb -772.3003 518.0601 75 75>}}
   }
  }
  Image_3 {
   widgetClassPath=&Image
   bIsVariable=true
   parent=Graphics
   widget=Image_3
   properties {
    brush {
     bIsDynamicallyLoaded=false
     drawAs=Image
     imageType=FullColor
     mirroring=NoMirror
     resourceName=None
     resourceObject=&_3L_bite
     tiling=NoTile
     imageSize=<134 54>
     margin=<ltrb 0 0 0 0>
     outlineSettings {
      bUseBrushTransparency=false
      roundingType=HalfHeightRadius
      width=0
      color {colorUseRule=UseColor_Specified, specifiedColor=<rgba 0 0 0 0>}
      cornerRadii=<0 0 0 0>
     }
     tintColor {colorUseRule=UseColor_Specified, specifiedColor=<rgba 1 1 1 1>}
     uVRegion {bIsValid=false, max=<0 0>, min=<0 0>}
    }
    colorAndOpacity=<rgba 1 1 1 0.4>
   }
   slot {
    class=CanvasPanelSlot
    properties {layoutData {alignment=<0.5 0.5>, anchors {maximum=<0.5 0.5>, minimum=<0.5 0.5>}, offsets=<ltrb -664 520.9595 75 75>}}
   }
  }
  Image_4 {
   widgetClassPath=&Image
   bIsVariable=true
   parent=Graphics
   widget=Image_4
   properties {
    brush {
     bIsDynamicallyLoaded=false
     drawAs=Image
     imageType=FullColor
     mirroring=NoMirror
     resourceName=None
     resourceObject=&_3L_bite
     tiling=NoTile
     imageSize=<134 54>
     margin=<ltrb 0 0 0 0>
     outlineSettings {
      bUseBrushTransparency=false
      roundingType=HalfHeightRadius
      width=0
      color {colorUseRule=UseColor_Specified, specifiedColor=<rgba 0 0 0 0>}
      cornerRadii=<0 0 0 0>
     }
     tintColor {colorUseRule=UseColor_Specified, specifiedColor=<rgba 1 1 1 1>}
     uVRegion {bIsValid=false, max=<0 0>, min=<0 0>}
    }
    colorAndOpacity=<rgba 1 1 1 0.4>
   }
   slot {
    class=CanvasPanelSlot
    properties {layoutData {alignment=<0.5 0.5>, anchors {maximum=<0.5 0.5>, minimum=<0.5 0.5>}, offsets=<ltrb -552 520 75 75>}}
   }
  }
  Image_8 {
   widgetClassPath=&Image
   bIsVariable=true
   parent=Graphics
   widget=Image_8
   properties {
    brush {
     bIsDynamicallyLoaded=false
     drawAs=Image
     imageType=FullColor
     mirroring=NoMirror
     resourceName=None
     resourceObject=/Game/Characters/Player/Common/SharedRigs/Graphics/_3L_press
     tiling=NoTile
     imageSize=<146 54>
     margin=<ltrb 0 0 0 0>
     outlineSettings {
      bUseBrushTransparency=false
      roundingType=HalfHeightRadius
      width=0
      color {colorUseRule=UseColor_Specified, specifiedColor=<rgba 0 0 0 0>}
      cornerRadii=<0 0 0 0>
     }
     tintColor {colorUseRule=UseColor_Specified, specifiedColor=<rgba 1 1 1 1>}
     uVRegion {bIsValid=false, max=<0 0>, min=<0 0>}
    }
    colorAndOpacity=<rgba 1 1 1 0.4>
   }
   slot {
    class=CanvasPanelSlot
    properties {layoutData {alignment=<0.5 0.5>, anchors {maximum=<0.5 0.5>, minimum=<0.5 0.5>}, offsets=<ltrb 22.0391 400.4595 100 100>}}
   }
  }
  Image_1 {
   widgetClassPath=&Image
   bIsVariable=true
   parent=Graphics
   widget=Image_1
   properties {
    brush {
     bIsDynamicallyLoaded=false
     drawAs=Image
     imageType=FullColor
     mirroring=NoMirror
     resourceName=None
     resourceObject=/Game/Characters/Player/Common/SharedRigs/Graphics/_3L_purse
     tiling=NoTile
     imageSize=<75 75>
     margin=<ltrb 0 0 0 0>
     outlineSettings {
      bUseBrushTransparency=false
      roundingType=HalfHeightRadius
      width=0
      color {colorUseRule=UseColor_Specified, specifiedColor=<rgba 0 0 0 0>}
      cornerRadii=<0 0 0 0>
     }
     tintColor {colorUseRule=UseColor_Specified, specifiedColor=<rgba 1 1 1 1>}
     uVRegion {bIsValid=false, max=<0 0>, min=<0 0>}
    }
    colorAndOpacity=<rgba 1 1 1 0.4>
   }
   slot {
    class=CanvasPanelSlot
    properties {layoutData {alignment=<0.5 0.5>, anchors {maximum=<0.5 0.5>, minimum=<0.5 0.5>}, offsets=<ltrb -665.4609 340 75 75>}}
   }
  }
  Image_2 {
   widgetClassPath=&Image
   bIsVariable=true
   parent=Graphics
   widget=Image_2
   properties {
    brush {
     bIsDynamicallyLoaded=false
     drawAs=Image
     imageType=FullColor
     mirroring=NoMirror
     resourceName=None
     resourceObject=/Game/Characters/Player/Common/SharedRigs/Graphics/_3L_funnel
     tiling=NoTile
     imageSize=<75 75>
     margin=<ltrb 0 0 0 0>
     outlineSettings {
      bUseBrushTransparency=false
      roundingType=HalfHeightRadius
      width=0
      color {colorUseRule=UseColor_Specified, specifiedColor=<rgba 0 0 0 0>}
      cornerRadii=<0 0 0 0>
     }
     tintColor {colorUseRule=UseColor_Specified, specifiedColor=<rgba 1 1 1 1>}
     uVRegion {bIsValid=false, max=<0 0>, min=<0 0>}
    }
    colorAndOpacity=<rgba 1 1 1 0.4>
   }
   slot {
    class=CanvasPanelSlot
    properties {layoutData {alignment=<0.5 0.5>, anchors {maximum=<0.5 0.5>, minimum=<0.5 0.5>}, offsets=<ltrb -549.4609 340 75 75>}}
   }
  }
  Image_5 {
   widgetClassPath=&Image
   bIsVariable=true
   parent=Graphics
   widget=Image_5
   properties {
    brush {
     bIsDynamicallyLoaded=false
     drawAs=Image
     imageType=FullColor
     mirroring=NoMirror
     resourceName=None
     resourceObject=/Game/Characters/Player/Common/SharedRigs/Graphics/_3L_together
     tiling=NoTile
     imageSize=<99 92>
     margin=<ltrb 0 0 0 0>
     outlineSettings {
      bUseBrushTransparency=false
      roundingType=HalfHeightRadius
      width=0
      color {colorUseRule=UseColor_Specified, specifiedColor=<rgba 0 0 0 0>}
      cornerRadii=<0 0 0 0>
     }
     tintColor {colorUseRule=UseColor_Specified, specifiedColor=<rgba 1 1 1 1>}
     uVRegion {bIsValid=false, max=<0 0>, min=<0 0>}
    }
    colorAndOpacity=<rgba 1 1 1 0.4>
   }
   slot {
    class=CanvasPanelSlot
    properties {layoutData {alignment=<0.5 0.5>, anchors {maximum=<0.5 0.5>, minimum=<0.5 0.5>}, offsets=<ltrb -124.9609 401.4595 100 100>}}
   }
  }
  Image_11 {
   widgetClassPath=&Image
   bIsVariable=true
   parent=Graphics
   widget=Image_11
   properties {
    brush {
     bIsDynamicallyLoaded=false
     drawAs=Image
     imageType=FullColor
     mirroring=NoMirror
     resourceName=None
     resourceObject=/Game/Characters/Player/Common/SharedRigs/Graphics/MI_BorderWTCH
     tiling=NoTile
     imageSize=<99 92>
     margin=<ltrb 0 0 0 0>
     outlineSettings {
      bUseBrushTransparency=false
      roundingType=HalfHeightRadius
      width=0
      color {colorUseRule=UseColor_Specified, specifiedColor=<rgba 0 0 0 0>}
      cornerRadii=<0 0 0 0>
     }
     tintColor {colorUseRule=UseColor_Specified, specifiedColor=<rgba 1 1 1 1>}
     uVRegion {bIsValid=false, max=<0 0>, min=<0 0>}
    }
    colorAndOpacity=<rgba 1 1 1 0.4>
   }
   slot {
    class=CanvasPanelSlot
    properties {layoutData {alignment=<0.5 0.5>, anchors {maximum=<0.5 0.5>, minimum=<0.5 0.5>}, offsets=<ltrb -52.9479 571.1922 164.026 103.4655>}}
   }
  }
  Image_12 {
   widgetClassPath=&Image
   bIsVariable=true
   parent=Graphics
   widget=Image_12
   properties {
    brush {
     bIsDynamicallyLoaded=false
     drawAs=Image
     imageType=FullColor
     mirroring=NoMirror
     resourceName=None
     resourceObject=&MI_BorderWTCH
     tiling=NoTile
     imageSize=<99 92>
     margin=<ltrb 0 0 0 0>
     outlineSettings {
      bUseBrushTransparency=false
      roundingType=HalfHeightRadius
      width=0
      color {colorUseRule=UseColor_Specified, specifiedColor=<rgba 0 0 0 0>}
      cornerRadii=<0 0 0 0>
     }
     tintColor {colorUseRule=UseColor_Specified, specifiedColor=<rgba 1 1 1 1>}
     uVRegion {bIsValid=false, max=<0 0>, min=<0 0>}
    }
    colorAndOpacity=<rgba 1 1 1 0.4>
   }
   slot {
    class=CanvasPanelSlot
    properties {layoutData {alignment=<0.5 0.5>, anchors {maximum=<0.5 0.5>, minimum=<0.5 0.5>}, offsets=<ltrb 460.1415 456.1892 270.205 97.4595>}}
   }
  }
  TextBlock_1 {
   widgetClassPath=&TextBlock
   parent=Graphics
   widget=TextBlock_1
   properties {
    text=tongue
    colorAndOpacity {colorUseRule=UseColor_Specified, specifiedColor=<rgba 1 1 1 0.5>}
    font {
     bForceMonospaced=false
     bMaterialIsStencil=false
     fontMaterial=null
     fontObject=&Roboto
     letterSpacing=0
     monospacedWidth=1
     size=11.25
     skewAmount=0
     typefaceFontName=Bold
     outlineSettings {
      bApplyOutlineToDropShadows=false
      bMiteredCorners=false
      bSeparateFillAlpha=false
      outlineMaterial=null
      outlineSize=0
      outlineColor=<rgba 0 0 0 1>
     }
    }
   }
   slot {
    class=CanvasPanelSlot
    properties {bAutoSize=true, layoutData {alignment=<0 0>, anchors {maximum=<0.5 0.5>, minimum=<0.5 0.5>}, offsets=<ltrb -126.9609 75.4595 100 30>}}
   }
  }
  TextBlock {
   widgetClassPath=&TextBlock
   parent=Graphics
   widget=TextBlock
   properties {
    text=teeth
    colorAndOpacity {colorUseRule=UseColor_Specified, specifiedColor=<rgba 1 1 1 0.5>}
    font {
     bForceMonospaced=false
     bMaterialIsStencil=false
     fontMaterial=null
     fontObject=&Roboto
     letterSpacing=0
     monospacedWidth=1
     size=11.25
     skewAmount=0
     typefaceFontName=Bold
     outlineSettings {
      bApplyOutlineToDropShadows=false
      bMiteredCorners=false
      bSeparateFillAlpha=false
      outlineMaterial=null
      outlineSize=0
      outlineColor=<rgba 0 0 0 1>
     }
    }
   }
   slot {
    class=CanvasPanelSlot
    properties {bAutoSize=true, layoutData {alignment=<0 0>, anchors {maximum=<0.5 0.5>, minimum=<0.5 0.5>}, offsets=<ltrb 347.0391 -76 100 30>}}
   }
  }
  TextBlock_20 {
   widgetClassPath=&TextBlock
   parent=Graphics
   widget=TextBlock_20
   properties {
    text=thickness
    colorAndOpacity {colorUseRule=UseColor_Specified, specifiedColor=<rgba 1 1 1 0.5>}
    font {
     bForceMonospaced=false
     bMaterialIsStencil=false
     fontMaterial=null
     fontObject=&Roboto
     letterSpacing=0
     monospacedWidth=1
     size=11.25
     skewAmount=0
     typefaceFontName=Bold
     outlineSettings {
      bApplyOutlineToDropShadows=false
      bMiteredCorners=false
      bSeparateFillAlpha=false
      outlineMaterial=null
      outlineSize=0
      outlineColor=<rgba 0 0 0 1>
     }
    }
   }
   slot {
    class=CanvasPanelSlot
    properties {bAutoSize=true, layoutData {alignment=<0 0>, anchors {maximum=<0.5 0.5>, minimum=<0.5 0.5>}, offsets=<ltrb 360 95 100 30>}}
   }
  }
  TextBlock_21 {
   widgetClassPath=&TextBlock
   parent=Graphics
   widget=TextBlock_21
   properties {
    text=roll
    colorAndOpacity {colorUseRule=UseColor_Specified, specifiedColor=<rgba 1 1 1 0.5>}
    font {
     bForceMonospaced=false
     bMaterialIsStencil=false
     fontMaterial=null
     fontObject=&Roboto
     letterSpacing=0
     monospacedWidth=1
     size=11.25
     skewAmount=0
     typefaceFontName=Bold
     outlineSettings {
      bApplyOutlineToDropShadows=false
      bMiteredCorners=false
      bSeparateFillAlpha=false
      outlineMaterial=null
      outlineSize=0
      outlineColor=<rgba 0 0 0 1>
     }
    }
   }
   slot {
    class=CanvasPanelSlot
    properties {bAutoSize=true, layoutData {alignment=<0 0>, anchors {maximum=<0.5 0.5>, minimum=<0.5 0.5>}, offsets=<ltrb 476 95 100 30>}}
   }
  }
  TextBlock_23 {
   widgetClassPath=&TextBlock
   parent=Graphics
   widget=TextBlock_23
   properties {
    text=lips push-pull
    colorAndOpacity {colorUseRule=UseColor_Specified, specifiedColor=<rgba 1 1 1 0.5>}
    font {
     bForceMonospaced=false
     bMaterialIsStencil=false
     fontMaterial=null
     fontObject=&Roboto
     letterSpacing=0
     monospacedWidth=1
     size=11.25
     skewAmount=0
     typefaceFontName=Bold
     outlineSettings {
      bApplyOutlineToDropShadows=false
      bMiteredCorners=false
      bSeparateFillAlpha=false
      outlineMaterial=null
      outlineSize=0
      outlineColor=<rgba 0 0 0 1>
     }
    }
   }
   slot {
    class=CanvasPanelSlot
    properties {bAutoSize=true, layoutData {alignment=<0 0>, anchors {maximum=<0.5 0.5>, minimum=<0.5 0.5>}, offsets=<ltrb 648 95 100 30>}}
   }
  }
  TextBlock_22 {
   widgetClassPath=&TextBlock
   parent=Graphics
   widget=TextBlock_22
   properties {
    text=sharpness
    colorAndOpacity {colorUseRule=UseColor_Specified, specifiedColor=<rgba 1 1 1 0.5>}
    font {
     bForceMonospaced=false
     bMaterialIsStencil=false
     fontMaterial=null
     fontObject=&Roboto
     letterSpacing=0
     monospacedWidth=1
     size=11.25
     skewAmount=0
     typefaceFontName=Bold
     outlineSettings {
      bApplyOutlineToDropShadows=false
      bMiteredCorners=false
      bSeparateFillAlpha=false
      outlineMaterial=null
      outlineSize=0
      outlineColor=<rgba 0 0 0 1>
     }
    }
   }
   slot {
    class=CanvasPanelSlot
    properties {bAutoSize=true, layoutData {alignment=<0 0>, anchors {maximum=<0.5 0.5>, minimum=<0.5 0.5>}, offsets=<ltrb 552 95 100 30>}}
   }
  }
  TextBlock_9 {
   widgetClassPath=&TextBlock
   parent=Graphics
   widget=TextBlock_9
   properties {
    text=eyelids
    colorAndOpacity {colorUseRule=UseColor_Specified, specifiedColor=<rgba 1 1 1 0.5>}
    font {
     bForceMonospaced=false
     bMaterialIsStencil=false
     fontMaterial=null
     fontObject=&Roboto
     letterSpacing=0
     monospacedWidth=1
     size=11.25
     skewAmount=0
     typefaceFontName=Bold
     outlineSettings {
      bApplyOutlineToDropShadows=false
      bMiteredCorners=false
      bSeparateFillAlpha=false
      outlineMaterial=null
      outlineSize=0
      outlineColor=<rgba 0 0 0 1>
     }
    }
   }
   slot {
    class=CanvasPanelSlot
    properties {bAutoSize=true, layoutData {alignment=<0 0>, anchors {maximum=<0.5 0.5>, minimum=<0.5 0.5>}, offsets=<ltrb 732 -180 100 30>}}
   }
  }
  TextBlock_19 {
   widgetClassPath=&TextBlock
   parent=Graphics
   widget=TextBlock_19
   properties {
    text=eyelids
    colorAndOpacity {colorUseRule=UseColor_Specified, specifiedColor=<rgba 1 1 1 0.5>}
    font {
     bForceMonospaced=false
     bMaterialIsStencil=false
     fontMaterial=null
     fontObject=&Roboto
     letterSpacing=0
     monospacedWidth=1
     size=11.25
     skewAmount=0
     typefaceFontName=Bold
     outlineSettings {
      bApplyOutlineToDropShadows=false
      bMiteredCorners=false
      bSeparateFillAlpha=false
      outlineMaterial=null
      outlineSize=0
      outlineColor=<rgba 0 0 0 1>
     }
    }
   }
   slot {
    class=CanvasPanelSlot
    properties {bAutoSize=true, layoutData {alignment=<0 0>, anchors {maximum=<0.5 0.5>, minimum=<0.5 0.5>}, offsets=<ltrb 300 -180.5405 100 30>}}
   }
  }
  TextBlock_5 {
   widgetClassPath=&TextBlock
   parent=Graphics
   widget=TextBlock_5
   properties {
    text=towards
    colorAndOpacity {colorUseRule=UseColor_Specified, specifiedColor=<rgba 1 1 1 0.5>}
    font {
     bForceMonospaced=false
     bMaterialIsStencil=false
     fontMaterial=null
     fontObject=&Roboto
     letterSpacing=0
     monospacedWidth=1
     size=11.25
     skewAmount=0
     typefaceFontName=Bold
     outlineSettings {
      bApplyOutlineToDropShadows=false
      bMiteredCorners=false
      bSeparateFillAlpha=false
      outlineMaterial=null
      outlineSize=0
      outlineColor=<rgba 0 0 0 1>
     }
    }
   }
   slot {
    class=CanvasPanelSlot
    properties {bAutoSize=true, layoutData {alignment=<0 0>, anchors {maximum=<0.5 0.5>, minimum=<0.5 0.5>}, offsets=<ltrb -810.9609 263.4595 100 30>}}
   }
  }
  TextBlock_6 {
   widgetClassPath=&TextBlock
   parent=Graphics
   widget=TextBlock_6
   properties {
    text=purse
    colorAndOpacity {colorUseRule=UseColor_Specified, specifiedColor=<rgba 1 1 1 0.5>}
    font {
     bForceMonospaced=false
     bMaterialIsStencil=false
     fontMaterial=null
     fontObject=&Roboto
     letterSpacing=0
     monospacedWidth=1
     size=11.25
     skewAmount=0
     typefaceFontName=Bold
     outlineSettings {
      bApplyOutlineToDropShadows=false
      bMiteredCorners=false
      bSeparateFillAlpha=false
      outlineMaterial=null
      outlineSize=0
      outlineColor=<rgba 0 0 0 1>
     }
    }
   }
   slot {
    class=CanvasPanelSlot
    properties {bAutoSize=true, layoutData {alignment=<0 0>, anchors {maximum=<0.5 0.5>, minimum=<0.5 0.5>}, offsets=<ltrb -702.9609 263.4595 100 30>}}
   }
  }
  TextBlock_7 {
   widgetClassPath=&TextBlock
   parent=Graphics
   widget=TextBlock_7
   properties {
    text=funnel
    colorAndOpacity {colorUseRule=UseColor_Specified, specifiedColor=<rgba 1 1 1 0.5>}
    font {
     bForceMonospaced=false
     bMaterialIsStencil=false
     fontMaterial=null
     fontObject=&Roboto
     letterSpacing=0
     monospacedWidth=1
     size=11.25
     skewAmount=0
     typefaceFontName=Bold
     outlineSettings {
      bApplyOutlineToDropShadows=false
      bMiteredCorners=false
      bSeparateFillAlpha=false
      outlineMaterial=null
      outlineSize=0
      outlineColor=<rgba 0 0 0 1>
     }
    }
   }
   slot {
    class=CanvasPanelSlot
    properties {bAutoSize=true, layoutData {alignment=<0.5 0.5>, anchors {maximum=<0.5 0.5>, minimum=<0.5 0.5>}, offsets=<ltrb -565.7736 276.4685 100 30>}}
   }
  }
  TextBlock_8 {
   widgetClassPath=&TextBlock
   parent=Graphics
   widget=TextBlock_8
   properties {
    text=lip blow
    colorAndOpacity {colorUseRule=UseColor_Specified, specifiedColor=<rgba 1 1 1 0.5>}
    font {
     bForceMonospaced=false
     bMaterialIsStencil=false
     fontMaterial=null
     fontObject=&Roboto
     letterSpacing=0
     monospacedWidth=1
     size=11.25
     skewAmount=0
     typefaceFontName=Bold
     outlineSettings {
      bApplyOutlineToDropShadows=false
      bMiteredCorners=false
      bSeparateFillAlpha=false
      outlineMaterial=null
      outlineSize=0
      outlineColor=<rgba 0 0 0 1>
     }
    }
   }
   slot {
    class=CanvasPanelSlot
    properties {bAutoSize=true, layoutData {alignment=<0.5 0.5>, anchors {maximum=<0.5 0.5>, minimum=<0.5 0.5>}, offsets=<ltrb -105.4359 220.4685 100 30>}}
   }
  }
  TextBlock_2 {
   widgetClassPath=&TextBlock
   parent=Graphics
   widget=TextBlock_2
   properties {
    text=TWEAKERS
    colorAndOpacity {colorUseRule=UseColor_Specified, specifiedColor=<rgba 1 1 1 0.5>}
    font {
     bForceMonospaced=false
     bMaterialIsStencil=false
     fontMaterial=null
     fontObject=&Roboto
     letterSpacing=0
     monospacedWidth=1
     size=18.75
     skewAmount=0
     typefaceFontName=Bold
     outlineSettings {
      bApplyOutlineToDropShadows=false
      bMiteredCorners=false
      bSeparateFillAlpha=false
      outlineMaterial=null
      outlineSize=0
      outlineColor=<rgba 0 0 0 1>
     }
    }
   }
   slot {
    class=CanvasPanelSlot
    properties {bAutoSize=true, layoutData {alignment=<0.5 0.5>, anchors {maximum=<0.5 0.5>, minimum=<0.5 0.5>}, offsets=<ltrb 548.6021 302.4745 100 30>}}
   }
  }
  TextBlock_10 {
   widgetClassPath=&TextBlock
   parent=Graphics
   widget=TextBlock_10
   properties {
    text=lips together
    colorAndOpacity {colorUseRule=UseColor_Specified, specifiedColor=<rgba 1 1 1 0.5>}
    font {
     bForceMonospaced=false
     bMaterialIsStencil=false
     fontMaterial=null
     fontObject=&Roboto
     letterSpacing=0
     monospacedWidth=1
     size=11.25
     skewAmount=0
     typefaceFontName=Bold
     outlineSettings {
      bApplyOutlineToDropShadows=false
      bMiteredCorners=false
      bSeparateFillAlpha=false
      outlineMaterial=null
      outlineSize=0
      outlineColor=<rgba 0 0 0 1>
     }
    }
   }
   slot {
    class=CanvasPanelSlot
    properties {bAutoSize=true, layoutData {alignment=<0.5 0.5>, anchors {maximum=<0.5 0.5>, minimum=<0.5 0.5>}, offsets=<ltrb -133.1696 316.4685 100 30>}}
   }
  }
  TextBlock_11 {
   widgetClassPath=&TextBlock
   parent=Graphics
   widget=TextBlock_11
   properties {
    text=lips bite
    colorAndOpacity {colorUseRule=UseColor_Specified, specifiedColor=<rgba 1 1 1 0.5>}
    font {
     bForceMonospaced=false
     bMaterialIsStencil=false
     fontMaterial=null
     fontObject=&Roboto
     letterSpacing=0
     monospacedWidth=1
     size=11.25
     skewAmount=0
     typefaceFontName=Bold
     outlineSettings {
      bApplyOutlineToDropShadows=false
      bMiteredCorners=false
      bSeparateFillAlpha=false
      outlineMaterial=null
      outlineSize=0
      outlineColor=<rgba 0 0 0 1>
     }
    }
   }
   slot {
    class=CanvasPanelSlot
    properties {bAutoSize=true, layoutData {alignment=<0 0>, anchors {maximum=<0.5 0.5>, minimum=<0.5 0.5>}, offsets=<ltrb -806.9609 431.4595 100 30>}}
   }
  }
  TextBlock_3 {
   widgetClassPath=&TextBlock
   parent=Graphics
   widget=TextBlock_3
   properties {
    text=lips press
    colorAndOpacity {colorUseRule=UseColor_Specified, specifiedColor=<rgba 1 1 1 0.5>}
    font {
     bForceMonospaced=false
     bMaterialIsStencil=false
     fontMaterial=null
     fontObject=&Roboto
     letterSpacing=0
     monospacedWidth=1
     size=11.25
     skewAmount=0
     typefaceFontName=Bold
     outlineSettings {
      bApplyOutlineToDropShadows=false
      bMiteredCorners=false
      bSeparateFillAlpha=false
      outlineMaterial=null
      outlineSize=0
      outlineColor=<rgba 0 0 0 1>
     }
    }
   }
   slot {
    class=CanvasPanelSlot
    properties {bAutoSize=true, layoutData {alignment=<0 0>, anchors {maximum=<0.5 0.5>, minimum=<0.5 0.5>}, offsets=<ltrb -700.9609 431.4595 100 30>}}
   }
  }
  TextBlock_4 {
   widgetClassPath=&TextBlock
   parent=Graphics
   widget=TextBlock_4
   properties {
    text=lips tighten
    colorAndOpacity {colorUseRule=UseColor_Specified, specifiedColor=<rgba 1 1 1 0.5>}
    font {
     bForceMonospaced=false
     bMaterialIsStencil=false
     fontMaterial=null
     fontObject=&Roboto
     letterSpacing=0
     monospacedWidth=1
     size=11.25
     skewAmount=0
     typefaceFontName=Bold
     outlineSettings {
      bApplyOutlineToDropShadows=false
      bMiteredCorners=false
      bSeparateFillAlpha=false
      outlineMaterial=null
      outlineSize=0
      outlineColor=<rgba 0 0 0 1>
     }
    }
   }
   slot {
    class=CanvasPanelSlot
    properties {bAutoSize=true, layoutData {alignment=<0 0>, anchors {maximum=<0.5 0.5>, minimum=<0.5 0.5>}, offsets=<ltrb -588 431.4595 100 30>}}
   }
  }
  TextBlock_12 {
   widgetClassPath=&TextBlock
   parent=Graphics
   widget=TextBlock_12
   properties {
    text=chin raise
    colorAndOpacity {colorUseRule=UseColor_Specified, specifiedColor=<rgba 1 1 1 0.5>}
    font {
     bForceMonospaced=false
     bMaterialIsStencil=false
     fontMaterial=null
     fontObject=&Roboto
     letterSpacing=0
     monospacedWidth=1
     size=11.25
     skewAmount=0
     typefaceFontName=Bold
     outlineSettings {
      bApplyOutlineToDropShadows=false
      bMiteredCorners=false
      bSeparateFillAlpha=false
      outlineMaterial=null
      outlineSize=0
      outlineColor=<rgba 0 0 0 1>
     }
    }
   }
   slot {
    class=CanvasPanelSlot
    properties {bAutoSize=true, layoutData {alignment=<0 0>, anchors {maximum=<0.5 0.5>, minimum=<0.5 0.5>}, offsets=<ltrb -384.5413 378.6517 100 30>}}
   }
  }
  TextBlock_13 {
   widgetClassPath=&TextBlock
   parent=Graphics
   widget=TextBlock_13
   properties {
    text=mouth press
    colorAndOpacity {colorUseRule=UseColor_Specified, specifiedColor=<rgba 1 1 1 0.5>}
    font {
     bForceMonospaced=false
     bMaterialIsStencil=false
     fontMaterial=null
     fontObject=&Roboto
     letterSpacing=0
     monospacedWidth=1
     size=11.25
     skewAmount=0
     typefaceFontName=Bold
     outlineSettings {
      bApplyOutlineToDropShadows=false
      bMiteredCorners=false
      bSeparateFillAlpha=false
      outlineMaterial=null
      outlineSize=0
      outlineColor=<rgba 0 0 0 1>
     }
    }
   }
   slot {
    class=CanvasPanelSlot
    properties {bAutoSize=true, layoutData {alignment=<0 0>, anchors {maximum=<0.5 0.5>, minimum=<0.5 0.5>}, offsets=<ltrb -26.9609 307.4595 100 30>}}
   }
  }
  TextBlock_14 {
   widgetClassPath=&TextBlock
   parent=Graphics
   widget=TextBlock_14
   properties {
    text=lip shift
    colorAndOpacity {colorUseRule=UseColor_Specified, specifiedColor=<rgba 1 1 1 0.5>}
    font {
     bForceMonospaced=false
     bMaterialIsStencil=false
     fontMaterial=null
     fontObject=&Roboto
     letterSpacing=0
     monospacedWidth=1
     size=11.25
     skewAmount=0
     typefaceFontName=Bold
     outlineSettings {
      bApplyOutlineToDropShadows=false
      bMiteredCorners=false
      bSeparateFillAlpha=false
      outlineMaterial=null
      outlineSize=0
      outlineColor=<rgba 0 0 0 1>
     }
    }
   }
   slot {
    class=CanvasPanelSlot
    properties {bAutoSize=true, layoutData {alignment=<0 0>, anchors {maximum=<0.5 0.5>, minimum=<0.5 0.5>}, offsets=<ltrb 661.0391 -76.5405 100 30>}}
   }
  }
  TextBlock_15 {
   widgetClassPath=&TextBlock
   parent=Graphics
   widget=TextBlock_15
   properties {
    text="lips \r\ntowards \r\nteeth"
    colorAndOpacity {colorUseRule=UseColor_Specified, specifiedColor=<rgba 1 1 1 0.5>}
    font {
     bForceMonospaced=false
     bMaterialIsStencil=false
     fontMaterial=null
     fontObject=&Roboto
     letterSpacing=0
     monospacedWidth=1
     size=11.25
     skewAmount=0
     typefaceFontName=Bold
     outlineSettings {
      bApplyOutlineToDropShadows=false
      bMiteredCorners=false
      bSeparateFillAlpha=false
      outlineMaterial=null
      outlineSize=0
      outlineColor=<rgba 0 0 0 1>
     }
    }
   }
   slot {
    class=CanvasPanelSlot
    properties {bAutoSize=true, layoutData {alignment=<0 0>, anchors {maximum=<0.5 0.5>, minimum=<0.5 0.5>}, offsets=<ltrb 227.0391 -112.5405 100 30>}}
   }
  }
  TextBlock_16 {
   widgetClassPath=&TextBlock
   parent=Graphics
   widget=TextBlock_16
   properties {
    text=W TCH
    colorAndOpacity {colorUseRule=UseColor_Specified, specifiedColor=<rgba 1 1 1 0.5>}
    font {
     bForceMonospaced=false
     bMaterialIsStencil=false
     fontMaterial=null
     fontObject=&Roboto
     letterSpacing=0
     monospacedWidth=1
     size=15
     skewAmount=0
     typefaceFontName=Bold
     outlineSettings {
      bApplyOutlineToDropShadows=false
      bMiteredCorners=false
      bSeparateFillAlpha=false
      outlineMaterial=null
      outlineSize=0
      outlineColor=<rgba 0 0 0 1>
     }
    }
   }
   slot {
    class=CanvasPanelSlot
    properties {bAutoSize=true, layoutData {alignment=<0 0>, anchors {maximum=<0.5 0.5>, minimum=<0.5 0.5>}, offsets=<ltrb -134.9609 487.4595 100 30>}}
   }
  }
  TextBlock_17 {
   widgetClassPath=&TextBlock
   parent=Graphics
   widget=TextBlock_17
   properties {
    text=face GUI follow
    colorAndOpacity {colorUseRule=UseColor_Specified, specifiedColor=<rgba 1 1 1 0.5>}
    font {
     bForceMonospaced=false
     bMaterialIsStencil=false
     fontMaterial=null
     fontObject=&Roboto
     letterSpacing=0
     monospacedWidth=1
     size=11.25
     skewAmount=0
     typefaceFontName=Bold
     outlineSettings {
      bApplyOutlineToDropShadows=false
      bMiteredCorners=false
      bSeparateFillAlpha=false
      outlineMaterial=null
      outlineSize=0
      outlineColor=<rgba 0 0 0 1>
     }
    }
   }
   slot {
    class=CanvasPanelSlot
    properties {bAutoSize=true, layoutData {alignment=<0 0>, anchors {maximum=<0.5 0.5>, minimum=<0.5 0.5>}, offsets=<ltrb 341.0391 427.4595 100 7.7778>}}
   }
  }
  TextBlock_18 {
   widgetClassPath=&TextBlock
   parent=Graphics
   widget=TextBlock_18
   properties {
    text=eyes aim follow
    colorAndOpacity {colorUseRule=UseColor_Specified, specifiedColor=<rgba 1 1 1 0.5>}
    font {
     bForceMonospaced=false
     bMaterialIsStencil=false
     fontMaterial=null
     fontObject=&Roboto
     letterSpacing=0
     monospacedWidth=1
     size=11.25
     skewAmount=0
     typefaceFontName=Bold
     outlineSettings {
      bApplyOutlineToDropShadows=false
      bMiteredCorners=false
      bSeparateFillAlpha=false
      outlineMaterial=null
      outlineSize=0
      outlineColor=<rgba 0 0 0 1>
     }
    }
   }
   slot {
    class=CanvasPanelSlot
    properties {bAutoSize=true, layoutData {alignment=<0 0>, anchors {maximum=<0.5 0.5>, minimum=<0.5 0.5>}, offsets=<ltrb 341.0391 471.4595 100 7.7778>}}
   }
  }
  CTRL_Master {
   widgetClassPath=&EUW_Picker_Button
   bIsVariable=true
   parent=FN_MH_Face_CanvasPanel
   widget=CTRL_Master
   properties {
    draw Style=Square
    label Font Size=12
    label Text=Master Control
    normal Tint {colorUseRule=UseColor_Specified, specifiedColor=<rgba 0 0.0108 0.4156 1>}
   }
   slot {
    class=CanvasPanelSlot
    properties {layoutData {alignment=<0.5 0.5>, anchors {maximum=<0.5 0.5>, minimum=<0.5 0.5>}, offsets=<ltrb -329.9609 716.4595 150 50>}}
   }
  }
  CTRL_WrinkleMap {
   widgetClassPath=&EUW_Picker_Button
   bIsVariable=true
   parent=FN_MH_Face_CanvasPanel
   widget=CTRL_WrinkleMap
   properties {draw Style=Circle, label Font Size=12, label Text=WM, normal Tint {colorUseRule=UseColor_Specified, specifiedColor=<rgba 0.6667 0.5459 0 1>}}
   slot {
    class=CanvasPanelSlot
    properties {layoutData {alignment=<0.5 0.5>, anchors {maximum=<0.5 0.5>, minimum=<0.5 0.5>}, offsets=<ltrb -337.9609 632 50 50>}}
   }
  }
  SelectAll {
   widgetClassPath=&EUW_Picker_Button
   bIsVariable=true
   parent=FN_MH_Face_CanvasPanel
   widget=SelectAll
   properties {
    label Font Size=15
    label Text=Select All
    toolTipText=Selects all Face Rig Controls
    normal Tint {colorUseRule=UseColor_Specified, specifiedColor=<rgba 0.625 0.2072 0.0359 1>}
   }
   slot {class=CanvasPanelSlot, properties {layoutData {alignment=<0 0>, anchors {maximum=<0 0>, minimum=<0 0>}, offsets=<ltrb 952 112 150 35>}}}
  }
  UpperFace {
   widgetClassPath=&EUW_Picker_Button
   bIsVariable=true
   parent=FN_MH_Face_CanvasPanel
   widget=UpperFace
   properties {
    label Font Size=15
    label Text=Upper Face
    toolTipText=Selects Upper Face Region Controls
    normal Tint {colorUseRule=UseColor_Specified, specifiedColor=<rgba 0.625 0.4915 0 1>}
   }
   slot {class=CanvasPanelSlot, properties {layoutData {alignment=<0 0>, anchors {maximum=<0 0>, minimum=<0 0>}, offsets=<ltrb 952 280 150 35>}}}
  }
  LowerFace {
   widgetClassPath=&EUW_Picker_Button
   bIsVariable=true
   parent=FN_MH_Face_CanvasPanel
   widget=LowerFace
   properties {
    label Font Size=15
    label Text=Lower Face
    toolTipText=Selects Lower Face Region Controls
    normal Tint {colorUseRule=UseColor_Specified, specifiedColor=<rgba 0.625 0.4915 0 1>}
   }
   slot {class=CanvasPanelSlot, properties {layoutData {alignment=<0 0>, anchors {maximum=<0 0>, minimum=<0 0>}, offsets=<ltrb 952 336 150 35>}}}
  }
  zeroAll {
   widgetClassPath=&EditorUtilityButton
   parent=FN_MH_Face_CanvasPanel
   widget=zeroAll
   properties {widgetStyle {
     clickedSlateSound {resourceObject=null}
     disabled {
      bIsDynamicallyLoaded=false
      drawAs=RoundedBox
      imageType=NoImage
      mirroring=NoMirror
      resourceName=None
      resourceObject=null
      tiling=NoTile
      imageSize=<0 0>
      margin=<ltrb 0 0 0 0>
      outlineSettings {
       bUseBrushTransparency=false
       roundingType=FixedRadius
       width=1
       color {colorUseRule=UseColor_ColorTable, specifiedColor=<rgba 0 0 0 0>}
       cornerRadii=<4 4 4 4>
      }
      tintColor {colorUseRule=UseColor_Specified, specifiedColor=<rgba 0.0395 0.0395 0.0395 1>}
      uVRegion {bIsValid=false, max=<0 0>, min=<0 0>}
     }
     disabledForeground {colorUseRule=UseColor_ColorTable, specifiedColor=<rgba 0 0 0 0>}
     hovered {
      bIsDynamicallyLoaded=false
      drawAs=RoundedBox
      imageType=NoImage
      mirroring=NoMirror
      resourceName=None
      resourceObject=null
      tiling=NoTile
      imageSize=<32 32>
      margin=<ltrb 0 0 0 0>
      outlineSettings {
       bUseBrushTransparency=false
       roundingType=FixedRadius
       width=1
       color {colorUseRule=UseColor_ColorTable, specifiedColor=<rgba 0 0 0 0>}
       cornerRadii=<4 4 4 4>
      }
      tintColor {colorUseRule=UseColor_Specified, specifiedColor=<rgba 0.0953 0.0953 0.0953 1>}
      uVRegion {bIsValid=false, max=<0 0>, min=<0 0>}
     }
     hoveredForeground {colorUseRule=UseColor_ColorTable, specifiedColor=<rgba 0 0 0 0>}
     hoveredSlateSound {resourceObject=null}
     normal {
      bIsDynamicallyLoaded=false
      drawAs=RoundedBox
      imageType=NoImage
      mirroring=NoMirror
      resourceName=None
      resourceObject=null
      tiling=NoTile
      imageSize=<32 32>
      margin=<ltrb 0 0 0 0>
      outlineSettings {
       bUseBrushTransparency=false
       roundingType=FixedRadius
       width=1
       color {colorUseRule=UseColor_ColorTable, specifiedColor=<rgba 0 0 0 0>}
       cornerRadii=<4 4 4 4>
      }
      tintColor {colorUseRule=UseColor_Specified, specifiedColor=<rgba 0.624 0.2086 0.0356 1>}
      uVRegion {bIsValid=false, max=<0 0>, min=<0 0>}
     }
     normalForeground {colorUseRule=UseColor_ColorTable, specifiedColor=<rgba 0 0 0 0>}
     normalPadding=<ltrb 12 1.5 12 1.5>
     pressed {
      bIsDynamicallyLoaded=false
      drawAs=RoundedBox
      imageType=NoImage
      mirroring=NoMirror
      resourceName=None
      resourceObject=null
      tiling=NoTile
      imageSize=<32 32>
      margin=<ltrb 0 0 0 0>
      outlineSettings {
       bUseBrushTransparency=false
       roundingType=FixedRadius
       width=1
       color {colorUseRule=UseColor_ColorTable, specifiedColor=<rgba 0 0 0 0>}
       cornerRadii=<4 4 4 4>
      }
      tintColor {colorUseRule=UseColor_Specified, specifiedColor=<rgba 0.0284 0.0284 0.0284 1>}
      uVRegion {bIsValid=false, max=<0 0>, min=<0 0>}
     }
     pressedForeground {colorUseRule=UseColor_ColorTable, specifiedColor=<rgba 0 0 0 0>}
     pressedPadding=<ltrb 12 2.5 12 0.5>
     pressedSlateSound {resourceObject=null}
    }}
   slot {
    class=CanvasPanelSlot
    properties {layoutData {alignment=<0 0>, anchors {maximum=<0 0>, minimum=<0 0>}, offsets=<ltrb 952 168 150.0119 34.1656>}}
   }
  }
  TextBlock_0 {
   widgetClassPath=&TextBlock
   parent=zeroAll
   widget=TextBlock_0
   properties {
    text=Zero All
    font {
     bForceMonospaced=false
     bMaterialIsStencil=false
     fontMaterial=null
     fontObject=&Roboto
     letterSpacing=0
     monospacedWidth=1
     size=15
     skewAmount=0
     typefaceFontName=Bold
     outlineSettings {
      bApplyOutlineToDropShadows=false
      bMiteredCorners=false
      bSeparateFillAlpha=false
      outlineMaterial=null
      outlineSize=0
      outlineColor=<rgba 0 0 0 1>
     }
    }
   }
   slot {class=ButtonSlot}
  }
  zeroSelected {
   widgetClassPath=&EditorUtilityButton
   parent=FN_MH_Face_CanvasPanel
   widget=zeroSelected
   properties {widgetStyle {
     clickedSlateSound {resourceObject=null}
     disabled {
      bIsDynamicallyLoaded=false
      drawAs=RoundedBox
      imageType=NoImage
      mirroring=NoMirror
      resourceName=None
      resourceObject=null
      tiling=NoTile
      imageSize=<0 0>
      margin=<ltrb 0 0 0 0>
      outlineSettings {
       bUseBrushTransparency=false
       roundingType=FixedRadius
       width=1
       color {colorUseRule=UseColor_ColorTable, specifiedColor=<rgba 0 0 0 0>}
       cornerRadii=<4 4 4 4>
      }
      tintColor {colorUseRule=UseColor_Specified, specifiedColor=<rgba 0.0395 0.0395 0.0395 1>}
      uVRegion {bIsValid=false, max=<0 0>, min=<0 0>}
     }
     disabledForeground {colorUseRule=UseColor_ColorTable, specifiedColor=<rgba 0 0 0 0>}
     hovered {
      bIsDynamicallyLoaded=false
      drawAs=RoundedBox
      imageType=NoImage
      mirroring=NoMirror
      resourceName=None
      resourceObject=null
      tiling=NoTile
      imageSize=<32 32>
      margin=<ltrb 0 0 0 0>
      outlineSettings {
       bUseBrushTransparency=false
       roundingType=FixedRadius
       width=1
       color {colorUseRule=UseColor_ColorTable, specifiedColor=<rgba 0 0 0 0>}
       cornerRadii=<4 4 4 4>
      }
      tintColor {colorUseRule=UseColor_Specified, specifiedColor=<rgba 0.0953 0.0953 0.0953 1>}
      uVRegion {bIsValid=false, max=<0 0>, min=<0 0>}
     }
     hoveredForeground {colorUseRule=UseColor_ColorTable, specifiedColor=<rgba 0 0 0 0>}
     hoveredSlateSound {resourceObject=null}
     normal {
      bIsDynamicallyLoaded=false
      drawAs=RoundedBox
      imageType=NoImage
      mirroring=NoMirror
      resourceName=None
      resourceObject=null
      tiling=NoTile
      imageSize=<32 32>
      margin=<ltrb 0 0 0 0>
      outlineSettings {
       bUseBrushTransparency=false
       roundingType=FixedRadius
       width=1
       color {colorUseRule=UseColor_ColorTable, specifiedColor=<rgba 0 0 0 0>}
       cornerRadii=<4 4 4 4>
      }
      tintColor {colorUseRule=UseColor_Specified, specifiedColor=<rgba 0.624 0.2086 0.0356 1>}
      uVRegion {bIsValid=false, max=<0 0>, min=<0 0>}
     }
     normalForeground {colorUseRule=UseColor_ColorTable, specifiedColor=<rgba 0 0 0 0>}
     normalPadding=<ltrb 12 1.5 12 1.5>
     pressed {
      bIsDynamicallyLoaded=false
      drawAs=RoundedBox
      imageType=NoImage
      mirroring=NoMirror
      resourceName=None
      resourceObject=null
      tiling=NoTile
      imageSize=<32 32>
      margin=<ltrb 0 0 0 0>
      outlineSettings {
       bUseBrushTransparency=false
       roundingType=FixedRadius
       width=1
       color {colorUseRule=UseColor_ColorTable, specifiedColor=<rgba 0 0 0 0>}
       cornerRadii=<4 4 4 4>
      }
      tintColor {colorUseRule=UseColor_Specified, specifiedColor=<rgba 0.0284 0.0284 0.0284 1>}
      uVRegion {bIsValid=false, max=<0 0>, min=<0 0>}
     }
     pressedForeground {colorUseRule=UseColor_ColorTable, specifiedColor=<rgba 0 0 0 0>}
     pressedPadding=<ltrb 12 2.5 12 0.5>
     pressedSlateSound {resourceObject=null}
    }}
   slot {
    class=CanvasPanelSlot
    properties {layoutData {alignment=<0 0>, anchors {maximum=<0 0>, minimum=<0 0>}, offsets=<ltrb 952 224 150.2979 34.5088>}}
   }
  }
  TextBlock_24 {
   widgetClassPath=&TextBlock
   parent=zeroSelected
   widget=TextBlock_24
   properties {
    text=Zero selected
    font {
     bForceMonospaced=false
     bMaterialIsStencil=false
     fontMaterial=null
     fontObject=&Roboto
     letterSpacing=0
     monospacedWidth=1
     size=15
     skewAmount=0
     typefaceFontName=Bold
     outlineSettings {
      bApplyOutlineToDropShadows=false
      bMiteredCorners=false
      bSeparateFillAlpha=false
      outlineMaterial=null
      outlineSize=0
      outlineColor=<rgba 0 0 0 1>
     }
    }
   }
   slot {class=ButtonSlot}
  }
 ]
}
