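# Widget-less base class for Fortnite face animation picker Editor Utility Widgets, linking picker elements to face Control Rig controls in Sequencer.
# Finds the MR_FN_Face_V01 rig on the bound actor and derives Legacy/V1/V2 type from head skeleton via embedded Python.
# Matches widget names to control names after "/", then binds, selects, syncs and zeroes 1D/2D sliders.
# Helpers load type-specific pickers and recreate the rig to switch viewport faceboards.
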EditorUtilityWidgetBlueprint /AnimInEdTools/EpicControlRig/Face/EUW_FN_Face_Base
graphs [
 Categorize Face Ctrls {nodes [
   n1 = Categorize Face Ctrls!() -> (then, Controls:Array of Widget Object References) {is=title K2Node_FunctionEntry}
   > n2 = For Each Loop(Array:Array of Widget Object References=@n1) -> (Array Element:Widget Object Reference)
   Return Node!(
    execute=@n2.Completed
    1D Sliders:Array of EUW Picker 1DSlider Object References=Get local 1D sliders() -> (local 1D sliders:Array of EUW Picker 1DSlider Object References)
    2D Sliders:Array of EUW Picker 2DSlider Object References=Get local 2D sliders() -> (local 2D sliders:Array of EUW Picker 2DSlider Object References)
   )
   > Branch(Condition=Does Object Implement Interface!(
     TestObject:Object Reference=@n2
     Interface:Interface Interface=/AnimInEdTools/AnimationPicker/PickerElements/BPI_SelectablePickerElements.BPI_SelectablePickerElements_C
    ) -> (ReturnValue:bool))
   n3 = Make Array("[0]":Name=String To Name(InString=GetDisplayName(Object=@n2))) -> (Array:Array of Names)
   > Set Initial Vars!(
    execute:Exec
    self:Object Reference=@n2
    Control Rig:Control Rig Object Reference=Get Control Rig() -> (Control Rig:Control Rig Object Reference)
    Level Sequence:Level Sequence Object Reference=Get Level Sequence() -> (Level Sequence:Level Sequence Object Reference)
    Control Bindings:Map of Control Rig Object References to Animation Picker String Array Struct Structures=Make Map!(
     Key 0:Control Rig Object Reference=Get Control Rig()
     Value 0:Animation Picker String Array Struct Structure=Make Animation Picker String Array Struct!(ControlNames_6_#1:Array of Names=@n3) -> (Animation_Picker_StringArray_Struct:Animation Picker String Array Struct Structure) {is=title K2Node_MakeStruct}
    ) -> (Map:Map of Control Rig Object References to Animation Picker String Array Struct Structures) {is=title K2Node_MakeMap}
    Anim Picker EUW:Animation Picker Object Reference=Get Anim Picker EUW() -> (Anim Picker EUW:Animation Picker Object Reference)
    Level Actors:Array of Movie Scene Binding Proxy Structures
   ) -> (then:Exec) {is=title K2Node_Message}
   > n4 = Cast To EUW_Picker_1DSlider!(execute, Object=@n2) -> (then, CastFailed, AsEUW Picker 1DSlider:EUW Picker 1DSlider Object Reference) {is=title K2Node_DynamicCast}
   > Add(
    TargetArray:Array of EUW Picker 1DSlider Object References=Get local 1D sliders()
    NewItem:"EUW Picker 1DSlider Object Reference (by ref)"=@n4
   ) {is=title K2Node_CallArrayFunction}
   n5 = Cast To EUW_Picker_2DSlider!(execute=@n4.CastFailed, Object=@n2) -> (then, CastFailed, AsEUW Picker 2DSlider:EUW Picker 2DSlider Object Reference) {is=title K2Node_DynamicCast}
   > Add(
    TargetArray:Array of EUW Picker 2DSlider Object References=Get local 2D sliders()
    NewItem:"EUW Picker 2DSlider Object Reference (by ref)"=@n5
   )
  ]}
 EventGraph {links [], nodes []}
 Get Button Control Pairs {nodes [
   n1 = Get Button Control Pairs!() -> (
    then
    Main Canvas:Panel Widget Object Reference
    Control Rig:Control Rig Object Reference
    Excluded Buttons:Array of Widget Object References
   ) {is=title K2Node_FunctionEntry}
   > n2 = Sequence()
   > Clear!(
    execute
    TargetMap:Map of Widget Object References to Names=Get Local Button Ctrl Pairs() -> (Local Button Ctrl Pairs:Map of Widget Object References to Names)
   ) -> (then) {is=title K2Node_CallFunction}
   n3 = Get Control Rig()
   > n4 = For Each Loop with Break!(
    Exec:Exec
    Array:Array of Rig Element Key Structures=Get Controls!(
     self:Rig Hierarchy Object Reference=GetHierarchy!(self:Control Rig Object Reference=@n3) -> (ReturnValue:Rig Hierarchy Object Reference)
     bTraverse:bool=true
    ) -> (ReturnValue:Array of Rig Element Key Structures) {is=title K2Node_CallFunction}
    Break:Exec
   ) -> (LoopBody:Exec, Array Element_Type:ERigElementType Enum, Array Element_Name:Name, Array Index:int, Completed:Exec)
   > Add(
    TargetArray:Array of Names=Get Local All Control Names() -> (Local All Control Names:Array of Names)
    NewItem:"Name (by ref)"=@n4.Array Element_Name
   )
   n5 = For Each Loop(Exec=@n2.then_1, Array=GetAllChildren(self=@n1))
   Return Node!(execute=@n5.Completed, NewParam:Map of Widget Object References to Names=Get Local Button Ctrl Pairs())
   > n6 = Branch(Condition=Contains Item(TargetArray:Array of Widget Object References=@n1.Excluded Buttons, ItemToFind:"Widget Object Reference (by ref)"=@n5))
   n7 = Cast To EUW_Picker_Element!(execute=@n6.else, Object=@n5) -> (then, CastFailed, AsEUW Picker Element:EUW Picker Element Object Reference) {is=title K2Node_DynamicCast}
   > n8 = For Each Loop with Break!(Exec, Array:Array of Names=Get Local All Control Names(), Break) -> (LoopBody, Array Element:Name, Array Index, Completed)
   n9 = "To String (Name)"(InName=@n8)
   > Branch(Condition="Equal Exactly (String)"(A=Split(SourceString=@n9, InStr=/, SearchDir=FromStart).RightS, B=GetObjectName(Object=@n7)))
   > Add!(
    execute
    TargetMap:Map of Widget Object References to Names=Get Local Button Ctrl Pairs()
    Key:"Widget Object Reference (by ref)"=@n7
    Value:"Name (by ref)"=@n8
   ) -> (then) {is=title K2Node_CallFunction}
  ]}
 Get Face Picker by Type {nodes [
   Get Face Picker by Type!() -> (then, FaceType:String, OnFace:bool=~) {is=title K2Node_FunctionEntry}
   > n1 = Sequence()
   n2 = Get FaceType() -> (FaceType:String)
   > n3 = Branch(Condition="Equal Exactly (String)"(A=@n2, B=Legacy))
   > Set FacePickerToOpen(FacePickerToOpen:User Widget Class Reference=&/EUW_FN_Legacy_Face_AnimPicker.EUW_FN_Legacy_Face_AnimPicker_C) -> (Output_Get:User Widget Class Reference)
   n4 = Branch(execute=@n3.else, Condition="Equal Exactly (String)"(A=@n2, B=V1))
   > n5 = Branch(Condition=Get OnFace() -> (OnFace:bool))
   > Set FacePickerToOpen(FacePickerToOpen=&/EUW_FN_MH_Face_AnimPicker.EUW_FN_MH_Face_AnimPicker_C)
   Set FacePickerToOpen(execute=@n5.else, FacePickerToOpen=&EUW_FN_MH_Face_AnimPicker_C)
   n6 = Branch(execute=@n4.else, Condition="Equal Exactly (String)"(A=@n2, B=V2))
   > n7 = Branch(Condition=Get OnFace())
   > Set FacePickerToOpen(FacePickerToOpen=&EUW_FN_MH_Face_AnimPicker_C)
   Set FacePickerToOpen(execute=@n7.else, FacePickerToOpen=&EUW_FN_MH_Face_AnimPicker_C)
   Set FacePickerToOpen(execute=@n6.else)
   Return Node!(execute=@n1.then_1, FacePicker:User Widget Class Reference=Get FacePickerToOpen() -> (FacePickerToOpen:User Widget Class Reference))
  ]}
 Get Face Type {nodes [
   Get Face Type() {is=title K2Node_FunctionEntry}
   > n1 = Execute Python Script!(
    execute
    PythonScript="def get_head_type_from_mann(mann):\r\n    if not mann:\r\n        return \"\"\r\n\r\n    ## Map skeleton names to head types\r\n    SKEL_TO_HEAD_TYPE = {\r\n        'Fortnite_Base_Head_Export_Skeleton': 'Legacy',\r\n        'Fortnite_Base_3L_Head_Skeleton': 'V1',\r\n        'Fortnite_Base_3L_V2_Head_Skeleton': 'V2',\r\n    }\r\n    \r\n    ## get cid from mann\r\n    cid = mann.get_editor_property('AthenaCharacter')\r\n    \r\n    ## Find HEAD character part\r\n    head_cp = next(\r\n        (cp for cp in cid.base_character_parts\r\n         if cp.get_character_part_type().name == 'HEAD'),\r\n        None\r\n    )\r\n    \r\n    ## get head skeleton to determin the type\r\n    head_type = \"\"\r\n    if head_cp:\r\n        mesh = head_cp.get_skeletal_mesh()\r\n        skeleton = mesh.skeleton if mesh else None\r\n    \r\n        if skeleton:\r\n            skel_name = skeleton.get_name()\r\n    \r\n            ## Direct match on main skeleton\r\n            head_type = SKEL_TO_HEAD_TYPE.get(skel_name, \"\")\r\n    \r\n            ## Fallback: handle custom skeletons via compatible skeletons\r\n            if head_type is None:\r\n                compatible_skels = getattr(skeleton, \"compatible_skeletons\", None) or []\r\n                for c_skel in compatible_skels:\r\n                    c_name = c_skel.get_name()\r\n                    head_type = SKEL_TO_HEAD_TYPE.get(c_name)\r\n                    if head_type is not None:\r\n                        break\r\n\r\n    return head_type\r\n\r\n## get the head type\r\nface_type = get_head_type_from_mann(mann)\r\n"
    mann:Actor Object Reference=Get As Actor() -> (As Actor:Actor Object Reference)
   ) -> (then, ReturnValue, face_type:String) {inputs [mann], outputs [face_type]}
   > n2 = Sequence()
   > n3 = Branch(Condition=IsEmpty(InString=@n1.face_type))
   > Set local mann face type(local mann face type:String) -> (Output_Get:String)
   Set local mann face type(execute=@n3.else, local mann face type=@n1.face_type)
   n4 = Get local mann face type() -> (local mann face type:String)
   Return Node!(execute=@n2.then_1, face_type:String=@n4)
   PrintString(InString=Append(A=FaceType:, B=@n4))
  ]}
 Get Primary Panel {nodes [
   Get Primary Panel() {is=title K2Node_FunctionEntry}
   > Return Node!(execute, Panel:Panel Widget Object Reference)
  ]}
 Get Selected Slider Controls {nodes [
   Get Selected Slider Controls() {is=title K2Node_FunctionEntry}
   > n1 = Sequence()
   > Clear(TargetArray:Array of Names=Get Local Ctrl Names() -> (Local Ctrl Names:Array of Names)) {is=title K2Node_CallArrayFunction}
   > n2 = GetSections!(
    execute
    Track:Movie Scene Track Object Reference=Get Control Rig Track() -> (Control Rig Track:Movie Scene Control Rig Parameter Track Object Reference)
   ) -> (then, ReturnValue:Array of Movie Scene Section Object References)
   > n3 = For Each Loop with Break(Array:Array of Movie Scene Section Object References=@n2) -> (Array Element:Movie Scene Section Object Reference)
   > Branch(Condition=IsActive!(self:Movie Scene Section Object Reference=@n3) -> (ReturnValue:bool))
   > Set Local Ctrl Names(Local Ctrl Names:Array of Names=CurrentControlSelection!(self:Control Rig Object Reference=Get Control Rig()) -> (ReturnValue:Array of Names)) -> (Output_Get:Array of Names)
   Return Node!(execute=@n1.then_1, Control Names:Array of Names=Get Local Ctrl Names())
  ]}
 Select All Controls {nodes [
   Select All Controls() {is=title K2Node_FunctionEntry}
   > BeginTransaction(Context=AnimationPicker, Description="LOCTEXT(\"\\\"Select All\\\" button pressed\")")
   n1 = Get Control Rig()
   > n2 = For Each Loop!(Exec, Array:Array of Rig Element Key Structures=Get Controls(self=GetHierarchy(self=@n1), bTraverse=true)) -> (LoopBody, Array Element_Type:ERigElementType Enum, Array Element_Name:Name, Array Index, Completed)
   EndTransaction(execute=@n2.Completed)
   SelectControl!(
    execute=@n2.LoopBody
    self:Control Rig Object Reference=@n1
    InControlName:"Name (by ref)"=@n2.Array Element_Name
    bSelect:bool=true
    bSetupUndo:bool=false
   ) -> (then)
  ]}
 Select Multiple Sliders {nodes [
   n1 = Select Multiple Sliders!() -> (
    then
    Button:EUW Picker Element Object Reference
    2D Sliders to Select:Array of EUW Picker 2DSlider Object References
    1D Sliders to Select:Array of EUW Picker 1DSlider Object References
   ) {is=title K2Node_FunctionEntry}
   > Clear(TargetMap:Map of Control Rig Object References to Animation Picker String Array Struct Structures=Get Control Bindings!(self:EUW Picker Element Object Reference=@n1) -> (Control Bindings:Map of Control Rig Object References to Animation Picker String Array Struct Structures)) {is=title K2Node_CallFunction}
   > n2 = Sequence!(execute) -> (then_0, then_1, then_2:Exec)
   > Clear(TargetArray=Get Control Names() -> (Control Names:Array of Names)) {is=title K2Node_CallArrayFunction}
   > n3 = For Each Loop!(Exec, Array:Array of EUW Picker 2DSlider Object References=@n1.2D Sliders to Select) -> (LoopBody, Array Element:EUW Picker 2DSlider Object Reference, Array Index, Completed)
   > n4 = Values!(execute, TargetMap:Map of Control Rig Object References to Animation Picker String Array Struct Structures=Get Control Bindings(self=@n3)) -> (then, Values:Array of Animation Picker String Array Struct Structures)
   > n5 = For Each Loop!(Exec, Array:Array of Animation Picker String Array Struct Structures=@n4) -> (LoopBody, Array Element_ControlNames_6_#1:Array of Names, Array Index, Completed)
   > n6 = For Each Loop!(Exec, Array:Array of Names=@n5) -> (LoopBody, Array Element:Name, Array Index, Completed)
   > Add Unique(TargetArray:Array of Names=Get Control Names(), NewItem:"Name (by ref)"=@n6)
   n7 = For Each Loop(Exec=@n2.then_1, Array:Array of EUW Picker 1DSlider Object References=@n1.1D Sliders to Select) -> (Array Element:EUW Picker 1DSlider Object Reference)
   > n8 = Values(TargetMap=Get Control Bindings(self=@n7))
   > n9 = For Each Loop!(Exec, Array:Array of Animation Picker String Array Struct Structures=@n8) -> (LoopBody, Array Element_ControlNames_6_#1:Array of Names, Array Index, Completed)
   > n10 = For Each Loop!(Exec, Array:Array of Names=@n9) -> (LoopBody, Array Element:Name, Array Index, Completed)
   > Add Unique(TargetArray=Get Control Names(), NewItem=@n10)
   Set Controls for Primary Control Rig!(
    execute=@n2.then_2
    self:EUW Picker Element Object Reference=Get Button() -> (Button:EUW Picker Element Object Reference)
    Control Names:Array of Names=Get Control Names()
   ) -> (then) {is=title K2Node_CallFunction}
   > Return Node!(execute)
  ]}
 Select Multiple Sliders1D {nodes [
   n1 = Select Multiple Sliders1D!() -> (then, Button:EUW Picker Element Object Reference, Sliders to Select:Array of EUW Picker 1DSlider Object References) {is=title K2Node_FunctionEntry}
   > Clear(TargetMap=Get Control Bindings(self=@n1)) {is=title K2Node_CallFunction}
   > Clear(TargetArray=Get Control Names()) {is=title K2Node_CallArrayFunction}
   > n2 = For Each Loop(Array:Array of EUW Picker 1DSlider Object References=@n1.Sliders to Select) -> (Array Element:EUW Picker 1DSlider Object Reference)
   Set Controls for Primary Control Rig(execute=@n2.Completed, self=@n1, Control Names=Get Control Names())
   > Return Node()
   n3 = Values(execute=@n2.LoopBody, TargetMap=Get Control Bindings(self=@n2))
   > n4 = For Each Loop!(Exec, Array:Array of Animation Picker String Array Struct Structures=@n3) -> (LoopBody, Array Element_ControlNames_6_#1:Array of Names, Array Index, Completed)
   > n5 = For Each Loop!(Exec, Array:Array of Names=@n4) -> (LoopBody, Array Element:Name, Array Index, Completed)
   > Add Unique(TargetArray=Get Control Names(), NewItem=@n5)
   Select Multiple Buttons!(
    execute
    self:EUW Asset Picker Base Object Reference
    Button:EUW Picker Button Object Reference
    Buttons to Select:Array of EUW Picker Button Object References
   ) -> (then)
  ]}
 Set Face Core Variables {nodes [
   n1 = Sequence!(execute) -> (then_0, then_1, then_2, then_3:Exec, then_4:Exec)
   n2 = Get Control Rig()
   PrintString(execute=@n1.then_1, InString=Append(A="[Face Core] Control Rig:", B=GetDisplayName(Object=@n2)))
   n3 = Get Level Sequence()
   PrintString(execute=@n1.then_0, InString=Append(A="[Face Core] Level Sequence:", B=GetDisplayName(Object=@n3)))
   n4 = Get As Actor()
   PrintString(execute=@n1.then_4, InString=Append(A="[Face Core] Actor:", B=GetDisplayName(Object=@n4)))
   n5 = Get Modular Rig() -> (Modular Rig:Modular Rig Object Reference)
   PrintString(execute=@n1.then_3, InString=Append(A="[Face Core] Modular Rig:", B=GetDisplayName(Object=@n5)))
   n6 = Get Control Rig Track()
   PrintString(execute=@n1.then_2, InString=Append(A="[Face Core] Control Rig Track:", B=GetDisplayName(Object=@n6)))
   Set Face Core Variables() {is=title K2Node_FunctionEntry}
   > n7 = GetFocusedLevelSequence!(execute) -> (then, ReturnValue:Level Sequence Object Reference)
   > Set Level Sequence(Level Sequence:Level Sequence Object Reference=@n7) -> (Output_Get:Level Sequence Object Reference)
   > n8 = Sequence!(execute) -> (then_0, then_1, then_2)
   > n9 = GetControlRigs!(execute, LevelSequence:Level Sequence Object Reference=Get Level Sequence()) -> (then, ReturnValue:Array of Control Rig Sequencer Binding Proxy Structures)
   > n10 = For Each Loop(Array:Array of Control Rig Sequencer Binding Proxy Structures=@n9) -> (Array Element:Control Rig Sequencer Binding Proxy Structure)
   n11 = Break Control Rig Sequencer Binding Proxy!(ControlRigSequencerBindingProxy:"Control Rig Sequencer Binding Proxy Structure (by ref)"=@n10) -> (
    Proxy:Movie Scene Binding Proxy Structure
    ControlRig:Control Rig Object Reference
    Track:Movie Scene Control Rig Parameter Track Object Reference
   ) {is=title K2Node_BreakStruct}
   > Branch(Condition="Equal (Class)"(
     A=Get Class(Object=@n11.ControlRig) -> (ReturnValue:Control Rig Class Reference)
     B=/EpicControlRig/Templates/MR_FN_Face_V01.MR_FN_Face_V01_C
    ))
   > n12 = GetBindingID!(
    execute
    Sequence:Movie Scene Sequence Object Reference=Get Level Sequence()
    InBinding:"Movie Scene Binding Proxy Structure (by ref)"=Break Control Rig Sequencer Binding Proxy(ControlRigSequencerBindingProxy=Get Control Rig Proxy Binding() -> (Control Rig Proxy Binding:Control Rig Sequencer Binding Proxy Structure))
   ) -> (then, ReturnValue:Movie Scene Object Binding ID Structure)
   > n13 = GetBindingID(Sequence=Get Level Sequence(), InBinding=@n11)
   n14 = GetBoundObjects!(ObjectBinding:Movie Scene Object Binding ID Structure=@n13) -> (ReturnValue:Array of Object References)
   n15 = GetBoundObjects(ObjectBinding=@n12)
   > Branch(Condition="Equal (Object)"(A="Get (a copy)"(Array:Array of Object References=@n14) -> (Output:Object Reference), B="Get (a copy)"(Array=@n15)))
   > Set Control Rig Proxy(Control Rig Proxy:Movie Scene Binding Proxy Structure=@n11) -> (Output_Get:Movie Scene Binding Proxy Structure)
   > Set Control Rig(Control Rig:Control Rig Object Reference=@n11.ControlRig) -> (Output_Get:Control Rig Object Reference)
   > Set Control Rig Track(Control Rig Track:Movie Scene Control Rig Parameter Track Object Reference=@n11.Track) -> (Output_Get:Movie Scene Control Rig Parameter Track Object Reference)
   n16 = Cast To ModularRig!(execute=@n8.then_1, Object=Get Control Rig()) -> (then, CastFailed, AsModular Rig:Modular Rig Object Reference) {is=title K2Node_DynamicCast}
   > Set Modular Rig(Modular Rig:Modular Rig Object Reference=@n16) -> (Output_Get:Modular Rig Object Reference)
   > n17 = GetBindingID(
    Sequence=Get Level Sequence()
    InBinding=Break Control Rig Sequencer Binding Proxy(ControlRigSequencerBindingProxy=Get Control Rig Proxy Binding())
   )
   n18 = GetBoundObjects(ObjectBinding=@n17)
   > Branch(Condition="integer > integer"(A=Length(TargetArray:Array of Object References=@n18)))
   > n19 = Cast To Actor!(execute, Object="Get (a copy)"(Array=@n18)) -> (then, CastFailed, AsActor:Actor Object Reference) {is=title K2Node_DynamicCast}
   > Set As Actor(As Actor:Actor Object Reference=@n19) -> (Output_Get:Actor Object Reference)
   > n20 = Get Face Type!(execute, self) -> (then, face_type:String) {is=title K2Node_CallFunction}
   > Set Face Type(Face Type:String=@n20) -> (Output_Get:String)
   Reroute Node(InputPin=@n8.then_2)
  ]}
 Set Face Picker {nodes [
   n1 = Set Face Picker!() -> (then, EUW_FacePicker:User Widget Class Reference, Layout:Vertical Box Object Reference) {is=title K2Node_FunctionEntry}
   > n2 = Branch(Condition=IsValidClass!(Class:Object Class Reference=@n1) -> (ReturnValue:bool))
   n3 = Get Layout() -> (Layout:Vertical Box Object Reference)
   > n4 = Branch(Condition="integer > integer"(A=GetChildrenCount!(self:Panel Widget Object Reference=@n3) -> (ReturnValue:int=0)))
   n5 = Get Layout()
   > ClearChildren(self=@n5)
   > Return Node!(execute, Success:bool=false, Out Face Picker:EUW Asset Picker Base Object Reference)
   ClearChildren(execute=@n4.else, self=@n5)
   > n6 = Create Widget!(
    execute:Exec
    WorldContextObject:Object Reference
    Class:User Widget Class Reference=Get EUW_FacePicker() -> (EUW_FacePicker:User Widget Class Reference)
    OwningPlayer:Player Controller Object Reference
   ) -> (then:Exec, ReturnValue:User Widget Object Reference)
   > n7 = Cast To EUW_Asset_Picker_Base!(execute, Object=@n6) -> (then, CastFailed, AsEUW Asset Picker Base:EUW Asset Picker Base Object Reference) {is=title K2Node_DynamicCast}
   > Set FacePickerBase(FacePickerBase:EUW Asset Picker Base Object Reference=@n7) -> (Output_Get:EUW Asset Picker Base Object Reference)
   n8 = Get FacePickerBase() -> (FacePickerBase:EUW Asset Picker Base Object Reference)
   > AddChild(self=Get Layout(), Content=@n8)
   > SetSize!(
    execute
    self:Vertical Box Slot Object Reference=SlotAsVerticalBoxSlot(Widget=@n8)
    InSize:Slate Child Size Structure=Make Slate Child Size!(Value:float=1, SizeRule:ESlateSizeRule Enum=Fill) -> (SlateChildSize:Slate Child Size Structure) {is=title K2Node_MakeStruct}
   ) -> (then)
   > Widget has loaded!(
    execute:Exec
    self:Object Reference=Get FacePickerBase()
    Control Rig Proxy Binding:Control Rig Sequencer Binding Proxy Structure=Get Control Rig Proxy Binding()
    Animatable Creature:Animatable Creature Object Reference=Get Animatable Creature() -> (Animatable Creature:Animatable Creature Object Reference)
    Anim Picker EUW:Animation Picker Object Reference=Get Anim Picker EUW()
   ) -> (then:Exec) {is=title K2Node_Message}
   n9 = Get Anim Picker EUW()
   > Update All Picker Elements Array(self:Animation Picker Object Reference=@n9)
   > SetCheckedState!(
    execute
    self:Check Box Object Reference=Get Auto_CheckBox!(self:Animation Picker Object Reference=@n9) -> (Auto_CheckBox:Editor Utility Check Box Object Reference)
    InCheckedState:ECheckBoxState Enum=Unchecked
   ) -> (then)
   n10 = Get EUW_FacePicker()
   > PrintString(InString=Append(A="Face picker is set to ", B=Get Class Display Name(Class=@n10)))
   > Return Node(Success=true, Out Face Picker=Get FacePickerBase())
   LogString(execute=@n7.CastFailed, InString=Face Rig is not a child of Asset Picker Base)
   > Return Node(Success=false)
   LogString(execute=@n2.else, InString=Append(A=EUW Face Picker is not valid:, B=Get Class Display Name(Class=@n1)))
   > Return Node(Success=false)
   Show Message Dialog(
    Title=MakeLiteralText(Value="LOCTEXT(\"Cannot find Facial Rig\")")
    Message=MakeLiteralText(Value="LOCTEXT(\"Please add a Legacy or 3L face ControlRig\")")
    MessageType=Ok
    DefaultValue=No
    MessageCategory=Warning
   )
  ]}
 Set Mann Faceboard {nodes [
   Set Mann Faceboard!() -> (then, face type:String, onFace:bool=~) {is=title K2Node_FunctionEntry}
   > n1 = Sequence!(execute) -> (then_0, then_1)
   > n2 = Branch(Condition=Get onFace() -> (onFace:bool))
   > Set local face type map value(local face type map value:String=Append(A=Get face type() -> (face type:String), B=" OnFace")) -> (Output_Get:String)
   Set local face type map value(execute=@n2.else, local face type map value=Get face type())
   n3 = Get local face type map value() -> (local face type map value:String)
   RecreateControlRigWithNewSettings!(
    execute=@n1.then_1
    InTrack:Movie Scene Control Rig Parameter Track Object Reference=Get Control Rig Track()
    InSettingsToOverride:Map of Names to Strings=Make Map(Key 0:Name=FaceCtrlType, Value 0:String=@n3) -> (Map:Map of Names to Strings)
   ) -> (then, ReturnValue:bool)
   > PrintString(InString=Append(A="Face GUI in the viewport is set to ", B=@n3))
  ]}
 Set Zero 1D Slider {nodes [
   n1 = Set Zero 1D Slider!() -> (then, Picker 1D Slider:EUW Picker 1DSlider Object Reference, Control Name:Name) {is=title K2Node_FunctionEntry}
   > Zero Out Slider(self:EUW Picker 1DSlider Object Reference=@n1)
   > SetLocalControlRigFloat!(
    execute
    LevelSequence:Level Sequence Object Reference=Get Level Sequence()
    ControlRig:Control Rig Object Reference=Get Control Rig()
    ControlName:Name=Get Control Name() -> (Control Name:Name=None)
    Frame:Frame Number Structure
    Value:float=0
    TimeUnit:EMovieSceneTimeUnit Enum=DisplayRate
    bSetKey:bool=false
   ) -> (then)
  ]}
 Set Zero 2D Slider {nodes [
   n1 = Set Zero 2D Slider!() -> (then, Picker 2D Slider:EUW Picker 2DSlider Object Reference, Control Name:Name) {is=title K2Node_FunctionEntry}
   > Zero Out Slider(self:EUW Picker 2DSlider Object Reference=@n1)
   > SetLocalControlRigVector2D!(
    execute
    LevelSequence:Level Sequence Object Reference=Get Level Sequence()
    ControlRig:Control Rig Object Reference=Get Control Rig()
    ControlName:Name=Get Control Name() -> (Control Name=None)
    Frame:Frame Number Structure
    Value:Vector 2D Structure
    TimeUnit:EMovieSceneTimeUnit Enum=DisplayRate
    bSetKey:bool=false
   ) -> (then)
  ]}
 Set Zero Sliders {nodes [
   Set Zero Sliders!() -> (then, Button Ctrl Pairs:Map of Widget Object References to Names, Selected Ctrl Only:bool=~) {is=title K2Node_FunctionEntry}
   > n1 = Sequence()
   n2 = "For Each Loop (Map)"!(
    execute:Exec=@n1.then_1
    MapPin:Map of Widget Object References to Names=Get Button Ctrl Pairs() -> (Button Ctrl Pairs:Map of Widget Object References to Names)
    BreakPin:Exec
   ) -> (then:Exec, KeyPin:Widget Object Reference, ValuePin:Name, CompletedPin:Exec)
   > n3 = Sequence()
   > n4 = Branch(Condition=Get Selected Ctrl Only() -> (Selected Ctrl Only:bool))
   > n5 = Branch(Condition=Contains Item(
     TargetArray:Array of Names=Get Local Selected Controls() -> (Local Selected Controls:Array of Names)
     ItemToFind:"Name (by ref)"=@n2.ValuePin
    ))
   > Set Local Continue(Local Continue:bool=true) -> (Output_Get:bool)
   Set Local Continue(execute=@n5.else, Local Continue=false)
   Set Local Continue(execute=@n4.else, Local Continue=true)
   Branch(execute=@n3.then_1, Condition=Get Local Continue() -> (Local Continue:bool))
   > n6 = Cast To EUW_Picker_1DSlider(Object=@n2)
   > Set Zero 1D Slider!(execute, self, Picker 1D Slider:EUW Picker 1DSlider Object Reference=@n6, Control Name:Name=@n2.ValuePin) -> (then) {is=title K2Node_CallFunction}
   n7 = Cast To EUW_Picker_2DSlider(execute=@n6.CastFailed, Object=@n2)
   > Set Zero 2D Slider!(execute, self, Picker 2D Slider:EUW Picker 2DSlider Object Reference=@n7, Control Name:Name=@n2.ValuePin) -> (then) {is=title K2Node_CallFunction}
   Branch(execute=@n1.then_0, Condition=Get Selected Ctrl Only())
   > n8 = Get Selected Slider Controls!(execute, self) -> (then, Control Names:Array of Names) {is=title K2Node_CallFunction}
   > Set Local Selected Controls(Local Selected Controls:Array of Names=@n8) -> (Output_Get:Array of Names)
  ]}
 Setup Face MR Naming For Buttons {nodes [
   n1 = Setup Face MR Naming For Buttons!() -> (
    then
    Main Canvas:Panel Widget Object Reference
    Control Rig:Control Rig Object Reference
    Excluded Buttons:Array of Widget Object References
   ) {is=title K2Node_FunctionEntry}
   > n2 = Sequence()
   > n3 = Get Button Control Pairs!(
    execute
    self
    Main Canvas:Panel Widget Object Reference=@n1
    Control Rig:Control Rig Object Reference=@n1.Control Rig
    Excluded Buttons:Array of Widget Object References=@n1.Excluded Buttons
   ) -> (then, NewParam:Map of Widget Object References to Names) {is=title K2Node_CallFunction}
   > n4 = "For Each Loop (Map)"(MapPin=@n3)
   > n5 = Cast To EUW_Picker_Element(Object=@n4)
   > Set Controls for Primary Control Rig(self=@n5, Control Names=Make Array("[0]"=@n4.ValuePin))
   > SetToolTipText!(execute, self:Widget Object Reference=@n5, InToolTipText:"Text (by ref)"="To Text (Name)"!(InName:Name=@n4.ValuePin) -> (ReturnValue:Text)) -> (then)
   Return Node!(execute=@n2.then_1, NewParam:Map of Widget Object References to Names=@n3)
  ]}
 Sync Controls to Buttons {nodes [
   n1 = Sync Controls to Buttons!() -> (then, Button Ctrl Pairs:Map of Widget Object References to Names) {is=title K2Node_FunctionEntry}
   > n2 = "For Each Loop (Map)"(MapPin=@n1)
   > n3 = Get Current Time!(execute, TimeUnit:EMovieSceneTimeUnit Enum=DisplayRate) -> (then, ReturnValue:Movie Scene Sequence Playback Params Structure) {is=title K2Node_CallFunction}
   > n4 = Cast To EUW_Picker_1DSlider(Object=@n2)
   n5 = Make Frame Number!(Value:int=Break Movie Scene Sequence Playback Params!(MovieSceneSequencePlaybackParams:"Movie Scene Sequence Playback Params Structure (by ref)"=@n3) -> (
     Frame_FrameNumber_Value:int
     Frame_SubFrame:float=~
     Time:float=~
     MarkedFrame:String
     Timecode:Timecode Structure
     PositionType:EMovieScenePositionType Enum
     UpdateMethod:EUpdatePositionMethod Enum
     bHasJumped:bool=~
    ) {is=title K2Node_BreakStruct}) -> (FrameNumber:Frame Number Structure) {is=title K2Node_MakeStruct}
   > n6 = GetLocalControlRigFloat!(
    execute
    LevelSequence:Level Sequence Object Reference=Get Level Sequence()
    ControlRig:Control Rig Object Reference=Get Control Rig()
    ControlName:Name=@n2.ValuePin
    Frame:Frame Number Structure=@n5
    TimeUnit:EMovieSceneTimeUnit Enum=DisplayRate
   ) -> (then, ReturnValue:float)
   > SetValue!(
    execute
    self:Slider Object Reference=Get Slider!(self:EUW Picker 1DSlider Object Reference=@n4) -> (Slider:Editor Utility Slider Object Reference)
    InValue:float=@n6
   ) -> (then)
   n7 = Cast To EUW_Picker_2DSlider(execute=@n4.CastFailed, Object=@n2)
   > n8 = GetLocalControlRigVector2D!(
    execute
    LevelSequence:Level Sequence Object Reference=Get Level Sequence()
    ControlRig:Control Rig Object Reference=Get Control Rig()
    ControlName:Name=@n2.ValuePin
    Frame:Frame Number Structure=@n5
    TimeUnit:EMovieSceneTimeUnit Enum=DisplayRate
   ) -> (then, ReturnValue:Vector 2D Structure)
   > SetValue(
    self:Synth 2DSlider Object Reference=Get Slider(self:EUW Picker 2DSlider Object Reference=@n7) -> (Slider:Synth 2DSlider Object Reference)
    InValue:Vector 2D Structure=@n8
   )
  ]}
 Update Sliders based on Controls {nodes [
   Update Sliders based on Controls!() -> (
    then
    1D Sliders:Array of EUW Picker 1DSlider Object References
    2D Sliders:Array of EUW Picker 2DSlider Object References
    Button Ctrl Pairs:Map of Widget Object References to Names
   ) {is=title K2Node_FunctionEntry}
   > n1 = Get Current Local Time!(execute, TimeUnit:EMovieSceneTimeUnit Enum=DisplayRate) -> (then, ReturnValue:Movie Scene Sequence Playback Params Structure) {is=title K2Node_CallFunction}
   > Set Current Frame Number(Current Frame Number:Frame Number Structure=Make Frame Number(Value=Break Movie Scene Sequence Playback Params(MovieSceneSequencePlaybackParams=@n1) -> (Frame_SubFrame=~, Time=~, bHasJumped=~))) -> (Output_Get:Frame Number Structure)
   > n2 = Sequence!(execute) -> (then_0, then_1, then_2:Exec)
   > n3 = For Each Loop(Array:Array of EUW Picker 1DSlider Object References=Get 1D Sliders() -> (1D Sliders:Array of EUW Picker 1DSlider Object References)) -> (Array Element:EUW Picker 1DSlider Object Reference)
   n4 = Find(TargetMap:Map of Widget Object References to Names=Get Button Ctrl Pairs(), Key:"Widget Object Reference (by ref)"=@n3) -> (Value:Name)
   > n5 = GetLocalControlRigFloat(
    LevelSequence=Get Level Sequence()
    ControlRig=Get Control Rig()
    ControlName=@n4
    Frame=Get Current Frame Number() -> (Current Frame Number:Frame Number Structure)
    TimeUnit=DisplayRate
   )
   > SetValue(
    self:Slider Object Reference=Get Slider(self:EUW Picker 1DSlider Object Reference=@n3) -> (Slider:Editor Utility Slider Object Reference)
    InValue:float=@n5
   )
   > Update Selection Display!(
    execute:Exec
    self:Object Reference=@n3
    Selected:bool=IsControlSelected!(self:Control Rig Object Reference=Get Control Rig(), InControlName:"Name (by ref)"=@n4) -> (ReturnValue:bool)
   ) -> (then:Exec, Out Selected:bool) {is=title K2Node_Message}
   n6 = For Each Loop(
    Exec=@n2.then_1
    Array:Array of EUW Picker 2DSlider Object References=Get 2D Sliders() -> (2D Sliders:Array of EUW Picker 2DSlider Object References)
   ) -> (Array Element:EUW Picker 2DSlider Object Reference)
   n7 = Find(TargetMap=Get Button Ctrl Pairs(), Key=@n6)
   > n8 = GetLocalControlRigVector2D(LevelSequence=Get Level Sequence(), ControlRig=Get Control Rig(), ControlName=@n7, Frame=Get Current Frame Number(), TimeUnit=DisplayRate)
   > SetValue(
    self:Synth 2DSlider Object Reference=Get Slider(self:EUW Picker 2DSlider Object Reference=@n6) -> (Slider:Synth 2DSlider Object Reference)
    InValue:Vector 2D Structure=Convert to Slider 2D Relative Value!(self:EUW Picker 2DSlider Object Reference=@n6, Value:Vector 2D Structure=@n8) -> (ReturnValue:Vector 2D Structure)
   )
   > Update Selection Display(self=@n6, Selected=IsControlSelected(self=Get Control Rig(), InControlName=@n7))
   Return Node!(execute=@n2.then_2)
  ]}
]
variables [
 As Actor {type=Actor, default=null}
 Control Rig {type=ControlRig, default=null}
 Control Rig Proxy {type=MovieSceneBindingProxy, default {bindingId=#2, sequence=null}}
 Control Rig Track {type=MovieSceneControlRigParameterTrack, default=null}
 Face Type {type=string, default=""}
 Level Sequence {type=LevelSequence, default=null}
 Modular Rig {type=ModularRig, default=null}
 PickerElementRef {type=EUW_Picker_Element_C, default=null}
]
widgets {
 info {parentClass=/AnimInEdTools/AnimationPicker/EUW_Asset_Picker_Base.EUW_Asset_Picker_Base_C, rootWidgetClass=null, widgetCount=0}
 widgets []
}
